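# One-shot stereo MetaSound source that plays a Sonic Spear death explosion sound.
# On play, a random wave from a three-variant array feeds a wave player, whose channels are boosted by a shared Multiplier before output.
# Source settings add pickaxe-impact attenuation, a concurrency override and SonicSpear control-bus volume modulation.
# An audio bus send feeds what appears, by name, to be a music high-pass sidechain.
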MetaSoundSource /SonicSpearGameplay/Assets/Audio/MSS_SonicSpear_DeathMetal
inputs [
 Death Explo {type=WaveAsset:Array}
 Multiplier {type=Float}
 UE.Source.OnPlay {type=Trigger}
]
interfaces [UE.OutputFormat.Stereo:1.0, UE.Source.OneShot:1.0, UE.Source:1.1]
outputs [
 "UE.OutputFormat.Stereo.Audio:0" {type=Audio}
 "UE.OutputFormat.Stereo.Audio:1" {type=Audio}
 UE.Source.OneShot.OnFinished {type=Trigger}
]
pages [Default {
  input_defaults {
   Multiplier=1.8
   UE.Source.OnPlay=null
   Death Explo [
    /SonicSpearGameplay/Assets/Audio/wav/SonicSpear_Death_01
    /SonicSpearGameplay/Assets/Audio/wav/SonicSpear_Death_02
    /SonicSpearGameplay/Assets/Audio/wav/SonicSpear_Death_03
   ]
  }
  nodes [
   ArrayRandomGet_0 = Array.Random Get:WaveAsset:Array(In Array=^Death Explo, Next=^UE.Source.OnPlay) {also {defaults {Weights []}}}
   WavePlayer_0 = UE.Wave Player:Stereo(Play=@ArrayRandomGet_0.On Next, Wave Asset=@ArrayRandomGet_0.Value) {also {version [1,1]}}
   Multiply_0 = UE.Multiply:Audio by Float(AdditionalOperands=^Multiplier, PrimaryOperand=@WavePlayer_0.Out Right)
   Multiply_1 = UE.Multiply:Audio by Float(AdditionalOperands=^Multiplier, PrimaryOperand=@WavePlayer_0.Out Left)
  ]
  links [
   {dst="@UE.OutputFormat.Stereo.Audio:1", src=Multiply_0.Out}
   {dst="@UE.OutputFormat.Stereo.Audio:0", src=Multiply_1.Out}
   {dst="@UE.Source.OneShot.OnFinished", src=WavePlayer_0.On Finished}
  ]
 }]
properties {
 assetImportData=&self:AssetImportData
 attenuationSettings=/Game/Sounds/Attenuation/Fort_Weapon_Pickaxe_Player_Impact_Att
 bOverrideConcurrency=true
 maxDistance=3512
 numChannels=2
 outputFormat=Stereo
 sampleRate=48000
 soundClassObject=&Fort_Impacts
 busSends [{
   audioBus=/Game/Sounds/Modulation/EnvelopeFollowers/HPF/AB_SideChain_FortMusic_HPF
   minSendDistance=300
   sendLevel=0.8
   sourceBusSendLevelControlMethod=Linear
  }]
 concurrencyOverrides {maxCount=5, retriggerTime=0.1, volumeScale=0.6}
 modulationSettings {volumeRouting=Override, volumeModulationDestination {bEnableModulation=true, modulators [&CB_SonicSpear]}}
}
