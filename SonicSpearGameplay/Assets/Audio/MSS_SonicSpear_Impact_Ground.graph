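# One-shot stereo MetaSound for the Sonic Spear's ground impact.
# On play it layers random close, distant and splat waves, crossfading close to distant by attenuation distance and lowering the splat's ladder-filter cutoff with distance.
# Both layers are mixed, and the sound finishes after all three players end.
# Golden and Size inputs only feed debug prints; a bus send feeds the music sidechain.
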MetaSoundSource /SonicSpearGameplay/Assets/Audio/MSS_SonicSpear_Impact_Ground
inputs [
 Close {type=WaveAsset:Array}
 Distant {type=WaveAsset:Array}
 Golden {type=Bool}
 Size {type=Float}
 Splat {type=WaveAsset:Array}
 UE.Attenuation.Distance {type=Float}
 UE.Source.OnPlay {type=Trigger}
]
interfaces [UE.Attenuation:1.0, UE.OutputFormat.Stereo:1.0, UE.Source.OneShot:1.0, UE.Source:1.1]
outputs [
 "UE.OutputFormat.Stereo.Audio:0" {type=Audio}
 "UE.OutputFormat.Stereo.Audio:1" {type=Audio}
 UE.Source.OneShot.OnFinished {type=Trigger}
]
pages [Default {
  input_defaults {
   Golden=null
   Size=0
   UE.Attenuation.Distance=0
   UE.Source.OnPlay=null
   Close [
    /SonicSpearGameplay/Assets/Audio/wav/SonicSpear_Impact_Ground_01
    /SonicSpearGameplay/Assets/Audio/wav/SonicSpear_Impact_Ground_02
    /SonicSpearGameplay/Assets/Audio/wav/SonicSpear_Impact_Ground_03
    /SonicSpearGameplay/Assets/Audio/wav/SonicSpear_Impact_Ground_04
   ]
   Distant [
    /SonicSpearGameplay/Assets/Audio/wav/SonicSpear_Impact_Ground_Dist_01
    /SonicSpearGameplay/Assets/Audio/wav/SonicSpear_Impact_Ground_Dist_02
    /SonicSpearGameplay/Assets/Audio/wav/SonicSpear_Impact_Ground_Dist_03
    /SonicSpearGameplay/Assets/Audio/wav/SonicSpear_Impact_Ground_Dist_04
   ]
   Splat [
    /SonicSpearGameplay/Assets/Audio/wav/SonicSpear_Impact_Splat_01
    /SonicSpearGameplay/Assets/Audio/wav/SonicSpear_Impact_Splat_02
    /SonicSpearGameplay/Assets/Audio/wav/SonicSpear_Impact_Splat_03
    /SonicSpearGameplay/Assets/Audio/wav/SonicSpear_Impact_Splat_04
    /SonicSpearGameplay/Assets/Audio/wav/SonicSpear_Impact_Splat_05
   ]
  }
  nodes [
   n1 = MapRange.MapRange:Float(In=^UE.Attenuation.Distance) {also {defaults {In Range B=7000}}}
   ArrayRandomGet_0 = Array.Random Get:WaveAsset:Array(In Array=^Distant, Next=^UE.Source.OnPlay) {also {defaults {Weights []}}}
   WavePlayer_1 = UE.Wave Player:Stereo(Play=@ArrayRandomGet_0.On Next, Wave Asset=@ArrayRandomGet_0.Value) {also {version [1,1]}}
   ArrayRandomGet_1 = Array.Random Get:WaveAsset:Array(In Array=^Close, Next=^UE.Source.OnPlay) {also {defaults {Weights []}}}
   WavePlayer_2 = UE.Wave Player:Stereo(Play=@ArrayRandomGet_1.On Next, Wave Asset=@ArrayRandomGet_1.Value) {also {version [1,1]}}
   CrossfadeTriggerRouteAudio2_0 = "Crossfade.Trigger Route (Audio, 2):Audio"(Crossfade Value=@n1.Out Value, In 1=@WavePlayer_1.Out Right, In 0=@WavePlayer_2.Out Right)
   CrossfadeTriggerRouteAudio2_1 = "Crossfade.Trigger Route (Audio, 2):Audio"(Crossfade Value=@n1.Out Value, In 1=@WavePlayer_1.Out Left, In 0=@WavePlayer_2.Out Left)
   n2 = MapRange.MapRange:Float(In=^UE.Attenuation.Distance) {also {defaults {In Range B=7500, Out Range A=20000, Out Range B=800}}}
   ArrayRandomGet_2 = Array.Random Get:WaveAsset:Array(In Array=^Splat, Next=^UE.Source.OnPlay) {also {defaults {Weights []}}}
   WavePlayer_0 = UE.Wave Player:Stereo(Play=@ArrayRandomGet_2.On Next, Wave Asset=@ArrayRandomGet_2.Value) {also {version [1,1]}}
   LadderFilter_0 = UE.Ladder Filter:Audio(Cutoff Frequency=@n2.Out Value, In=@WavePlayer_0.Out Right)
   LadderFilter_1 = UE.Ladder Filter:Audio(Cutoff Frequency=@n2.Out Value, In=@WavePlayer_0.Out Left)
   AudioMixerAudioMixerStereo2_0 = "AudioMixer.Audio Mixer (Stereo, 2)"(In 1 R=@CrossfadeTriggerRouteAudio2_0.Out, In 1 L=@CrossfadeTriggerRouteAudio2_1.Out, In 0 R=@LadderFilter_0.Out, In 0 L=@LadderFilter_1.Out) {also {defaults {Gain 0=2, Gain 1=1.5}}}
   /FortMetasound/Patches/MSP_PrintBool(Input=^Golden, Trigger=^UE.Source.OnPlay) {also {defaults {Label="Sonic Spear Bool: "}}}
   /FortMetasound/Patches/PrintFloat(Input=^Size, Trigger=^UE.Source.OnPlay) {also {defaults {Label="Sonic Spear Size:   "}}}
   TriggerAccumulatorTriggerAccumulate3_0 = "TriggerAccumulator.Trigger Accumulate (3)"(In 1=@WavePlayer_0.On Finished, In 2=@WavePlayer_1.On Finished, In 0=@WavePlayer_2.On Finished)
  ]
  links [
   {dst="@UE.OutputFormat.Stereo.Audio:0", src=AudioMixerAudioMixerStereo2_0.Out L}
   {dst="@UE.OutputFormat.Stereo.Audio:1", src=AudioMixerAudioMixerStereo2_0.Out R}
   {dst="@UE.Source.OneShot.OnFinished", src=TriggerAccumulatorTriggerAccumulate3_0.Out}
  ]
 }]
properties {
 assetImportData=&self:AssetImportData
 attenuationSettings=&/ATT_SonicSpear_Impact
 bOverrideConcurrency=true
 maxDistance=132000
 numChannels=2
 outputFormat=Stereo
 sampleRate=48000
 soundClassObject=/Game/Sounds/FortSoundClassesAndModes/Fort_Events_Explosions
 busSends [{
   audioBus=/Game/Sounds/Modulation/EnvelopeFollowers/HPF/AB_SideChain_FortMusic_HPF
   minSendDistance=300
   sendLevel=0.8
   sourceBusSendLevelControlMethod=Linear
  }]
 concurrencyOverrides {maxCount=5, retriggerTime=0.1, volumeScale=0.6}
 modulationSettings {volumeRouting=Override, volumeModulationDestination {bEnableModulation=true, modulators [&CB_SonicSpear]}}
}
