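# Looping stereo MetaSound source for the Sonic Spear in flight, layering a falling-wind loop with incoming one-shots.
# On play, a random Close wave loops from a random start time and pitch, shaped by an attack envelope.
# AudioIncoming and attenuation-distance threshold checks trigger random pitch-varied Incoming waves, whose completion resets the threshold flag.
# Both layers are gained separately and mixed, with debug prints for both states.
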MetaSoundSource /SonicSpearGameplay/Assets/Audio/MSS_SonicSpear_InAir_Loop_04
inputs [
 AudioIncoming {type=Bool}
 Close {type=WaveAsset:Array}
 Incoming {type=WaveAsset:Array}
 IncomingMultiplier {type=Float}
 Multiplier {type=Float}
 UE.Attenuation.Distance {type=Float}
 UE.Source.OnPlay {type=Trigger}
]
interfaces [UE.Attenuation:1.0, UE.OutputFormat.Stereo:1.0, UE.Source:1.1]
outputs [
 "UE.OutputFormat.Stereo.Audio:0" {type=Audio}
 "UE.OutputFormat.Stereo.Audio:1" {type=Audio}
]
pages [Default {
  input_defaults {
   AudioIncoming="<MetaSoundBuilderResult.SUCCEEDED: 0>"
   IncomingMultiplier=3
   Multiplier=5
   UE.Attenuation.Distance=12400.001
   UE.Source.OnPlay=null
   Close [/SonicSpearGameplay/Assets/Audio/wav/SonicSpear_Falling_Loop_04]
   Incoming [/SonicSpearGameplay/Assets/Audio/wav/SonicSpear_Incoming_01, /SonicSpearGameplay/Assets/Audio/wav/SonicSpear_Incoming_02]
  }
  nodes [
   ArrayRandomGet_0 = Array.Random Get:WaveAsset:Array(In Array=^Close, Next=^UE.Source.OnPlay)
   n1 = UE.RandomTime(Next=@ArrayRandomGet_0.On Next) {also {defaults {Max=8, Min=1}}}
   ADSREnvelopeADSREnvelope_0 = ADSR Envelope.ADSR Envelope:Audio(Trigger Attack=@n1.On Next) {also {defaults {Attack Time=0.2, Sustain Level=1}}}
   n2 = UE.RandomFloat(Next=@ADSREnvelopeADSREnvelope_0.On Attack Triggered) {also {defaults {Max=1.1, Min=0.9}}}
   WavePlayer_1 = UE.Wave Player:Stereo(
    Play=@n2.On Next
    Pitch Shift=@n2.Value
    Start Time=@n1.Value
    Wave Asset=UE.Reroute(Value=UE.Reroute(Value=@ArrayRandomGet_0.Value).Value).Value
   ) {also {defaults {Loop="<MetaSoundBuilderResult.SUCCEEDED: 0>"}, version [1,1]}}
   n3 = VariableDeferredAccessor.Trigger(Variable=InitVariable.Trigger().Variable)
   n4 = Value.Value:Bool(Set=UE.TriggerOnThreshold:Float(In=^UE.Attenuation.Distance) {also {defaults {Threshold=3200, Type=1}}}.Out, Reset=@n3.Value)
   TriggerOnValueChange_0 = UE.Trigger On Value Change:Bool(Value=@n4.Output Value)
   TriggerOnValueChange_1 = UE.Trigger On Value Change:Bool(Value=^AudioIncoming)
   n5 = InitVariable.Bool()
   TriggerAnyTriggerAny2_0 = "TriggerAny.Trigger Any (2)"(
    In 1=TriggerCompare.Clamp:Bool(B=^AudioIncoming, Compare=@TriggerOnValueChange_0.Trigger, A=@n4.Output Value).True
    In 0=TriggerCompare.Clamp:Bool(A=^AudioIncoming, Compare=@TriggerOnValueChange_1.Trigger, B=VariableAccessor.Bool(Variable=@n5.Variable).Value).True
   )
   ArrayRandomGet_1 = Array.Random Get:WaveAsset:Array(In Array=^Incoming, Next=@TriggerAnyTriggerAny2_0.Out) {also {defaults {Weights []}}}
   n6 = UE.RandomFloat(Next=@ArrayRandomGet_1.On Next) {also {defaults {Max=1.1, Min=0.9}}}
   WavePlayer_0 = UE.Wave Player:Stereo(Wave Asset=@ArrayRandomGet_1.Value, Play=@n6.On Next, Pitch Shift=@n6.Value) {also {version [1,1]}}
   AudioMixerAudioMixerStereo2_0 = "AudioMixer.Audio Mixer (Stereo, 2)"(
    Gain 1=^IncomingMultiplier
    In 0 R=UE.Multiply:Audio by Float(
     AdditionalOperands=^Multiplier
     PrimaryOperand=UE.Multiply:Audio(PrimaryOperand=@ADSREnvelopeADSREnvelope_0.Out Envelope, AdditionalOperands=@WavePlayer_1.Out Right).Out
    ).Out
    In 0 L=UE.Multiply:Audio by Float(
     AdditionalOperands=^Multiplier
     PrimaryOperand=UE.Multiply:Audio(PrimaryOperand=@ADSREnvelopeADSREnvelope_0.Out Envelope, AdditionalOperands=@WavePlayer_1.Out Left).Out
    ).Out
    In 1 L=@WavePlayer_0.Out Left
    In 1 R=@WavePlayer_0.Out Right
   )
   /FortMetasound/Patches/MSP_PrintBool(Trigger=@n4.On Set, Input=@n4.Output Value) {also {defaults {Label="MSS: Inside Threshold:  "}}}
   &MSP_PrintBool(Input=^AudioIncoming, Trigger=@TriggerOnValueChange_1.Trigger) {also {defaults {Label="MSS: Incoming:  "}}}
   VariableMutator.Trigger(Variable=@n3.Variable, Value=@WavePlayer_0.On Finished)
  ]
  links [
   {dst="@UE.OutputFormat.Stereo.Audio:0", src=AudioMixerAudioMixerStereo2_0.Out L}
   {dst="@UE.OutputFormat.Stereo.Audio:1", src=AudioMixerAudioMixerStereo2_0.Out R}
  ]
 }]
properties {
 assetImportData=&self:AssetImportData
 attenuationSettings=&/ATT_SonicSpear_FallingWind
 bLooping=true
 duration=10000
 maxDistance=30030
 numChannels=2
 outputFormat=Stereo
 sampleRate=48000
 soundClassObject=&Fort_SFX_InGame
 concurrencySet [&/CONC_InAirLoops]
 modulationSettings {volumeRouting=Override, volumeModulationDestination {bEnableModulation=true, modulators [&CB_SonicSpear]}}
}
