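# One-shot stereo MetaSound for the Sonic Spear's spawn drop, layering random fall and rift samples with a distant variant.
# The rift and distant layers start on play and the close fall follows after a short delay, with random pitch on close and rift.
# The close-rift mix crossfades to the distant layer by attenuation distance, and the source finishes once all three players end.
# It also feeds a music sidechain bus.
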MetaSoundSource /SonicSpearGameplay/Assets/Audio/MSS_SonicSpear_Spawn_Drop
inputs [
 Close {type=WaveAsset:Array}
 CloseMultiplier {type=Float}
 Distant {type=WaveAsset:Array}
 RiftLayer {type=WaveAsset:Array}
 UE.Attenuation.Distance {type=Float}
 UE.Source.OnPlay {type=Trigger}
]
interfaces [UE.Attenuation:1.0, UE.OutputFormat.Stereo:1.0, UE.Source.OneShot:1.0, UE.Source:1.1]
outputs [
 "UE.OutputFormat.Stereo.Audio:0" {type=Audio}
 "UE.OutputFormat.Stereo.Audio:1" {type=Audio}
 UE.Source.OneShot.OnFinished {type=Trigger}
]
pages [Default {
  input_defaults {
   CloseMultiplier=1.4
   UE.Attenuation.Distance=0
   UE.Source.OnPlay=null
   Close [
    /SonicSpearGameplay/Assets/Audio/wav/SonicSpear_SpawnFall_01
    /SonicSpearGameplay/Assets/Audio/wav/SonicSpear_SpawnFall_02
    /SonicSpearGameplay/Assets/Audio/wav/SonicSpear_SpawnFall_03
    /SonicSpearGameplay/Assets/Audio/wav/SonicSpear_SpawnFall_04
    /SonicSpearGameplay/Assets/Audio/wav/SonicSpear_SpawnFall_05
    /SonicSpearGameplay/Assets/Audio/wav/SonicSpear_SpawnFall_06
    /SonicSpearGameplay/Assets/Audio/wav/SonicSpear_SpawnFall_07
   ]
   Distant [
    /SonicSpearGameplay/Assets/Audio/wav/SonicSpear_SpawnFall_Dist_01
    /SonicSpearGameplay/Assets/Audio/wav/SonicSpear_SpawnFall_Dist_02
    /SonicSpearGameplay/Assets/Audio/wav/SonicSpear_SpawnFall_Dist_03
    /SonicSpearGameplay/Assets/Audio/wav/SonicSpear_SpawnFall_Dist_04
   ]
   RiftLayer [
    /SonicSpearGameplay/Assets/Audio/wav/SonicSpear_SpawnFall_RiftLayer_01
    /SonicSpearGameplay/Assets/Audio/wav/SonicSpear_SpawnFall_RiftLayer_02
    /SonicSpearGameplay/Assets/Audio/wav/SonicSpear_SpawnFall_RiftLayer_03
    /SonicSpearGameplay/Assets/Audio/wav/SonicSpear_SpawnFall_RiftLayer_04
   ]
  }
  nodes [
   ArrayRandomGet_0 = Array.Random Get:WaveAsset:Array(In Array=^RiftLayer, Next=^UE.Source.OnPlay) {also {defaults {Shared State Behavior=0, Weights []}}}
   n1 = UE.RandomFloat(Reset=^UE.Source.OnPlay, Next=@ArrayRandomGet_0.On Next) {also {defaults {Max=2, Min=-5}}}
   WavePlayer_2 = UE.Wave Player:Stereo(Wave Asset=@ArrayRandomGet_0.Value, Play=@n1.On Next, Pitch Shift=@n1.Value) {also {version [1,1]}}
   TriggerDelay_0 = UE.Trigger Delay(In=^UE.Source.OnPlay) {also {defaults {Delay Time=0.3}}}
   ArrayRandomGet_1 = Array.Random Get:WaveAsset:Array(In Array=^Close, Next=@TriggerDelay_0.Out) {also {defaults {Shared State Behavior=0}}}
   n2 = UE.RandomFloat(Next=@ArrayRandomGet_1.On Next) {also {defaults {Min=-3}}}
   WavePlayer_1 = UE.Wave Player:Stereo(Wave Asset=@ArrayRandomGet_1.Value, Play=@n2.On Next, Pitch Shift=@n2.Value) {also {version [1,1]}}
   AudioMixerAudioMixerStereo2_0 = "AudioMixer.Audio Mixer (Stereo, 2)"(
    Gain 0=^CloseMultiplier
    Gain 1=UE.RandomFloat(Reset=@n1.On Reset, Next=@WavePlayer_2.On Play) {also {defaults {Min=0.3}}}.Value
    In 0 L=@WavePlayer_1.Out Left
    In 0 R=@WavePlayer_1.Out Right
    In 1 L=@WavePlayer_2.Out Left
    In 1 R=@WavePlayer_2.Out Right
   )
   n3 = MapRange.MapRange:Float(In=^UE.Attenuation.Distance) {also {defaults {In Range B=80000}}}
   ArrayRandomGet_2 = Array.Random Get:WaveAsset:Array(In Array=^Distant, Next=^UE.Source.OnPlay) {also {defaults {Weights []}}}
   WavePlayer_0 = UE.Wave Player:Stereo(Play=@ArrayRandomGet_2.On Next, Wave Asset=@ArrayRandomGet_2.Value) {also {version [1,1]}}
   CrossfadeTriggerRouteAudio2_0 = "Crossfade.Trigger Route (Audio, 2):Audio"(In 0=@AudioMixerAudioMixerStereo2_0.Out L, Crossfade Value=@n3.Out Value, In 1=@WavePlayer_0.Out Left)
   CrossfadeTriggerRouteAudio2_1 = "Crossfade.Trigger Route (Audio, 2):Audio"(In 0=@AudioMixerAudioMixerStereo2_0.Out R, Crossfade Value=@n3.Out Value, In 1=@WavePlayer_0.Out Right)
   InitVariable.Int32()
   TriggerAccumulatorTriggerAccumulate3_0 = "TriggerAccumulator.Trigger Accumulate (3)"(In 2=@WavePlayer_0.On Finished, In 0=@WavePlayer_1.On Finished, In 1=@WavePlayer_2.On Finished)
  ]
  links [
   {dst="@UE.OutputFormat.Stereo.Audio:0", src=CrossfadeTriggerRouteAudio2_0.Out}
   {dst="@UE.OutputFormat.Stereo.Audio:1", src=CrossfadeTriggerRouteAudio2_1.Out}
   {dst="@UE.Source.OneShot.OnFinished", src=TriggerAccumulatorTriggerAccumulate3_0.Out}
  ]
 }]
properties {
 assetImportData=&self:AssetImportData
 attenuationSettings=&/ATT_SonicSpear_DropSpawn
 bOverrideConcurrency=true
 maxDistance=28000
 numChannels=2
 outputFormat=Stereo
 sampleRate=48000
 soundClassObject=/Game/Sounds/FortSoundClassesAndModes/Fort_Events_Explosions
 busSends [{
   audioBus=&AB_Sidechain_FortMusic_Vol
   maxSendDistance=50000
   maxSendLevel=0.7
   minSendDistance=8000
   minSendLevel=0.2
   sendLevel=1
   sourceBusSendLevelControlMethod=Linear
  }]
 concurrencyOverrides {maxCount=5, retriggerTime=0.08, volumeScale=0.6}
 modulationSettings {volumeRouting=Override, volumeModulationDestination {bEnableModulation=true, modulators [&CB_SonicSpear]}}
}
