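# One-shot stereo MetaSound source that plays the Sonic Spear thunder strike outro sound.
# On play it draws a random entry from the Thunder array, which defaults to a single wave, feeds it to a stereo wave player and signals finished when playback ends.
# It uses thunder-crack attenuation, the explosions sound class, a concurrency override and Sonic Spear volume modulation.
# A bus send feeds the music high-pass sidechain envelope follower.
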MetaSoundSource /SonicSpearGameplay/Assets/Audio/MSS_SonicSpear_ThunderStrike_Outro
inputs [
 Thunder {type=WaveAsset:Array}
 UE.Source.OnPlay {type=Trigger}
]
interfaces [UE.OutputFormat.Stereo:1.0, UE.Source.OneShot:1.0, UE.Source:1.1]
outputs [
 "UE.OutputFormat.Stereo.Audio:0" {type=Audio}
 "UE.OutputFormat.Stereo.Audio:1" {type=Audio}
 UE.Source.OneShot.OnFinished {type=Trigger}
]
pages [Default {
  input_defaults {UE.Source.OnPlay=null, Thunder [/SonicSpearGameplay/Assets/Audio/wav/SonicSpear_Storm_ThunderStrike_Outro]}
  nodes [
   ArrayRandomGet_0 = Array.Random Get:WaveAsset:Array(In Array=^Thunder, Next=^UE.Source.OnPlay) {also {defaults {Weights []}}}
   WavePlayer_0 = UE.Wave Player:Stereo(Play=@ArrayRandomGet_0.On Next, Wave Asset=@ArrayRandomGet_0.Value) {also {version [1,1]}}
  ]
  links [
   {dst="@UE.Source.OneShot.OnFinished", src=WavePlayer_0.On Finished}
   {dst="@UE.OutputFormat.Stereo.Audio:0", src=WavePlayer_0.Out Left}
   {dst="@UE.OutputFormat.Stereo.Audio:1", src=WavePlayer_0.Out Right}
  ]
 }]
properties {
 assetImportData=&self:AssetImportData
 attenuationSettings=&/ATT_SonicSpear_ThunderCrack
 bOverrideConcurrency=true
 maxDistance=13000
 numChannels=2
 outputFormat=Stereo
 sampleRate=48000
 soundClassObject=/Game/Sounds/FortSoundClassesAndModes/Fort_Events_Explosions
 busSends [{
   audioBus=/Game/Sounds/Modulation/EnvelopeFollowers/HPF/AB_SideChain_FortMusic_HPF
   minSendDistance=300
   sendLevel=0.8
   sourceBusSendLevelControlMethod=Linear
  }]
 concurrencyOverrides {maxCount=5, retriggerTime=0.1, volumeScale=0.6}
 modulationSettings {volumeRouting=Override, volumeModulationDestination {bEnableModulation=true, modulators [&CB_SonicSpear]}}
}
