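# Sound Cue for the SonicSpear UI alert, playing the HUD-appear stinger through a single wave player at boosted volume.
# It routes to the critical in-game UI sound class and partially sends to a music-volume sidechain bus, presumably to duck music.
# Its own concurrency override sets a retrigger time of several seconds to keep the stinger from repeating in quick succession.
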SoundCue /SonicSpearGameplay/Assets/Audio/SC_SonicSpear_UI_Alert
root=WavePlayer_0
nodes [{id=WavePlayer_0, properties {soundWaveAssetPtr=/SonicSpearGameplay/Assets/Audio/wav/SonicSpear_IP_HUD_Appear_Stinger}}]
properties {
 bOverrideConcurrency=true
 maxDistance=2097152
 soundClassObject=&Fort_UI_InGame_Critical
 volumeMultiplier=2
 busSends [{audioBus=&AB_Sidechain_FortMusic_Vol, minSendLevel=0, sendLevel=0.2, soundSourceBus=null}]
 concurrencyOverrides {retriggerTime=8}
}
