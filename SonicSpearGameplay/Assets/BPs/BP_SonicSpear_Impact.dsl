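; Sonic Spear impact actor that shows and clears donut debris: a donut mesh, crumbs, frosting blobs, sprinkles and a decal.
; The construction script creates dynamic materials for the decal and each debris type, then assigns them to their components.
; SpawnDebris shows the donut and switches off decal shrinking and debris WPO.
; RemoveDebris hides it, stamps game time into the materials and re-enables both, leaving the removal animation to shaders without Tick logic.
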(event Custom|SpawnDebris
  (Rendering|SetVisibility (Variables|Default|GetSM_SonicSpear_DropObject_Donut_A) true false)
  (Rendering|Material|SetScalarParameterValue (Variables|Default|GetDecalMatDynamic) "Shrinking Enabled" 0.0 :!class /Script/Engine.MaterialInstanceDynamic)
  (Rendering|Material|SetScalarParameterValue (Variables|Default|GetCrumbMatDynamic) "WPO Enabled" 0.0 :!class /Script/Engine.MaterialInstanceDynamic)
  (Rendering|Material|SetScalarParameterValue (Variables|Default|GetFrostingMatDynamic) "WPO Enabled" 0.0 :!class /Script/Engine.MaterialInstanceDynamic)
  (Rendering|Material|SetScalarParameterValue (Variables|Default|GetSprinkleMatDynamic) "WPO Enabled" 0.0 :!class /Script/Engine.MaterialInstanceDynamic))

(event Custom|RemoveDebris
  (bind _returnvalue (Utilities|Time|GetGameTimeinSeconds))
  (bind _crumbmatdynamic (Variables|Default|GetCrumbMatDynamic))
  (bind _frostingmatdynamic (Variables|Default|GetFrostingMatDynamic))
  (bind _sprinklematdynamic (Variables|Default|GetSprinkleMatDynamic))
  (bind _decalmatdynamic (Variables|Default|GetDecalMatDynamic))
  (Rendering|SetVisibility (Variables|Default|GetSM_SonicSpear_DropObject_Donut_A) false false)
  (Rendering|Material|SetScalarParameterValue _decalmatdynamic "CurrentGameTime" _returnvalue :!class /Script/Engine.MaterialInstanceDynamic)
  (Rendering|Material|SetScalarParameterValue _crumbmatdynamic "CurrentGameTime" _returnvalue :!class /Script/Engine.MaterialInstanceDynamic)
  (Rendering|Material|SetScalarParameterValue _frostingmatdynamic "CurrentGameTime" _returnvalue :!class /Script/Engine.MaterialInstanceDynamic)
  (Rendering|Material|SetScalarParameterValue _sprinklematdynamic "CurrentGameTime" _returnvalue :!class /Script/Engine.MaterialInstanceDynamic)
  (Rendering|Material|SetScalarParameterValue _decalmatdynamic "Shrinking Enabled" 1.0 :!class /Script/Engine.MaterialInstanceDynamic)
  (Rendering|Material|SetScalarParameterValue _crumbmatdynamic "WPO Enabled" 1.0 :!class /Script/Engine.MaterialInstanceDynamic)
  (Rendering|Material|SetScalarParameterValue _frostingmatdynamic "WPO Enabled" 1.0 :!class /Script/Engine.MaterialInstanceDynamic)
  (Rendering|Material|SetScalarParameterValue _sprinklematdynamic "WPO Enabled" 1.0 :!class /Script/Engine.MaterialInstanceDynamic))

(event EventBeginPlay)

(event Collision|EventActorBeginOverlap
  (params (OtherActor :type Actor%20Object%20Reference)))

(event EventTick
  (params (DeltaSeconds :type Float%20%28single-precision%29)))

(fn ConstructionScript
  (Utilities|FlowControl|Sequence
    (:then_0
      (bind _returnvalue (Rendering|Components|Decal|CreateDynamicMaterialInstance (Variables|Default|GetDecal)))
      (Variables|Default|SetDecalMatDynamic _returnvalue)
      (bind _returnvalue_2 (Rendering|Material|CreateDynamicMaterialInstance (Variables|Default|GetCrumbMat) "Crumb Dynamic" "None" :!class /Script/Engine.KismetMaterialLibrary))
      (Variables|Default|SetCrumbMatDynamic _returnvalue_2)
      (bind _returnvalue_3 (Rendering|Material|CreateDynamicMaterialInstance (Variables|Default|GetFrostingMat) "Frosting Dynamic" "None" :!class /Script/Engine.KismetMaterialLibrary))
      (Variables|Default|SetFrostingMatDynamic _returnvalue_3)
      (bind _returnvalue_4 (Rendering|Material|CreateDynamicMaterialInstance (Variables|Default|GetSprinkleMat) "Sprinkle Dynamic" "None" :!class /Script/Engine.KismetMaterialLibrary))
      (Variables|Default|SetSprinkleMatDynamic _returnvalue_4))
    (:then_1
      (Rendering|Components|Decal|SetDecalMaterial (Variables|Default|GetDecal) (Variables|Default|GetDecalMatDynamic))
      (for _array_element (Utilities|Array|MakeArray (Variables|Default|GetSM_SonicSpear_DropObject_Donut_Crumb_A4) (Variables|Default|GetSM_SonicSpear_DropObject_Donut_Crumb_A5) (Variables|Default|GetSM_SonicSpear_DropObject_Donut_Crumb_A2) (Variables|Default|GetSM_SonicSpear_DropObject_Donut_Crumb_A1) (Variables|Default|GetSM_SonicSpear_DropObject_Donut_Crumb_A) (Variables|Default|GetSM_SonicSpear_DropObject_Donut_Crumb_A6) (Variables|Default|GetSM_SonicSpear_DropObject_Donut_Crumb_A7) (Variables|Default|GetSM_SonicSpear_DropObject_Donut_Crumb_A3))
        (Rendering|Material|SetMaterial _array_element 0 (Variables|Default|GetCrumbMatDynamic)))
      (for _array_element_2 (Utilities|Array|MakeArray (Variables|Default|GetSM_SonicSpear_DropObject_Donut_Blob_A) (Variables|Default|GetSM_SonicSpear_DropObject_Donut_Blob_A1) (Variables|Default|GetSM_SonicSpear_DropObject_Donut_Blob_A2) (Variables|Default|GetSM_SonicSpear_DropObject_Donut_Blob_A3))
        (Rendering|Material|SetMaterial _array_element_2 0 (Variables|Default|GetFrostingMatDynamic)))
      (for _array_element_3 (Utilities|Array|MakeArray (Variables|Default|GetSM_SonicSpear_DropObject_Donut_Sprinkle_A1) (Variables|Default|GetSM_SonicSpear_DropObject_Donut_Sprinkle_A) (Variables|Default|GetSM_SonicSpear_DropObject_Donut_Sprinkle_A2))
        (Rendering|Material|SetMaterial _array_element_3 0 (Variables|Default|GetSprinkleMatDynamic))))))
