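# Material function that gives a material a stylized banded look by rewriting Base Color, blended by a Banding Mask input.
# The reflection vector's alignment with the sky atmosphere light is quantized into steps between min and max colors.
# Static switches add a hard-edged Fresnel color, sized by a Z-axis dot product, and a thresholded highlight, additive or replacing.
# Another switch chooses between fixed palette colors and ones derived from the original Base Color.
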MaterialFunction /SonicSpearGameplay/Assets/Materials/MF_ColorBanding
graph {nodes [
  Constant()
  n1 = FunctionInput(Preview=MakeMaterialAttributes!(
    BaseColor=Constant3Vector() {constant=<rgba 1 0 0.7837 1>}
    Metallic
    Specular
    Roughness
    Anisotropy
    EmissiveColor
    Opacity
    OpacityMask
    Normal
    Tangent
    WorldPositionOffset
    SubsurfaceColor
    ClearCoat
    ClearCoatRoughness
    AmbientOcclusion
    Refraction
    CustomizedUVs_0
    CustomizedUVs_1
    CustomizedUVs_2
    CustomizedUVs_3
    CustomizedUVs_4
    CustomizedUVs_5
    CustomizedUVs_6
    CustomizedUVs_7
    PixelDepthOffset
    ShadingModel
    Displacement
   )) {inputType=FunctionInput_MaterialAttributes}
  n2 = GetMaterialAttributes(None=@n1) {attributeGetTypes [#1]}
  n3 = FunctionInput(Preview=Constant() {r=1}) {bUsePreviewValueAsDefault=true, inputName=Banding Mask, inputType=FunctionInput_Scalar, sortPriority=1}
  n4 = Multiply(A=@n2.BaseColor, B=OneMinus(None=@n3))
  n5 = NamedRerouteUsage()
  n6 = NamedRerouteUsage()
  n7 = CameraVectorWS()
  n8 = ConstantBiasScale(None=DotProduct(A=NamedRerouteUsage(), B=Constant3Vector() {constant=<rgba 0 0 1 0>}))
  n9 = Add(
   A=Saturate(None=DotProduct(A=@n6, B=@n7))
   B=LinearInterpolate(
    A=ScalarParameter() {defaultValue=-0.1, group=ColorBanding, parameterName=Banding Fresnel Size Min}
    B=ScalarParameter() {defaultValue=0.5, parameterName=Banding Fresnel Size Max}
    Alpha=Power(Base=@n8) {constExponent=1}
   )
  )
  n10 = Saturate(None=Floor(None=Multiply(A=@n9) {constB=2}))
  n11 = OneMinus(None=@n10)
  n12 = DotProduct(A=ReflectionVectorWS(CustomWorldNormal=NamedRerouteUsage()), B=SkyAtmosphereLightDirection())
  n13 = ScalarParameter() {defaultValue=4, group=ColorBanding, parameterName=Banding Count}
  n14 = Multiply(
   A=Power(Base=ConstantBiasScale(None=@n12), Exp=ScalarParameter() {defaultValue=1.5, group=ColorBanding, parameterName=Banding Power})
   B=@n13
  )
  n15 = LinearInterpolate(A=NamedRerouteUsage(), B=NamedRerouteUsage(), Alpha=Divide(A=Floor(None=@n14), B=@n13)) {constA=0.5}
  n16 = StaticSwitchParameter(True=Add(A=Multiply(A=@n5, B=@n11), B=Multiply(A=@n15, B=@n10)), False=@n15) {group=ColorBanding, parameterName=Use Banding Fresnel}
  n17 = ScalarParameter() {defaultValue=0.95, group=ColorBanding, parameterName=Highlight Size}
  n18 = DotProduct(A=ReflectionVectorWS(CustomWorldNormal=NamedRerouteUsage()), B=SkyAtmosphereLightDirection())
  n19 = Ceil(None=Saturate(None=InverseLinearInterpolate(A=@n17, Value=@n18)))
  n20 = Multiply(A=NamedRerouteUsage(), B=@n19)
  n21 = Add(A=@n16, B=@n20)
  n22 = Add(A=Multiply(A=@n16, B=OneMinus(None=@n19)), B=@n20)
  n23 = Multiply(
   A=@n3
   B=StaticSwitchParameter(True=StaticSwitchParameter(True=@n21, False=@n22) {group=ColorBanding, parameterName=Use Additive Highlight}, False=@n16) {group=ColorBanding, parameterName=Use Banding Highlight}
  )
  FunctionOutput(None=SetMaterialAttributes!(MaterialAttributes=@n1, BaseColor=Add(A=@n4, B=@n23)) {attributeSetTypes [#1]})
  NamedRerouteDeclaration(None=@n2.BaseColor) {name=Original Color, nodeColor=<rgba 1 0 0.6562 1>}
  n24 = NamedRerouteUsage()
  n25 = VectorParameter() {group=ColorBanding, parameterName=Banding Fresnel Color Multiplier, defaultValue=<rgba 0.9089 1.2672 0.9835 1>}
  NamedRerouteDeclaration(None=StaticSwitchParameter(
    True=Multiply(A=@n24, B=@n25.RGB)
    False=VectorParameter() {group=ColorBanding, parameterName=Banding Fresnel Color, defaultValue=<rgba 1 0.6266 0.1552 1>}.RGB
   ) {group=ColorBanding, parameterName=Use Original Color}) {name=Fresnel Color, nodeColor=<rgba 1 0 0.6562 1>}
  NamedRerouteDeclaration(None=StaticSwitchParameter(
    True=NamedRerouteUsage()
    False=VectorParameter() {group=ColorBanding, parameterName=Banding Base Color Max, defaultValue=<rgba 0.7073 0.3685 0.0013 1>}.RGB
   ) {group=ColorBanding, parameterName=Use Original Color}) {name=Base Color Max, nodeColor=<rgba 1 0 0.6562 1>}
  n26 = NamedRerouteUsage()
  n27 = VectorParameter() {group=ColorBanding, parameterName=Banding Base Color Min Multiplier, defaultValue=<rgba 0.4281 0.4281 0.4281 1>}
  NamedRerouteDeclaration(None=StaticSwitchParameter(
    True=Multiply(A=@n26, B=@n27.RGB)
    False=VectorParameter() {group=ColorBanding, parameterName=Banding Base Color Min, defaultValue=<rgba 0.4198 0.1718 0 1>}.RGB
   ) {group=ColorBanding, parameterName=Use Original Color}) {name=Base Color Min, nodeColor=<rgba 1 0 0.6562 1>}
  n28 = NamedRerouteUsage()
  n29 = VectorParameter() {group=ColorBanding, parameterName=Banding Highlight Multiplier, defaultValue=<rgba 2 2 2 1>}
  NamedRerouteDeclaration(None=StaticSwitchParameter(
    True=Multiply(A=@n28, B=@n29.RGB)
    False=VectorParameter() {group=ColorBanding, parameterName=Highlight Color, defaultValue=<rgba 1 0.8853 0.5886 1>}.RGB
   ) {group=ColorBanding, parameterName=Use Original Color}) {name=Highlight Color, nodeColor=<rgba 1 0 0.6562 1>}
  NamedRerouteDeclaration(None=FunctionInput(Preview=VertexNormalWS()) {bUsePreviewValueAsDefault=true, inputName=NormalWS, sortPriority=2}) {name=NormalWS, nodeColor=<rgba 1 0 0.2344 1>}
  VertexNormalWS()
  VertexNormalWS()
 ]}
