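# SonicSpear material, apparently for donut drop objects, with frosting, sprinkle colors and stylized shading.
# Packed texture channels decode into color indices, cycled by custom primitive data and looked up in a curve atlas; frosting takes a per-primitive tint and noise normal.
# Output chains through color banding, flat shading, sway, destructible damage and cracks, with optional height-based WPO.
# A polar-tiled, randomized SDF detail mask is built; timed squash-and-stretch math ends unconnected.
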Material /SonicSpearGameplay/Assets/Materials/SonicSpear_Colors
graph {
 properties {bAutomaticallySetUsageInEditor=false, bUseMaterialAttributes=true, usedShadingModels=Default Lit}
 nodes [
  n1 = TextureSample() {texture=/SonicSpearGameplay/Assets/Textures/T_SonicSpear_DropObject_Donut_RGBA_Test}
  n2 = ScalarParameter() {defaultValue=1.4, parameterName=Blend}
  n3 = Divide(B=@n2)
  n4 = Constant() {r=1}
  n5 = Constant()
  n6 = Constant() {r=2}
  n7 = Add(
   A=Add(A=If(A=@n1.R, B=@n3, "A > B"=@n4, "A < B"=@n5) {constB=0.99}, B=If(A=@n1.G, B=@n3, "A > B"=@n6, "A < B"=@n5) {constB=0.99})
   B=If(A=@n1.B, B=@n3, "A > B"=Constant() {r=4}, "A < B"=@n5) {constB=0.99}
  )
  n8 = VectorParameter() {bUseCustomPrimitiveData=true, parameterName=Frosting Color + Sprinkle Offset, defaultValue=<rgba 0.7758 0.0931 0.5271 0>}
  n9 = Add(A=Clamp(None=Subtract(A=@n7)) {maxDefault=6}, B=Round(None=@n8.A))
  n10 = Multiply(A=Frac(None=Divide(A=@n9) {constB=6})) {constB=6}
  n11 = Max(A=Floor(None=Divide(A=@n7) {constB=7}), B=OneMinus(None=Saturate(None=@n7)))
  n12 = Divide(A=LinearInterpolate(A=Add(A=@n10), B=@n7, Alpha=@n11)) {constB=7}
  n13 = CurveAtlasRowParameter(CurveTime=@n12) {atlas=/SonicSpearGameplay/Assets/Curves/CA_SonicSpear, curve=/SonicSpearGameplay/Assets/Curves/C_SonicSpear_Pink, parameterName=ColorCurve}
  n14 = Saturate(None=Multiply(A=@n1.RGB, B=@n2) {constB=1.01})
  n15 = MaterialFunctionCall(Float3=Divide(A=Frac(None=@n14), B=@n2) {constB=1.01}) {materialFunction=&BreakOutFloat3Components}
  n16 = Max(A=Max(A=Max(A=@n15.R, B=@n15.R), B=@n15.B), B=Saturate(None=Subtract(A=@n7) {constB=6}))
  n17 = Multiply(A=Multiply(A=@n11, B=OneMinus(None=@n16)), B=NamedRerouteUsage())
  Add(
   A=Multiply(A=@n13, B=OneMinus(None=@n17))
   B=Multiply(A=@n17, B=VectorParameter() {parameterName=Texture Color, defaultValue=<rgba 0 0 0 1>}.RGB)
  )
  Constant()
  n18 = InverseLinearInterpolate(
   A=ScalarParameter() {defaultValue=-25, parameterName=WPO Height Min}
   B=ScalarParameter() {defaultValue=100, parameterName=Wpo Height Max}
   Value=ComponentMask(None=LocalPosition().XYZ) {g=true}
  )
  Frac(None=@n18)
  n19 = Desaturation(None=@n8.RGB, Fraction=Constant() {r=0.3})
  n20 = SetMaterialAttributes!(
   MaterialAttributes
   Base Color=Multiply(A=LinearInterpolate(A=@n13, B=@n19, Alpha=@n16), B=@n1.A)
   Metallic=Constant()
   Specular=Constant()
   Roughness=@n12
   World Position Offset=NamedRerouteUsage()
  ) {attributeSetTypes [
    #1
    #2
    #3
    #4
    #5
   ]}
  n21 = ComponentMask(None=LocalPosition().XYZ) {g=true, r=true}
  n22 = ScalarParameter() {defaultValue=0.002, parameterName=Frosting Normal Tiling}
  n23 = MaterialFunctionCall(Float3=TextureSample(UVs=Multiply(A=@n21, B=@n22)) {samplerType=SAMPLERTYPE_Normal, texture=/SonicSpearGameplay/Assets/Textures/LowResBlurredNoiseNormal}.RGB) {materialFunction=&BreakOutFloat3Components}
  n24 = ScalarParameter() {defaultValue=10, parameterName=Frosting Normal Strength}
  n25 = Normalize(VectorInput=MaterialFunctionCall(X=@n23.R, Y=@n23.R, Z=Multiply(A=@n23.B, B=@n24)) {materialFunction=&MakeFloat3}.Result)
  n26 = MaterialFunctionCall!(
   Material In=MaterialFunctionCall!(In=@n20, Banding Mask=NamedRerouteUsage(), NormalWS=Transform(None=@n25)) {materialFunction=&/MF_ColorBanding}.Result
   UseFlatShading=StaticBoolParameter() {group=FlatShading, parameterName=UseFlatShading_All, sortPriority=1}
   IsWorldSpace=StaticBool() {value=true}
   OverrideUnlitShadinBaseColorDarkening
   OverrideUnlitShadinBaseColorEmissive
  ) {materialFunction=/WildEstate/Environment/BaseMaterials/MaterialFunctions/MF_WildEstate_FlatShading}
  MaterialFunctionCall!(
   In=MaterialFunctionCall(
    Material_In=MaterialFunctionCall!(In=@n26.Result) {materialFunction=/WildEstateTech/Environment/BaseMaterials/MaterialFunctions/MF_WE_WPOSway}.Result
    UseCrackedDamageTextures=StaticBool()
    UseAdditiveWPO=StaticBool() {value=true}
   ) {materialFunction=&MF_FN_DestructibleDamage}.Material_Out
   DestructibleDamageAmount
  ) {materialFunction=/WildEstateTech/Environment/BaseMaterials/MaterialFunctions/MF_WildEstate_Cracks}
  n27 = InverseLinearInterpolate(Value=NamedRerouteUsage()) {constA=0.1, constB=0.5}
  n28 = Saturate(None=LinearInterpolate(
    A=ScalarParameter() {defaultValue=1, parameterName=SDF Cutoff Min}
    B=ScalarParameter() {defaultValue=0.8, parameterName=SDF Cutoff Max}
    Alpha=Saturate(None=@n27)
   ))
  n29 = ScalarParameter() {defaultValue=-0.2, parameterName=Offset X}
  n30 = ScalarParameter() {defaultValue=0.22, parameterName=Offset Y}
  n31 = ConstantBiasScale(None=Add(A=TextureCoordinate(), B=AppendVector(A=@n29, B=@n30))) {bias=-0.5, scale=2}
  n32 = MaterialFunctionCall(Float2=@n31) {materialFunction=&BreakOutFloat2Components}
  n33 = Divide(A=Arctangent2Fast(Y=@n32.R, X=@n32.R)) {constB=3.1416}
  n34 = Length(None=@n31) {is=class MaterialExpressionLength}
  n35 = Multiply(
   A=AppendVector(A=ConstantBiasScale(None=@n33), B=@n34)
   B=AppendVector(A=ScalarParameter() {defaultValue=15, parameterName=Tiling X}, B=ScalarParameter() {defaultValue=7.5, parameterName=Tiling Y})
  )
  n36 = MaterialFunctionCall(Float2=Ceil(None=@n35)) {materialFunction=&BreakOutFloat2Components}
  n37 = MaterialFunctionCall!(seed=MaterialFunctionCall(X=@n36.R, Y=@n36.R) {materialFunction=&MakeFloat2}.Result) {materialFunction=&/MF_SonicSpear_PseudoRandom}
  n38 = MaterialFunctionCall(seed=MaterialFunctionCall(X=@n36.G, Y=@n36.G) {materialFunction=&MakeFloat2}.Result) {materialFunction=&MF_SonicSpear_PseudoRandom}
  n39 = ConstantBiasScale(None=Divide(A=Add(A=@n37.Result, B=@n38.Result))) {bias=-0.5, scale=2}
  n40 = ScalarParameter() {defaultValue=0.1, parameterName=Random Rotation Strength}
  n41 = ConstantBiasScale(None=MaterialFunctionCall(X=@n37.Result, Y=@n38.Result) {materialFunction=&MakeFloat2}.Result) {bias=-0.5, scale=2}
  n42 = ScalarParameter() {defaultValue=0.3, parameterName=Random Offset Strength}
  n43 = AppendVector(A=Add(A=Frac(None=@n35), B=Multiply(A=@n41, B=@n42)), B=Constant())
  n44 = RotateAboutAxis(
   NormalizedRotationAxis=Constant3Vector() {constant=<rgba 0 0 1 0>}
   RotationAngle=Add(A=Multiply(A=@n39, B=@n40), B=ScalarParameter() {defaultValue=0.5, parameterName=Starting Rotation})
   PivotPoint=Constant3Vector() {constant=<rgba 0.5 0.5 0 0>}
   Position=@n43
  )
  n45 = ConstantBiasScale(None=ComponentMask(None=Add(A=@n44, B=@n43)) {g=true, r=true}) {bias=-0.5, scale=2}
  n46 = ScalarParameter() {defaultValue=2.5, parameterName=Texture Scaling}
  n47 = TextureSample(
   UVs=ConstantBiasScale(None=Multiply(A=@n45, B=@n46))
   Tex=TextureObject() {samplerType=SAMPLERTYPE_Grayscale, texture=/SonicSpearGameplay/Assets/Textures/T_SonicSpear_Detail}
  ) {samplerSource=SSM_Clamp_WorldGroupSettings}
  NamedRerouteDeclaration(None=Power(Base=InverseLinearInterpolate(A=@n28, Value=@n47.R), Exp=ScalarParameter() {defaultValue=0.3, parameterName=SDF Power})) {name=Texture Mask, nodeColor=<rgba 0.6875 1 0 1>}
  NamedRerouteDeclaration(None=Saturate(None=@n34)) {name=Texture Falloff, nodeColor=<rgba 1 0 0.7031 1>}
  NamedRerouteDeclaration(None=@n16) {name=Frosting Mask, nodeColor=<rgba 1 0 0.9609 1>}
  n48 = Saturate(None=@n18)
  n49 = ScalarParameter() {defaultValue=2, parameterName=Ground WPO Power}
  n50 = Multiply(A=Multiply(A=Power(Base=@n48, Exp=@n49), B=ScalarParameter() {defaultValue=50, parameterName=Ground WPO Strength}), B=VertexNormalWS())
  n51 = NamedRerouteUsage()
  NamedRerouteDeclaration(None=StaticSwitchParameter(True=Multiply(A=@n50, B=@n51), False=Constant()) {parameterName=UseGroundWPO}) {name=WPO Out, nodeColor=<rgba 0 1 0.7188 1>}
  n52 = Time()
  n53 = ScalarParameter() {bUseCustomPrimitiveData=true, defaultValue=100, parameterName=BPGameTime, primitiveDataIndex=4}
  n54 = Saturate(None=Multiply(A=Subtract(A=@n52, B=@n53), B=ScalarParameter() {defaultValue=0.5, parameterName=Squash+Stretch Length}))
  n55 = Subtract(
   A=WorldPosition().XYZ
   B=MaterialFunctionCall() {materialFunction=&ObjectPivotPoint}.Object Pivot Location
  )
  n56 = AppendVector(A=Normalize(VectorInput=ComponentMask(None=@n55) {g=true, r=true}), B=Constant())
  n57 = Normalize(VectorInput=Transform(None=Constant3Vector() {constant=<rgba 0 1 0 0>}))
  n58 = Sine(None=Multiply(A=Time()))
  n59 = Multiply(
   A=Add(A=Constant3Vector() {constant=<rgba 0 0 -1 0>}, B=Multiply(A=@n56, B=@n57))
   B=Multiply(A=ScalarParameter() {defaultValue=200, parameterName=WPO Strength}, B=ConstantBiasScale(None=@n58))
  )
  n60 = VectorParameter() {parameterName=Squash Offset, defaultValue=<rgba 0 0 -100 1>}
  n61 = MaterialFunctionCall() {materialFunction=&ObjectScale}
  n62 = ScalarParameter() {defaultValue=100, parameterName=Scale Multiplier}
  n63 = Saturate(None=Divide(A=Subtract(A=@n55, B=@n60.RGB), B=Multiply(A=@n61.Scale XYZ, B=@n62)))
  Reroute(None=Multiply(A=OneMinus(None=@n54), B=Multiply(A=@n59, B=@n63)))
  TextureCoordinate()
 ]
}
