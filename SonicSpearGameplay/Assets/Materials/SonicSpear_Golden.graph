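# Lit material for the Sonic Spear's Golden variant, named for a gold look and built with material attributes.
# It thresholds the RGB channels of a donut drop-object texture into a weighted 1/2/4 bitmask, normalized into a small roughness offset over a constant base.
# Those attributes, with zero metallic and specular, pass through color banding, parameter-switched flat shading, destructible damage with additive WPO, and finally the WildEstate cracks function.
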Material /SonicSpearGameplay/Assets/Materials/SonicSpear_Golden
graph {
 properties {bAutomaticallySetUsageInEditor=false, bUseMaterialAttributes=true, usedShadingModels=Default Lit}
 nodes [
  n1 = Constant() {r=0.25}
  n2 = TextureSample() {samplerType=SAMPLERTYPE_LinearColor, texture=/SonicSpearGameplay/Assets/Textures/T_SonicSpear_DropObject_Donut_RGBA}
  n3 = Constant() {r=0.666}
  n4 = Constant()
  n5 = Add(
   A=If(A=@n2.R, B=@n3, "A > B"=Constant() {r=1}, "A < B"=@n4) {constB=0.99}
   B=If(A=@n2.G, B=@n3, "A > B"=Constant() {r=2}, "A < B"=@n4) {constB=0.99}
  )
  n6 = If(A=@n2.B, B=@n3, "A > B"=Constant() {r=4}, "A < B"=@n4) {constB=0.99}
  n7 = Divide(A=Divide(A=Add(A=@n5, B=@n6)) {constB=7}) {constB=10}
  n8 = MaterialFunctionCall!(
   In=SetMaterialAttributes!(MaterialAttributes, Metallic=Constant(), Specular=Constant(), Roughness=Add(A=@n1, B=@n7)) {attributeSetTypes [#1, #2, #3]}
   Banding Mask
   NormalWS
  ) {materialFunction=&/MF_ColorBanding}
  n9 = StaticBoolParameter() {group=FlatShading, parameterName=UseFlatShading_All, sortPriority=1}
  n10 = StaticBool()
  MaterialFunctionCall!(
   In=MaterialFunctionCall(
    Material_In=MaterialFunctionCall!(Material In=@n8.Result, UseFlatShading=@n9, IsWorldSpace=@n10, OverrideUnlitShadinBaseColorDarkening, OverrideUnlitShadinBaseColorEmissive) {materialFunction=/WildEstate/Environment/BaseMaterials/MaterialFunctions/MF_WildEstate_FlatShading}.Result
    UseCrackedDamageTextures=StaticBool()
    UseAdditiveWPO=StaticBool() {value=true}
   ) {materialFunction=&MF_FN_DestructibleDamage}.Material_Out
   DestructibleDamageAmount
  ) {materialFunction=/WildEstateTech/Environment/BaseMaterials/MaterialFunctions/MF_WildEstate_Cracks}
 ]
}
