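# Unlit translucent deferred decal material for the Sonic Spear's ground marker, drawing a soft circle that changes with distance from the target.
# PercentDistanceFromTarget samples a curve atlas row, steep-drop by default, whose red and blue channels blend start-to-end radius and hardness of a centered UV SphereMask.
# The green channel and an Overall Opacity parameter scale that mask, alongside a dark Color parameter.
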Material /SonicSpearGameplay/Assets/VFX/Materials/M_SonicSpearMarker_GroundDecal
graph {
 properties {
  bUsedWithMeshParticles=true
  blendMode=BLEND_Translucent
  materialDomain=MD_DeferredDecal
  pixelDepthOffsetMode=PDOM_Legacy
  shadingModel=MSM_Unlit
  twoSided=true
  usedShadingModels=Unlit
 }
 nodes [
  n1 = TextureCoordinate()
  n2 = Constant() {r=0.5}
  n3 = CurveAtlasRowParameter(CurveTime=ScalarParameter() {defaultValue=1, parameterName=PercentDistanceFromTarget, sliderMax=1}) {
   atlas=/SonicSpearGameplay/Assets/Curves/CA_SonicSpear
   curve=/SonicSpearGameplay/Assets/Curves/C_SonicSpear_Steep_Drop
   defaultValue=3
   parameterName=Curve
  }
  n4 = LinearInterpolate(A=ScalarParameter() {parameterName=Start Radius}, B=ScalarParameter() {defaultValue=0.5, parameterName=End Radius}, Alpha=@n3.R) {constA=0.2}
  n5 = LinearInterpolate(A=ScalarParameter() {parameterName=Start Hardness}, B=ScalarParameter() {defaultValue=0.9, parameterName=End Hardness}, Alpha=@n3.B) {constA=0.2}
  Multiply(
   A=Multiply(A=SphereMask(A=@n1, B=@n2, Radius=@n4, Hardness=@n5), B=@n3.G)
   B=ScalarParameter() {defaultValue=1, parameterName=Overall Opacity}
  )
  Subtract(B=@n3.G)
  VectorParameter() {parameterName=Color, defaultValue=<rgba 0.0168 0.0205 0.0312 1>}
 ]
}
