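# Post-process material for the Sonic Spear marker that adds a pulsing red glow to the screen.
# A soft SphereMask over scaled, offset viewport UVs places the glow around the top center, fading outward.
# A DebugTimeSine call oscillates the tint between red and a dimmed copy, which is multiplied by the mask and added to scene color.
# The Blend scalar parameter lerps between the untouched scene and the glowing result.
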Material /SonicSpearGameplay/Assets/VFX/Materials/M_SonicSpearMarker_PP_ScreenEdgeGlow
graph {
 properties {materialDomain=MD_PostProcess, pixelDepthOffsetMode=PDOM_Legacy, shadingModel=MSM_Unlit, usedShadingModels=Unlit}
 nodes [
  n1 = ScreenPosition()
  n2 = Constant2Vector() {g=1.5, r=1}
  n3 = SphereMask(A=Add(A=Multiply(A=@n1.ViewportUV, B=@n2), B=Constant2Vector() {g=0.7}), B=Constant() {r=0.5}, Radius=Constant() {r=1}, Hardness=Constant())
  n4 = Constant3Vector() {constant=<rgba 0.7 0.07 0.07 0>}
  n5 = LinearInterpolate(
   A=@n4
   B=Divide(A=@n4)
   Alpha=MaterialFunctionCall(Speed=Constant() {r=1.5}) {materialFunction=&DebugTimeSine}.Result
  )
  n6 = SceneTexture() {sceneTextureId=PPI_PostProcessInput0}
  LinearInterpolate(
   A=ComponentMask(None=SceneTexture() {sceneTextureId=PPI_PostProcessInput0}.Color) {b=true, g=true, r=true}
   B=Add(A=Multiply(A=@n3, B=@n5), B=ComponentMask(None=@n6.Color) {b=true, g=true, r=true})
   Alpha=ScalarParameter() {parameterName=Blend}
  )
  Multiply(A=@n5)
 ]
}
