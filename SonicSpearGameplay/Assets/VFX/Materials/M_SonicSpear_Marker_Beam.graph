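# Translucent unlit material for Niagara ribbons, named as the Sonic Spear marker beam.
# Particle color feeds emissive through a shared adjustment function, while particle alpha scales the beam's opacity.
# The beam's shape is half of a soft flare texture multiplied by a linear fade along U, then depth-faded.
# A fast time-driven sine pulses the result, with the Flicker dynamic parameter blending between that pulse and a steady value.
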Material /SonicSpearGameplay/Assets/VFX/Materials/M_SonicSpear_Marker_Beam
graph {
 properties {bUsedWithNiagaraRibbons=true, blendMode=BLEND_Translucent, pixelDepthOffsetMode=PDOM_Legacy, shadingModel=MSM_Unlit, usedShadingModels=Unlit}
 nodes [
  n1 = ParticleColor()
  MaterialFunctionCall(Emissive=@n1.RGB) {materialFunction=&MF_FX_EmissiveAdjustments}
  n2 = TextureCoordinate()
  n3 = TextureCoordinate() {uTiling=0.5}
  n4 = Constant2Vector() {r=0.5}
  n5 = Multiply(
   A=OneMinus(None=ComponentMask(None=@n2) {r=true})
   B=TextureSample(UVs=Add(A=@n3, B=@n4)) {texture=/Game/Effects/Fort_Effects/Textures/Flares/T_Flare_Soft}.R
  )
  n6 = Sine(None=Time()) {period=0.06}
  Multiply(
   A=DepthFade(Opacity=Multiply(A=@n5, B=@n1.A))
   B=LinearInterpolate(A=LinearInterpolate(Alpha=@n6) {constA=0.7}, Alpha=DynamicParameter() {paramNames [Flicker,Param2,Param3,Param4]}.Flicker) {constA=1}
  )
 ]
}
