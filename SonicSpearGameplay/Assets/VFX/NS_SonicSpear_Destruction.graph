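# Niagara burst effect for destroying a frosted donut target in Sonic Spear gameplay, with normal and golden variants.
# Torus-shaped bursts scale with User.Size; lingering smoke plays for both variants.
# User.Golden swaps food debris and User.Frosting-colored liquid sprites for gold debris and curl-noise sparkles with pulsing glitter.
# Time-of-day tints some normal debris and liquid, food-debris collision is gated by platform quality, and the Donut_Dissolve mesh emitter is disabled.
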NiagaraSystem /SonicSpearGameplay/Assets/VFX/NS_SonicSpear_Destruction
dependencies {
 usedDataInterfaces [
  &NiagaraDataInterfaceCurve
  &NiagaraDataInterfaceColorCurve
  &NiagaraDataInterfaceArrayColor
  &NiagaraDataInterfacePlatformSet
  &NiagaraDataInterfaceArrayFloat
  &NiagaraDataInterfaceMeshRendererInfo
 ]
 usedDynamicInputs []
 usedModules [&SystemState]
}
emitters {
 Donut_Dissolve {
  bEnabled=false
  emitterUpdateScript [
   &EmitterState()
   &SpawnBurst_Instantaneous(Use Spawn Probability=false, UseLoopCountLimit=false)
  ]
  particleSpawnScript [
   &InitializeParticle(
    Color=&MakeCustomLinearColorFromBool(
     Bool=$User.Golden
     True Color=&FNTimeOfDayColor(
      Night Color=<rgba 0.1771 0.1178 0 1>
      Morning Color=<rgba 0.526 0.1641 0 1>
      Day Color=<rgba 0.875 0.2729 0 1>
      Evening Color=<rgba 0.5521 0.1722 0 1>
     )
     False Color=<rgba 0.8021 0.3929 0.1129 1>
    )
    Lifetime=&Lerp_Float(A=0.1, B=0.075, Alpha=$User.Size)
    Mesh Scale Mode=::NewEnumerator3
    Mesh Uniform Scale=&Lerp_Float(A=0.5025, B=1.025, Alpha=$User.Size)
    UsePositionOffset=false
    Write Material Random=true
   )
   /SetVariables(Particles.VisibilityTag=&MakeCustomIntFromBool(Bool=$User.Golden, True INT=1, False INT=0)) {isSetParameters=true}
  ]
  particleUpdateScript [
   &ParticleState()
   &DynamicMaterialParameters(
    Index 0 Param 1=&MakeCustomFloatFromBool(
     Bool=$User.Golden
     True Float=1
     False Float=&FNTimeOfDayScalar(Night Scale=0.25, Morning Scale=0.5, Day Scale=0.7, Evening Scale=0.45)
    )
    Index 0 Param 3=&FloatFromCurve(FloatCurve=DataInterface() {propertyValues {Curve {
        defaultValue=0
        keys [
         {arriveTangent=0.3882, interpMode=RCIM_Cubic, leaveTangent=0.3882, tangentMode=RCTM_User, time=0, value=0}
         {arriveTangent=1.2482, interpMode=RCIM_Cubic, leaveTangent=1.2482, tangentMode=RCTM_User, time=1, value=1}
        ]
       }}})
    Index 0 Param 4=1
    Module.Index0Float0Write=true
    Module.Index0Float1Write=true
    Module.Index0Float2Write=true
    Module.Index0Float3Write=true
   )
  ]
  rendererClasses [&NiagaraMeshRendererProperties]
  renderers [
   {rendererClass=&NiagaraMeshRendererProperties, rendererIndex=0}
   {rendererClass=&NiagaraMeshRendererProperties, rendererIndex=1}
  ]
 }
 Food_Debris {
  emitterUpdateScript [
   &EmitterState()
   &SpawnBurst_Instantaneous(
    Spawn Count=&MakeCustomIntFromBool(Bool=$User.Golden, True INT=0, False INT=&RandomRangeInt(Minimum=30, Maximum=48))
    Use Spawn Probability=false
    UseLoopCountLimit=false
   )
   &SpawnBurst_Instantaneous(
    Spawn Count=&MakeCustomIntFromBool(Bool=$User.Golden, True INT=0, False INT=&RandomRangeInt(Minimum=18, Maximum=28))
    Spawn Group=1
    Use Spawn Probability=false
    UseLoopCountLimit=false
   ) {module=SpawnBurst_Instantaneous001}
  ]
  particleSpawnScript [
   /SetVariables(Particles.VisibilityTag=&SpawnGroup()) {isSetParameters=true, module=SetVariables_#1}
   &InitializeParticle(
    Color=&MakeLinearColorFromVectorAndFloat(
     "Vector (RGB)"=&MaskVector(
      Mask Source=0
      Mask Parameter=?
      Mask=0
      Masked Vector=&MakeVectorFromLinearColorRGB(Color=<rgba 1 1 1 1>)
      Passthrough Vector=&MakeVectorFromLinearColorRGB(Color=&SelectLinearColorFromArray(
        Array Sampling Mode=::NewEnumerator0
        Direct Array Index=?
        Color Selection Array=DataInterface() {propertyValues {ColorData [
           <rgba 1 0.0437 0.0685 1>
           <rgba 0 0.4621 0.9387 1>
           <rgba 0.007 0.5972 0.2016 1>
           <rgba 1 1 0.8308 1>
           <rgba 0.6445 0.314 0.0931 1>
           <rgba 1 1 0 1>
          ]}}
        Random Seed=0
       ))
     )
     "Float (Alpha)"=1
    )
    Lifetime Max=1.4
    Lifetime Min=0.9
    Lifetime Mode=::NewEnumerator1
    Mass=&MaskFloatBySpawnGroup(Masked Float=4, Passthrough Float=1)
    Mass Mode=::NewEnumerator1
    Mesh Scale=&MaskVector(
     Mask Source=0
     Mask Parameter=?
     Mask=0
     Masked Vector=&Lerp_Vector(
      A=&RandomRangeVector(Minimum=<1.25 1.25 1.25>, Maximum=<7.65 7.65 7.65>)
      B=&RandomRangeVector(Minimum=<7.65 7.65 7.65>, Maximum=<15.3 15.3 15.3>)
      Alpha=$User.Size
     )
     Passthrough Vector=&Lerp_Vector(A=<18 10 10>, B=<36 20 20>, Alpha=$User.Size)
    )
    Mesh Scale Mode=::NewEnumerator4
    UsePositionOffset=false
    Write Material Random=true
   )
   &ShapeLocation(
    Handle Radius:NiagaraFloat=&Lerp_Float(A=70, B=190, Alpha=$User.Size)
    Large Radius=&Lerp_Float(A=170, B=375, Alpha=$User.Size)
    MaskSpawnGroup=false
    Non Uniform Scale=&MaskVector(Mask Source=0, Mask Parameter=?, Mask=0, Masked Vector=<1 1 0.8>, Passthrough Vector=<1 1 -0.8>)
    Offset Mode=::NewEnumerator0
    Override Seed=false
    Rotation Axis=<0 0 1>
    Rotation Mode=::NewEnumerator1
    Shape Primitive=::NewEnumerator3
    Surface Distribution=1
    UseOwnerScale=true
    V Distribution:NiagaraFloat=&MaskFloat(Mask Source=0, Mask Parameter=?, Mask=0, Masked Float=0, Passthrough Float=0.5)
   )
   &InitialMeshOrientation(Use Rotation Vector=false)
   &AddVelocity(
    Module.Use Falloff Distance=false
    Override Seed=false
    Velocity Mode=::NewEnumerator1
    Velocity Speed=&MaskFloat(Mask Source=0, Mask Parameter=?, Mask=0, Masked Float=250, Passthrough Float=300)
   )
   &AddVelocity(Velocity=&MaskVector(
     Mask Source=0
     Mask Parameter=?
     Mask=0
     Masked Vector=<0 0 0>
     Passthrough Vector=&RandomRangeVector(Minimum=<0 0 25>, Maximum=<0 0 200>)
    )) {module=AddVelocity001}
   &AddVelocity(Rotation Coordinate Space=::World, Velocity=&RandomRangeVector(Minimum=<0 0 200>, Maximum=<0 0 800>)) {module=AddVelocity002}
   &AddRotationalVelocity(Rotation Rate=&RandomVector(Vector Scale=1))
   &CameraOffset(Camera Offset Amount=50)
  ]
  particleUpdateScript [
   &ParticleState()
   &ScaleMeshSize(Scale Factor=&VectorFromFloat(Value=&FloatFromCurve(
      FloatCurve=DataInterface() {propertyValues {Curve {
         defaultValue=1
         keys [
          {arriveTangent=38.0665, interpMode=RCIM_Cubic, leaveTangent=38.0665, tangentMode=RCTM_User, time=0, value=0}
          {arriveTangent=1.4286, leaveTangent=1.4286, time=0.07, value=1}
          {arriveTangent=0.1167, interpMode=RCIM_Cubic, leaveTangent=0.1167, tangentMode=RCTM_User, time=0.7, value=1}
          {arriveTangent=-10.8391, interpMode=RCIM_Cubic, leaveTangent=-10.8391, tangentMode=RCTM_User, time=1, value=0}
         ]
        }}}
      Scale Curve=&MaskFloatBySpawnGroup(Masked Float=1.5, Passthrough Float=1.2)
     )))
   &Drag(Use Linear Drag=true, Use Rotational Drag=true)
   &GravityForce()
   &Collision(
    Age Colliding Particles=false
    Collision Enabled=&Boolean_AndOperation(
     Bool1=&IsPlatformSetActive(New Platform Set=DataInterface() {propertyValues {Platforms {qualityLevelMask=-8}}})
     Bool2=&SetBoolByFloatComparison(A=$Particles.NormalizedAge, B=0.2, Comparison Type=::NewEnumerator2)
    )
    KillOnCollision=false
    Module.Control Rest State Via Surface Normal=false
    Module.EnableMaxCollisionCount=false
    Module.Randomize Collision Normal Vector=false
   )
   &SolveForcesAndVelocity(Clamp Velocity=false, Limit Acceleration=false)
   &ScaleColor(
    Color Scale:LinearColor=&MaskLinearColorBySpawnGroup(
     Masked Linear Color=&FNTimeOfDayColor(Night Color=<rgba 1 1 1 1>, Morning Color=<rgba 1 1 1 1>, Day Color=<rgba 1 1 1 1>, Evening Color=<rgba 1 1 1 1>)
     Passthrough Linear Color=&FNTimeOfDayColor(Night Color=<rgba 0.4167 0.4167 0.4167 1>, Morning Color=<rgba 1 1 1 1>, Day Color=<rgba 1 1 1 1>, Evening Color=<rgba 1 1 1 1>)
    )
    Scale Mode=3
   )
  ]
  rendererClasses [&NiagaraMeshRendererProperties]
  renderers [
   {rendererClass=&NiagaraMeshRendererProperties, rendererIndex=0}
   {rendererClass=&NiagaraMeshRendererProperties, rendererIndex=1}
   {rendererClass=&NiagaraMeshRendererProperties, rendererIndex=2}
   {rendererClass=&NiagaraMeshRendererProperties, rendererIndex=3}
  ]
 }
 Food_Debris001 {
  emitterUpdateScript [
   &EmitterState()
   &SpawnBurst_Instantaneous(
    Spawn Count=&MakeCustomIntFromBool(Bool=$User.Golden, True INT=0, False INT=&RandomRangeInt(Minimum=30, Maximum=48))
    Use Spawn Probability=false
    UseLoopCountLimit=false
   )
  ]
  particleSpawnScript [
   /SetVariables(Particles.VisibilityTag=&SpawnGroup()) {isSetParameters=true, module=SetVariables_#1}
   &InitializeParticle(
    Lifetime Max=1.6
    Lifetime Min=1.2
    Lifetime Mode=::NewEnumerator1
    Mass=&MaskFloatBySpawnGroup(Masked Float=4, Passthrough Float=1)
    Mass Mode=::NewEnumerator1
    Mesh Scale Max=<0.2 0.2 0.2>
    Mesh Scale Min=<0.1 0.1 0.1>
    Mesh Scale Mode=::NewEnumerator2
    UsePositionOffset=false
    Write Material Random=true
   )
   &ShapeLocation(
    Handle Radius:NiagaraFloat=&Lerp_Float(A=70, B=190, Alpha=$User.Size)
    Large Radius=&Lerp_Float(A=170, B=375, Alpha=$User.Size)
    MaskSpawnGroup=false
    Non Uniform Scale=&MaskVector(Mask Source=0, Mask Parameter=?, Mask=0, Masked Vector=<1 1 0.8>, Passthrough Vector=<1 1 -0.8>)
    Offset Mode=::NewEnumerator0
    Override Seed=false
    Rotation Axis=<0 0 1>
    Rotation Mode=::NewEnumerator1
    Shape Primitive=::NewEnumerator3
    Surface Distribution=1
    UseOwnerScale=true
    V Distribution:NiagaraFloat=&MaskFloat(Mask Source=0, Mask Parameter=?, Mask=0, Masked Float=0, Passthrough Float=0.5)
   )
   &InitialMeshOrientation(Use Rotation Vector=false)
   &AddVelocity(
    Module.Use Falloff Distance=false
    Override Seed=false
    Velocity Mode=::NewEnumerator1
    Velocity Speed=&MaskFloat(Mask Source=0, Mask Parameter=?, Mask=0, Masked Float=250, Passthrough Float=300)
   )
   &AddVelocity(Velocity=&MaskVector(
     Mask Source=0
     Mask Parameter=?
     Mask=0
     Masked Vector=<0 0 0>
     Passthrough Vector=&RandomRangeVector(Minimum=<0 0 25>, Maximum=<0 0 200>)
    )) {module=AddVelocity001}
   &AddVelocity(Rotation Coordinate Space=::World, Velocity=&RandomRangeVector(Minimum=<0 0 200>, Maximum=<0 0 800>)) {module=AddVelocity002}
   &AddRotationalVelocity(Rotation Rate=&RandomVector(Vector Scale=1))
   &CameraOffset(Camera Offset Amount=50)
  ]
  particleUpdateScript [
   &ParticleState()
   &ScaleMeshSize(Scale Factor=&VectorFromFloat(Value=&FloatFromCurve(
      FloatCurve=DataInterface() {propertyValues {Curve {
         defaultValue=1
         keys [
          {arriveTangent=38.0665, interpMode=RCIM_Cubic, leaveTangent=38.0665, tangentMode=RCTM_User, time=0, value=0}
          {arriveTangent=1.4286, leaveTangent=1.4286, time=0.07, value=1}
          {arriveTangent=0.1167, interpMode=RCIM_Cubic, leaveTangent=0.1167, tangentMode=RCTM_User, time=0.7, value=1}
          {arriveTangent=-10.8391, interpMode=RCIM_Cubic, leaveTangent=-10.8391, tangentMode=RCTM_User, time=1, value=0}
         ]
        }}}
      Scale Curve=&Lerp_Float(A=0.8, B=1.2, Alpha=$User.Size)
     )))
   &Drag(Use Linear Drag=true, Use Rotational Drag=true)
   &GravityForce()
   &Collision(
    Age Colliding Particles=false
    Collision Enabled=&Boolean_AndOperation(
     Bool1=&IsPlatformSetActive(New Platform Set=DataInterface() {propertyValues {Platforms {qualityLevelMask=-8}}})
     Bool2=&SetBoolByFloatComparison(A=$Particles.NormalizedAge, B=0.2, Comparison Type=::NewEnumerator2)
    )
    KillOnCollision=false
    Module.Control Rest State Via Surface Normal=false
    Module.EnableMaxCollisionCount=false
    Module.Randomize Collision Normal Vector=false
   )
   &SolveForcesAndVelocity(Clamp Velocity=false, Limit Acceleration=false)
  ]
  rendererClasses [&NiagaraMeshRendererProperties]
  renderers [{rendererClass=&NiagaraMeshRendererProperties, rendererIndex=0}]
 }
 GoldDebris {
  emitterUpdateScript [
   &EmitterState()
   &SpawnBurst_Instantaneous(
    Spawn Count=&MakeCustomIntFromBool(Bool=$User.Golden, True INT=&RandomRangeInt(Minimum=15, Maximum=48), False INT=0)
    Use Spawn Probability=false
    UseLoopCountLimit=false
   )
  ]
  particleSpawnScript [
   /SetVariables(Particles.VisibilityTag=&SpawnGroup()) {isSetParameters=true, module=SetVariables_#1}
   &InitializeParticle(
    Color=<rgba 0.7157 0.3005 0 1>
    Mass=&MaskFloatBySpawnGroup(Masked Float=4, Passthrough Float=1)
    Mass Mode=::NewEnumerator1
    Mesh Renderer Array Visibility Mode=::NewEnumerator3
    Mesh Renderer Info=DataInterface() {propertyValues {MeshRenderer=&self:GoldDebris_0.NiagaraMeshRendererProperties_0}}
    Mesh Scale Mode=::NewEnumerator3
    Mesh Uniform Scale=&RandomRangeFloat(Minimum=0.3, Maximum=0.8)
    UsePositionOffset=false
    Write Material Random=true
   )
   &ShapeLocation(
    Handle Radius:NiagaraFloat=&Lerp_Float(A=70, B=190, Alpha=$User.Size)
    Large Radius=&Lerp_Float(A=170, B=375, Alpha=$User.Size)
    MaskSpawnGroup=false
    Non Uniform Scale=<1 1 0.8>
    Offset Mode=::NewEnumerator0
    Override Seed=false
    Rotation Axis=<0 0 1>
    Rotation Mode=::NewEnumerator1
    Shape Primitive=::NewEnumerator3
    UseOwnerScale=true
    V Distribution:NiagaraFloat=&MaskFloat(Mask Source=0, Mask Parameter=?, Mask=0, Masked Float=0, Passthrough Float=0.5)
   )
   &InitialMeshOrientation(Use Rotation Vector=false)
   &AddVelocity(
    Module.Use Falloff Distance=false
    Override Seed=false
    Velocity Mode=::NewEnumerator1
    Velocity Speed=&RandomRangeFloat(Minimum=250, Maximum=750)
   )
   &AddVelocity(Rotation Coordinate Space=::World, Velocity=&RandomRangeVector(Minimum=<0 0 200>, Maximum=<0 0 800>)) {module=AddVelocity001}
   &AddVelocity(
    Rotation Coordinate Space=::World
    Velocity=&RandomRangeVector(Minimum=<-250 -250 0>, Maximum=<250 250 1>)
   ) {module=AddVelocity002}
   &AddRotationalVelocity(Rotation Rate=&RandomVector(Vector Scale=1))
   &CameraOffset(Camera Offset Amount=50)
  ]
  particleUpdateScript [
   &ParticleState()
   &ScaleMeshSize(Scale Factor=&VectorFromFloat(Value=&FloatFromCurve(
      FloatCurve=DataInterface() {propertyValues {Curve {
         defaultValue=1
         keys [
          {arriveTangent=38.0665, interpMode=RCIM_Cubic, leaveTangent=38.0665, tangentMode=RCTM_User, time=0, value=0}
          <at 0.07 1 RCIM_Cubic>
          {arriveTangent=-0.7649, interpMode=RCIM_Cubic, leaveTangent=-0.7649, tangentMode=RCTM_User, time=1, value=0}
         ]
        }}}
      Scale Curve=&Lerp_Float(A=1, B=2.5, Alpha=$User.Size)
     )))
   &Drag(Drag=2, Use Linear Drag=true, Use Rotational Drag=true)
   &GravityForce()
   &SolveForcesAndVelocity(Clamp Velocity=false, Limit Acceleration=false)
  ]
  rendererClasses [&NiagaraMeshRendererProperties]
  renderers [{rendererClass=&NiagaraMeshRendererProperties, rendererIndex=0}]
 }
 LingerSmoke {
  emitterUpdateScript [
   &EmitterState()
   &SpawnBurst_Instantaneous(
    Spawn Count=&MakeIntFromFloat(Float=&Lerp_Float(
      A=&MakeFloatFromInt(Int=14)
      B=&MakeFloatFromInt(Int=26)
      Alpha=$User.Size
     ))
    Use Spawn Probability=true
    UseLoopCountLimit=false
   )
  ]
  particleSpawnScript [
   &InitializeParticle(
    Color=&MakeCustomLinearColorFromBool(
     Bool=$User.Golden
     True Color=&RandomRangeLinearColor(Minimum=<rgba 0.625 0.2958 0.0391 1>, Maximum=<rgba 0.5417 0.1888 0.0282 1>)
     False Color=&RandomRangeLinearColor(Minimum=<rgba 0.3646 0.1907 0.0563 1>, Maximum=<rgba 0.3385 0.1771 0.0522 1>)
    )
    Lifetime Max=1.7
    Lifetime Min=0.8
    Lifetime Mode=::NewEnumerator1
    Sprite Rotation Mode=::NewEnumerator2
    Sprite Size Mode=::NewEnumerator1
    Uniform Sprite Size Max=375
    Uniform Sprite Size Min=175
    UsePositionOffset=false
    Write Material Random=true
   )
   &ShapeLocation(
    Handle Radius:NiagaraFloat=&Lerp_Float(A=70, B=160, Alpha=$User.Size)
    Large Radius=&Lerp_Float(A=120, B=300, Alpha=$User.Size)
    MaskSpawnGroup=false
    Offset Mode=::NewEnumerator0
    Rotation Axis=<0 0 1>
    Rotation Mode=::NewEnumerator1
    Shape Primitive=::NewEnumerator3
    Torus Distribution Mode:ENiagaraTorusDistributionMode=::NewEnumerator1
    U Position=&ReturnNormalizedExecIndex()
    UseOwnerScale=true
    V Position=1
   )
   &AddVelocity(Module.Use Falloff Distance=false, Origin Offset=<0 0 60>, Override Seed=false, Velocity Mode=::NewEnumerator1, Velocity Speed=100)
   &CameraOffset(Camera Offset Amount=50)
  ]
  particleUpdateScript [
   &ParticleState()
   &Drag(Drag=2, Use Linear Drag=true, Use Rotational Drag=true)
   &SolveForcesAndVelocity(Clamp Velocity=false, Limit Acceleration=false)
   &SpriteRotationRate(Rotation Rate=&Multiply_Float(
     A=&UniformRangedFloat(Minimum=1, Maximum=1)
     B=&FloatFromCurve(
      FloatCurve=DataInterface() {propertyValues {Curve {keys [{arriveTangent=-3.245, interpMode=RCIM_Cubic, leaveTangent=-3.245, tangentMode=RCTM_User, time=0, value=1}, <at 0.1 0.05>, <at 1.0011 0.01>]}}}
      Scale Curve=350
     )
    ))
   &ScaleSpriteSize(Scale Factor=&Vector2DFromFloat(Value=&FloatFromCurve(
      FloatCurve=DataInterface() {propertyValues {Curve {keys [
          {arriveTangent=4.8654, interpMode=RCIM_Cubic, leaveTangent=4.8654, tangentMode=RCTM_User, time=0, value=0.25}
          {arriveTangent=0.75, interpMode=RCIM_Cubic, leaveTangent=0.75, time=0.2, value=0.75}
          {interpMode=RCIM_Cubic, leaveTangent=0.155, time=1, value=1}
         ]}}}
      Scale Curve=&Lerp_Float(A=1.6, B=1.7, Alpha=$User.Size)
     )))
   &ScaleColor(
    Scale Alpha=&FloatFromCurve(
     FloatCurve=DataInterface() {propertyValues {Curve {keys [
         {arriveTangent=0.9606, interpMode=RCIM_Cubic, leaveTangent=0.9606, tangentMode=RCTM_User, time=0.0029, value=0.0488}
         {interpMode=RCIM_Cubic, tangentMode=RCTM_User, time=0.025, value=1}
         {interpMode=RCIM_Cubic, tangentMode=RCTM_User, time=0.33, value=1}
         {arriveTangent=-1.735, interpMode=RCIM_Cubic, leaveTangent=-1.735, tangentMode=RCTM_User, time=1, value=0}
        ]}}}
     Scale Curve=0.5
    )
    Scale RGB=&MakeVectorFromLinearColorRGB(Color=&ColorFromCurve(ColorCurve=DataInterface() {propertyValues {
        AlphaCurve {keys [<at 0 1 RCIM_Cubic>]}
        BlueCurve {keys [<at 0 0.5521 RCIM_Cubic>,<at 0.1472 1 RCIM_Cubic>]}
        GreenCurve {keys [<at 0 0.5521 RCIM_Cubic>,<at 0.1472 1 RCIM_Cubic>]}
        RedCurve {keys [<at 0 0.5521 RCIM_Cubic>,<at 0.1472 1 RCIM_Cubic>]}
       }}))
    ScaleA=true
    ScaleRGB=true
   )
   &DynamicMaterialParameters(
    Index 0 Param 1=&Add_Float(
     A=&RandomRangeFloat(Minimum=0, Maximum=16)
     B=&FloatFromCurve(
      FloatCurve=DataInterface() {propertyValues {Curve {keys [
          {arriveTangent=-2.2561, interpMode=RCIM_Cubic, leaveTangent=-2.2561, tangentMode=RCTM_User, time=0, value=1}
          {arriveTangent=-1.0036, interpMode=RCIM_Cubic, leaveTangent=-1.0036, tangentMode=RCTM_User, time=0.2682, value=0.4215}
          <at 1 0>
         ]}}}
      Scale Curve=-0.5
     )
    )
    Index 1 Param 1=&FloatFromCurve(FloatCurve=DataInterface() {propertyValues {Curve {keys [{arriveTangent=-14.778, interpMode=RCIM_Cubic, time=0, value=1}, {interpMode=RCIM_Cubic, leaveTangent=-3.9602, time=0.2, value=1}]}}})
    Index 1 Param 2=&FloatFromCurve(
     FloatCurve=DataInterface() {propertyValues {Curve {keys [{tangentMode=RCTM_User, time=0, value=0}, {tangentMode=RCTM_User, time=1, value=1}]}}}
     Scale Curve=-0.2
    )
    Index 1 Param 3=1
    Index 1 Param 4=&FloatFromCurve(FloatCurve=DataInterface() {propertyValues {Curve {keys [{arriveTangent=-5.8231, time=0, value=1.5}, {leaveTangent=-3.9602, time=0.2, value=0.5}]}}})
    Index 2 Param 1=&FloatFromCurve(FloatCurve=DataInterface() {propertyValues {Curve {keys [{arriveTangent=2.5015, interpMode=RCIM_Cubic, time=0.25, value=0.5}, {interpMode=RCIM_Cubic, leaveTangent=1.688, time=0.75, value=1}]}}})
    Index 2 Param 2=&Add_Float(
     A=&RandomRangeFloat(Minimum=-1, Maximum=1)
     B=&FloatFromCurve(
      FloatCurve=DataInterface() {propertyValues {Curve {keys [{arriveTangent=2.6275, interpMode=RCIM_Cubic, leaveTangent=2.6275, tangentMode=RCTM_User, time=0.25, value=0}, <at 1 1 RCIM_Cubic>]}}}
      Scale Curve=&RandomRangeFloat(Minimum=-0.25, Maximum=-0.5)
     )
    )
    Index 2 Param 3=&FloatFromCurve(FloatCurve=DataInterface() {propertyValues {Curve {keys [{arriveTangent=0.6717, interpMode=RCIM_Cubic, time=0.25, value=0.8}, {interpMode=RCIM_Cubic, leaveTangent=2.715, time=1.0011, value=1}]}}})
    Index 2 Param 4=&FloatFromCurve(
     FloatCurve=DataInterface() {propertyValues {Curve {keys [{arriveTangent=-0.4045, time=0, value=0.75}, {leaveTangent=-0.318, time=1, value=1}]}}}
     Scale Curve=&RandomRangeFloat(Minimum=0.4, Maximum=0.5)
    )
    Module.Index0Float0Write=true
    Module.Index0Float1Write=true
    Module.Index0Float2Write=true
    Module.Index0Float3Write=true
    Module.Index1Float0Write=true
    Module.Index1Float1Write=true
    Module.Index1Float2Write=true
    Module.Index1Float3Write=true
    Module.Index2Float0Write=true
    Module.Index2Float1Write=true
    Module.Index2Float2Write=true
    Module.Index2Float3Write=true
    Write Parameter Index 1=::NewEnumerator1
    Write Parameter Index 2=::NewEnumerator1
   )
  ]
  rendererClasses [&NiagaraSpriteRendererProperties]
  renderers [{rendererClass=&NiagaraSpriteRendererProperties}]
 }
 Liquid_Particles {
  emitterUpdateScript [
   &EmitterState()
   &SpawnBurst_Instantaneous(
    Spawn Count=&MakeCustomIntFromBool(
     Bool=$User.Golden
     True INT=0
     False INT=&MakeIntFromFloat(Float=&Lerp_Float(
       A=&MakeFloatFromInt(Int=&RandomRangeInt(Minimum=16, Maximum=28))
       B=&MakeFloatFromInt(Int=&RandomRangeInt(Minimum=28, Maximum=36))
       Alpha=$User.Size
      ))
    )
    Use Spawn Probability=true
    UseLoopCountLimit=false
   )
   &SpawnBurst_Instantaneous(
    Spawn Count=&MakeCustomIntFromBool(
     Bool=$User.Golden
     True INT=0
     False INT=&MakeIntFromFloat(Float=&Lerp_Float(
       A=&MakeFloatFromInt(Int=&RandomRangeInt(Minimum=13, Maximum=18))
       B=&MakeFloatFromInt(Int=&RandomRangeInt(Minimum=27, Maximum=30))
       Alpha=$User.Size
      ))
    )
    Spawn Group=1
    Use Spawn Probability=true
    UseLoopCountLimit=false
   ) {module=SpawnBurst_Instantaneous001}
  ]
  particleSpawnScript [
   &InitializeParticle(
    Color=$User.Frosting
    Lifetime Max=1.6
    Lifetime Min=0.75
    Lifetime Mode=::NewEnumerator1
    Sprite Rotation Mode=::NewEnumerator2
    Sprite Size=/Niagara/DynamicInputs/Helpers/MaskVector2d(
     Mask Source:NiagaraInt32=0
     Mask Parameter:NiagaraInt32=?
     Mask:NiagaraInt32=0
     Masked Vector2D:Vector2f=&RandomRangeVector2D(Minimum=<15 30>, Maximum=<25 50>)
     Passthrough Vector2D:Vector2f=&Vector2DFromFloat(Value=&Multiply_Float(
       A=&RandomRangeFloat(Minimum=75, Maximum=210)
       B=&Lerp_Float(A=0.8, B=1, Alpha=$User.Size)
      ))
    )
    Sprite Size Mode=::NewEnumerator4
    Sprite UV Mode=::NewEnumerator3
    UsePositionOffset=false
    Write Material Random=true
   )
   /Niagara/Modules/Update/SubUV/SubUVAnimation(Number Of Frames=8, SubUV Animation Mode=::NewEnumerator1)
   &ShapeLocation(
    Handle Radius:NiagaraFloat=&Lerp_Float(A=90, B=190, Alpha=$User.Size)
    Large Radius=&Lerp_Float(A=195, B=375, Alpha=$User.Size)
    MaskSpawnGroup=false
    Non Uniform Scale=<1 1 -0.8>
    Offset Mode=::NewEnumerator0
    Override Seed=false
    Rotation Axis=<0 0 1>
    Rotation Mode=::NewEnumerator1
    Shape Primitive=::NewEnumerator3
    Surface Distribution=1
    UseOwnerScale=true
    V Distribution=0.5
   )
   &AddVelocity(
    Module.Use Falloff Distance=false
    Origin Offset=<0 0 60>
    Override Seed=false
    Velocity Mode=::NewEnumerator1
    Velocity Speed=&Lerp_Float(
     A=&RandomRangeFloat(Minimum=50, Maximum=275)
     B=&RandomRangeFloat(Minimum=275, Maximum=400)
     Alpha=$User.Size
    )
   )
   &AddVelocity(Velocity=&RandomRangeVector(Minimum=<0 0 25>, Maximum=<0 0 400>)) {module=AddVelocity001}
   &AddVelocity(Rotation Coordinate Space=::World, Velocity=&RandomRangeVector(Minimum=<0 0 200>, Maximum=<0 0 800>)) {module=AddVelocity002}
   /SetVariables(Particles.VisibilityTag=&SpawnGroup()) {isSetParameters=true}
   &CameraOffset(Camera Offset Amount=50)
  ]
  particleUpdateScript [
   &ParticleState()
   &Drag(
    Drag=&SelectFloatFromArray(
     Direct Array Index=&SpawnGroup()
     Float Selection Array=DataInterface() {propertyValues {FloatData [1.25,1,1.5]}}
    )
    Use Linear Drag=true
    Use Rotational Drag=true
   )
   &GravityForce()
   &SolveForcesAndVelocity(Clamp Velocity=false, Limit Acceleration=false)
   &SpriteRotationRate(Rotation Rate=&RandomRangeFloat(Minimum=-40, Maximum=40))
   &ScaleSpriteSize(
    Scale Factor=&Multiply_Vector2D_ByFloat(
     Float=&FloatFromCurve(
      FloatCurve=DataInterface() {propertyValues {Curve {keys [
          {arriveTangent=4.1785, interpMode=RCIM_Cubic, leaveTangent=4.1785, tangentMode=RCTM_User, time=0, value=0.2}
          {arriveTangent=0.8, interpMode=RCIM_Cubic, leaveTangent=0.8, time=0.1, value=0.7}
          {arriveTangent=0.0224, interpMode=RCIM_Cubic, leaveTangent=0.0224, tangentMode=RCTM_User, time=1, value=1}
         ]}}}
      Scale Curve=1.5
     )
     Vector2D=&Lerp_Vector2(
      A=<1 0.5>
      B=<1 1>
      Alpha=&FloatFromCurve(FloatCurve=DataInterface() {propertyValues {Curve {keys [
           {arriveTangent=4.1785, interpMode=RCIM_Cubic, leaveTangent=4.1785, tangentMode=RCTM_User, time=0, value=0}
           {arriveTangent=1.1179, interpMode=RCIM_Cubic, leaveTangent=1.1179, tangentMode=RCTM_User, time=0.2042, value=0.6228}
           {arriveTangent=0.0224, interpMode=RCIM_Cubic, leaveTangent=0.0224, tangentMode=RCTM_User, time=1, value=1}
          ]}}})
     )
    )
    Scale Sprite Size Mode=::NewEnumerator2
   )
   &DynamicMaterialParameters(
    Index 0 Param 1=&RandomRangeFloat(Minimum=0, Maximum=1)
    Index 0 Param 2=&FloatFromCurve(
     FloatCurve=DataInterface() {propertyValues {Curve {keys [{interpMode=RCIM_Cubic, tangentMode=RCTM_User, time=0.33, value=0}, {interpMode=RCIM_Cubic, leaveTangent=1.2673, time=0.8, value=1}]}}}
     Scale Curve=0
    )
    Index 0 Param 3=&FloatFromCurve(FloatCurve=DataInterface() {propertyValues {Curve {keys [{tangentMode=RCTM_User, time=0, value=1}, {tangentMode=RCTM_User, time=1, value=1}]}}})
    Index 0 Param 4=&FloatFromCurve(FloatCurve=DataInterface() {propertyValues {Curve {keys [
         {arriveTangent=-14.9399, interpMode=RCIM_Cubic, leaveTangent=-14.9399, tangentMode=RCTM_User, time=0, value=1}
         {interpMode=RCIM_Cubic, tangentMode=RCTM_User, time=0.15, value=0}
         {interpMode=RCIM_Cubic, tangentMode=RCTM_User, time=0.33, value=0}
         {arriveTangent=2.7964, interpMode=RCIM_Cubic, leaveTangent=2.7964, tangentMode=RCTM_User, time=1, value=1}
        ]}}})
    Index 1 Param 2=&FloatFromCurve(FloatCurve=DataInterface() {propertyValues {Curve {keys [
         {arriveTangent=23.2225, interpMode=RCIM_Cubic, leaveTangent=23.2225, tangentMode=RCTM_User, time=0, value=0}
         <at 0.1 1 RCIM_Cubic>
         {arriveTangent=-3.2993, interpMode=RCIM_Cubic, leaveTangent=-3.2993, tangentMode=RCTM_User, time=1, value=0}
        ]}}})
    Module.Index0Float0Write=true
    Module.Index0Float1Write=true
    Module.Index0Float2Write=true
    Module.Index0Float3Write=true
    Module.Index1Float0Write=true
    Module.Index1Float1Write=true
    Module.Index1Float2Write=true
    Module.Index1Float3Write=true
    Write Parameter Index 1=::NewEnumerator1
   )
   &ScaleColor(
    Color Scale:LinearColor=&FNTimeOfDayColor(Night Color=<rgba 0.4167 0.4167 0.4167 1>, Morning Color=<rgba 1 1 1 1>, Day Color=<rgba 1 1 1 1>, Evening Color=<rgba 1 1 1 1>)
    Scale Mode=3
   )
  ]
  rendererClasses [&NiagaraSpriteRendererProperties]
  renderers [
   {rendererClass=&NiagaraSpriteRendererProperties}
   {rendererClass=&NiagaraSpriteRendererProperties, rendererIndex=1}
  ]
 }
 NE_SmallRadiusDroplets001 {
  emitterUpdateScript [
   &EmitterState()
   &SpawnBurst_Instantaneous(
    Spawn Count=&MakeCustomIntFromBool(Bool=$User.Golden, True INT=&RandomRangeInt(Minimum=27, Maximum=65), False INT=0)
    Use Spawn Probability=false
    UseLoopCountLimit=false
   )
  ]
  particleSpawnScript [
   &InitializeParticle(
    Color=&RandomRangeLinearColor(Minimum=<rgba 1 0.6889 0 1>, Maximum=<rgba 1 0.3404 0 1>)
    Lifetime Max=1
    Lifetime Min=0.6
    Lifetime Mode=::NewEnumerator1
    Sprite Rotation Mode=::NewEnumerator2
    Sprite Size Mode=::NewEnumerator1
    Uniform Sprite Size Max=20
    Uniform Sprite Size Min=4
    UsePositionOffset=false
    Write Material Random=true
   )
   &ShapeLocation(
    Handle Radius:NiagaraFloat=&Lerp_Float(A=70, B=190, Alpha=$User.Size)
    Large Radius=&Lerp_Float(A=170, B=375, Alpha=$User.Size)
    MaskSpawnGroup=false
    Non Uniform Scale=<1 1 0.8>
    Offset Mode=::NewEnumerator0
    Override Seed=false
    Rotation Axis=<0 0 1>
    Rotation Mode=::NewEnumerator1
    Shape Primitive=::NewEnumerator3
    Surface Distribution=1
    UseOwnerScale=true
   )
   &AddVelocity(
    Module.Use Falloff Distance=false
    Override Seed=false
    Velocity Mode=::NewEnumerator1
    Velocity Speed=&RandomRangeFloat(Minimum=300, Maximum=750)
   )
  ]
  particleUpdateScript [
   &ParticleState()
   /SetVariables(Particles.GlitterAlpha:NiagaraFloat=&FloatFromCurve(
     FloatCurve=DataInterface() {propertyValues {Curve {keys [
         {interpMode=RCIM_Cubic, tangentMode=RCTM_User, time=0, value=0}
         {interpMode=RCIM_Cubic, tangentMode=RCTM_User, time=0.02, value=1}
         {interpMode=RCIM_Cubic, tangentMode=RCTM_User, time=0.25, value=0}
        ]}}}
     CurveIndex=&ModuloFloat(
      A=&Add_Float(
       A=&Multiply_Float(A=$Emitter.Age, B=2)
       B=&RandomRangeFloat(Minimum=0, Maximum=1)
      )
      B=1
     )
    )) {isSetParameters=true}
   &GravityForce()
   &Drag(Drag=2, Use Linear Drag=true, Use Rotational Drag=true)
   &CurlNoiseForce(Bias Noise=false, Noise Frequency=5, Noise Strength=500, Pan Noise=false)
   &SolveForcesAndVelocity(Clamp Velocity=false, Limit Acceleration=false)
   &ScaleSpriteSize(
    Uniform Curve Scale=&Lerp_Float(A=1, B=2, Alpha=$User.Size)
    Uniform Curve Sprite Scale=DataInterface() {propertyValues {Curve {keys [
        {interpMode=RCIM_Cubic, tangentMode=RCTM_User, time=0, value=0}
        {interpMode=RCIM_Cubic, tangentMode=RCTM_User, time=0.1, value=1}
        {arriveTangent=-3.5627, interpMode=RCIM_Cubic, leaveTangent=-3.5627, tangentMode=RCTM_User, time=1, value=0}
       ]}}}
   )
   &SpriteRotationRate(Rotation Rate=&RandomRangeFloat(Minimum=-20, Maximum=20))
   &ScaleColor(
    Scale RGB=&MakeVectorFromLinearColorRGB(Color=&LerpLinearColors(StartColor=<rgba 1 1 1 1>, EndColor=<rgba 20 18.8489 0 1>, LerpFactor=$Particles.GlitterAlpha))
    ScaleA=true
    ScaleRGB=true
   )
   &ScaleColor(Scale RGB=<4 4 4>, ScaleA=true, ScaleRGB=true) {module=ScaleColor001}
   /SetVariables(Particles.FLareSize:Vector2f=&Multiply_Vector2D_ByFloat(
     Float=&Lerp_Float(A=0, B=1, Alpha=$Particles.GlitterAlpha)
     Vector2D=&Multiply_Vector2D_ByFloat(Float=&Lerp_Float(A=4, B=12, Alpha=$User.Size), Vector2D=<5 8>)
    )) {isSetParameters=true}
   /SetVariables(Particles.FlareColor=&ColorFromCurve(
     ColorCurve=DataInterface() {propertyValues {
       AlphaCurve {keys [<at 0 0 RCIM_Cubic>,<at 1 0.2 RCIM_Cubic>]}
       BlueCurve {keys [<at 0 0 RCIM_Cubic>,<at 1 0 RCIM_Cubic>]}
       GreenCurve {keys [<at 0 0.383 RCIM_Cubic>,<at 1 0.4596 RCIM_Cubic>]}
       RedCurve {keys [<at 0 1 RCIM_Cubic>,<at 1 1 RCIM_Cubic>]}
      }}
     CurveIndex=$Particles.GlitterAlpha
    )) {isSetParameters=true}
   &DynamicMaterialParameters(
    Index 0 Param 1=1
    Index 0 Param 2=0.5
    Module.Index0Float0Write=true
    Module.Index0Float1Write=true
    Module.Index0Float2Write=true
    Module.Index0Float3Write=true
   )
  ]
  rendererClasses [&NiagaraSpriteRendererProperties]
  renderers [
   {rendererClass=&NiagaraSpriteRendererProperties}
   {rendererClass=&NiagaraSpriteRendererProperties, rendererIndex=1}
  ]
 }
}
summary {
 systemName=NS_SonicSpear_Destruction
 emitters [
  {emitterName=LingerSmoke, rendererClasses [&NiagaraSpriteRendererProperties]}
  {emitterName=Food_Debris, rendererClasses [&NiagaraMeshRendererProperties]}
  {emitterName=Liquid_Particles, rendererClasses [&NiagaraSpriteRendererProperties]}
  {bEnabled=false, emitterName=Donut_Dissolve, rendererClasses [&NiagaraMeshRendererProperties]}
  {emitterName=NE_SmallRadiusDroplets001, rendererClasses [&NiagaraSpriteRendererProperties]}
  {emitterName=GoldDebris, rendererClasses [&NiagaraMeshRendererProperties]}
  {emitterName=Food_Debris001, rendererClasses [&NiagaraMeshRendererProperties]}
 ]
 userVariables [
  User.Frosting {type=LinearColor, description="", defaultValue {struct=LinearColor, value=<rgba 1 0.1913 0.6753 1>}}
  User.Golden {type=NiagaraBool}
  User.Size {type=NiagaraFloat, defaultValue=0, description=""}
 ]
}
