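# Niagara effect for the Sonic Spear fall, with rising debris, glitter and wind streaks spawned around a ring.
# User.Golden swaps multicolored Food_Debris meshes for glittering golden sprites with curl noise and pulsing flares.
# Wind_Ribbons are upright meshes that stretch over life, animated through dynamic material parameters and faded in over system age.
# User.Size lerps radii and scales, while User.StopSpawning halts spawning and hides the wind.
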NiagaraSystem /SonicSpearGameplay/Assets/VFX/NS_SonicSpear_Fall
dependencies {
 usedDataInterfaces [
  &NiagaraDataInterfaceMeshRendererInfo
  &NiagaraDataInterfaceCurve
  &NiagaraDataInterfaceArrayColor
  &NiagaraDataInterfaceColorCurve
 ]
 usedDynamicInputs [&UniformRangedFloat]
 usedModules [&SystemState]
}
emitters {
 Food_Debris {
  emitterUpdateScript [
   &EmitterState()
   &SpawnRate(
    SpawnRate=&MakeCustomFloatFromBool(
     Bool=$User.Golden
     True Float=0
     False Float=&MakeCustomFloatFromBool(
      Bool=$User.StopSpawning
      True Float=0
      False Float=&FloatFromCurve(
       FloatCurve=DataInterface() {propertyValues {Curve {
          defaultValue=1
          keys [
           {arriveTangent=5.391, interpMode=RCIM_Cubic, leaveTangent=5.391, tangentMode=RCTM_User, time=0, value=0}
           {arriveTangent=0.0224, interpMode=RCIM_Cubic, leaveTangent=0.0224, tangentMode=RCTM_User, time=0.5, value=1}
          ]
         }}}
       CurveIndex=$Emitter.NormalizedLoopAge
       Scale Curve=5
      )
     )
    )
    Use Spawn Probability=false
   )
  ]
  particleSpawnScript [
   &InitializeParticle(
    Color=&SelectLinearColorFromArray(
     Array Sampling Mode=::NewEnumerator0
     Direct Array Index=?
     Color Selection Array=DataInterface() {propertyValues {ColorData [
        <rgba 1 0.0437 0.0685 1>
        <rgba 0 0.4621 0.9387 1>
        <rgba 0.007 0.5972 0.2016 1>
        <rgba 1 1 0.8308 1>
        <rgba 0.6445 0.314 0.0931 1>
        <rgba 1 1 0 1>
       ]}}
     Random Seed=0
    )
    Lifetime=&RandomRangeFloat(Minimum=0.8, Maximum=1.2)
    Mesh Scale=&Lerp_Vector(A=<18 10 10>, B=<36 20 20>, Alpha=$User.Size)
    Mesh Scale Mode=::NewEnumerator4
    UsePositionOffset=false
    Write Material Random=true
   )
   &ShapeLocation(
    Handle Radius:NiagaraFloat=&Lerp_Float(A=70, B=190, Alpha=$User.Size)
    Large Radius=&Lerp_Float(A=170, B=375, Alpha=$User.Size)
    MaskSpawnGroup=false
    Non Uniform Scale=<1 1 0.8>
    Offset Mode=::NewEnumerator0
    Override Seed=false
    Rotation Axis=<0 1 0>
    Rotation Mode=::NewEnumerator1
    Shape Primitive=::NewEnumerator3
    Surface Distribution=1
    UseOwnerScale=true
   )
   &InitialMeshOrientation(Use Rotation Vector=false)
   &AddVelocity(Rotation Coordinate Space=::World, Velocity=&RandomRangeVector(Minimum=<0 0 300>, Maximum=<0 0 800>))
   &AddRotationalVelocity(Rotation Rate=&RandomVector(Vector Scale=1))
   &CameraOffset(Camera Offset Amount=50)
  ]
  particleUpdateScript [
   &ParticleState()
   &ScaleMeshSize(Scale Factor=&VectorFromFloat(Value=&FloatFromCurve(
      FloatCurve=DataInterface() {propertyValues {Curve {
         defaultValue=1
         keys [
          {arriveTangent=38.0665, interpMode=RCIM_Cubic, leaveTangent=38.0665, tangentMode=RCTM_User, time=0, value=0}
          {arriveTangent=1.4286, leaveTangent=1.4286, time=0.07, value=1}
          {arriveTangent=0.1167, interpMode=RCIM_Cubic, leaveTangent=0.1167, tangentMode=RCTM_User, time=0.7, value=1}
          {arriveTangent=-10.8391, interpMode=RCIM_Cubic, leaveTangent=-10.8391, tangentMode=RCTM_User, time=1, value=0}
         ]
        }}}
      Scale Curve=&MaskFloatBySpawnGroup(Masked Float=1.5, Passthrough Float=1.2)
     )))
   &Drag(Use Linear Drag=true, Use Rotational Drag=true)
   &GravityForce()
   &SolveForcesAndVelocity(Clamp Velocity=false, Limit Acceleration=false)
   &ScaleColor(
    Color Scale:LinearColor=&FNTimeOfDayColor(Night Color=<rgba 0.4167 0.4167 0.4167 1>, Morning Color=<rgba 1 1 1 1>, Day Color=<rgba 1 1 1 1>, Evening Color=<rgba 1 1 1 1>)
    Scale Mode=3
   )
  ]
  rendererClasses [&NiagaraMeshRendererProperties]
  renderers [
   {rendererClass=&NiagaraMeshRendererProperties, rendererIndex=0}
   {rendererClass=&NiagaraMeshRendererProperties, rendererIndex=1}
  ]
 }
 NE_SmallRadiusDroplets001 {
  emitterUpdateScript [
   &EmitterState()
   &SpawnRate(
    SpawnRate=&MakeCustomFloatFromBool(
     Bool=$User.Golden
     True Float=&MakeCustomFloatFromBool(
      Bool=$User.StopSpawning
      True Float=0
      False Float=&FloatFromCurve(
       FloatCurve=DataInterface() {propertyValues {Curve {
          defaultValue=1
          keys [
           {arriveTangent=5.391, interpMode=RCIM_Cubic, leaveTangent=5.391, tangentMode=RCTM_User, time=0, value=0}
           {arriveTangent=0.0224, interpMode=RCIM_Cubic, leaveTangent=0.0224, tangentMode=RCTM_User, time=0.5, value=1}
          ]
         }}}
       CurveIndex=$Emitter.NormalizedLoopAge
       Scale Curve=25
      )
     )
     False Float=0
    )
    Use Spawn Probability=false
   )
  ]
  particleSpawnScript [
   &InitializeParticle(
    Color=&RandomRangeLinearColor(Minimum=<rgba 1 0.6889 0 1>, Maximum=<rgba 1 0.3404 0 1>)
    Lifetime Max=1
    Lifetime Min=0.6
    Lifetime Mode=::NewEnumerator1
    Sprite Rotation Mode=::NewEnumerator2
    Sprite Size Mode=::NewEnumerator1
    Uniform Sprite Size Max=40
    Uniform Sprite Size Min=12
    UsePositionOffset=false
    Write Material Random=true
   )
   &ShapeLocation(
    Handle Radius:NiagaraFloat=&Lerp_Float(A=70, B=190, Alpha=$User.Size)
    Large Radius=&Lerp_Float(A=170, B=375, Alpha=$User.Size)
    MaskSpawnGroup=false
    Non Uniform Scale=<1 1 0.8>
    Offset Mode=::NewEnumerator0
    Override Seed=false
    Rotation Axis=<0 1 0>
    Rotation Mode=::NewEnumerator1
    Shape Primitive=::NewEnumerator3
    Surface Distribution=1
    UseOwnerScale=true
   )
   &AddVelocity(Rotation Coordinate Space=::World, Velocity=&RandomRangeVector(Minimum=<0 0 300>, Maximum=<0 0 800>))
  ]
  particleUpdateScript [
   &ParticleState()
   /SetVariables(Particles.GlitterAlpha:NiagaraFloat=&FloatFromCurve(
     FloatCurve=DataInterface() {propertyValues {Curve {keys [
         {interpMode=RCIM_Cubic, tangentMode=RCTM_User, time=0, value=0}
         {interpMode=RCIM_Cubic, tangentMode=RCTM_User, time=0.02, value=1}
         {interpMode=RCIM_Cubic, tangentMode=RCTM_User, time=0.25, value=0}
        ]}}}
     CurveIndex=&ModuloFloat(
      A=&Add_Float(
       A=&Multiply_Float(A=$Emitter.Age, B=2)
       B=&RandomRangeFloat(Minimum=0, Maximum=1)
      )
      B=1
     )
    )) {isSetParameters=true}
   &GravityForce()
   &Drag(Use Linear Drag=true, Use Rotational Drag=true)
   &CurlNoiseForce(Bias Noise=false, Noise Frequency=5, Noise Strength=500, Pan Noise=false)
   &SolveForcesAndVelocity(Clamp Velocity=false, Limit Acceleration=false)
   &ScaleSpriteSize(
    Uniform Curve Scale=&Lerp_Float(A=1, B=2, Alpha=$User.Size)
    Uniform Curve Sprite Scale=DataInterface() {propertyValues {Curve {keys [
        {interpMode=RCIM_Cubic, tangentMode=RCTM_User, time=0, value=0}
        {interpMode=RCIM_Cubic, tangentMode=RCTM_User, time=0.1, value=1}
        {arriveTangent=-3.5627, interpMode=RCIM_Cubic, leaveTangent=-3.5627, tangentMode=RCTM_User, time=1, value=0}
       ]}}}
   )
   &SpriteRotationRate(Rotation Rate=&RandomRangeFloat(Minimum=-20, Maximum=20))
   &ScaleColor(
    Scale RGB=&MakeVectorFromLinearColorRGB(Color=&LerpLinearColors(StartColor=<rgba 1 1 1 1>, EndColor=<rgba 20 18.8489 0 1>, LerpFactor=$Particles.GlitterAlpha))
    ScaleA=true
    ScaleRGB=true
   )
   &ScaleColor(Scale RGB=<4 4 4>, ScaleA=true, ScaleRGB=true) {module=ScaleColor001}
   /SetVariables(Particles.FLareSize:Vector2f=&Multiply_Vector2D_ByFloat(
     Float=&Lerp_Float(A=0, B=1, Alpha=$Particles.GlitterAlpha)
     Vector2D=&Multiply_Vector2D_ByFloat(Float=&Lerp_Float(A=12, B=24, Alpha=$User.Size), Vector2D=<5 8>)
    )) {isSetParameters=true}
   /SetVariables(Particles.FlareColor=&ColorFromCurve(
     ColorCurve=DataInterface() {propertyValues {
       AlphaCurve {keys [<at 0 0 RCIM_Cubic>,<at 1 0.7 RCIM_Cubic>]}
       BlueCurve {keys [<at 0 0 RCIM_Cubic>,<at 1 0 RCIM_Cubic>]}
       GreenCurve {keys [<at 0 0.383 RCIM_Cubic>,<at 1 0.4596 RCIM_Cubic>]}
       RedCurve {keys [<at 0 1 RCIM_Cubic>,<at 1 1 RCIM_Cubic>]}
      }}
     CurveIndex=$Particles.GlitterAlpha
    )) {isSetParameters=true}
   &DynamicMaterialParameters(
    Index 0 Param 1=1
    Index 0 Param 2=0.5
    Module.Index0Float0Write=true
    Module.Index0Float1Write=true
    Module.Index0Float2Write=true
    Module.Index0Float3Write=true
   )
  ]
  rendererClasses [&NiagaraSpriteRendererProperties]
  renderers [
   {rendererClass=&NiagaraSpriteRendererProperties}
   {rendererClass=&NiagaraSpriteRendererProperties, rendererIndex=1}
  ]
 }
 Wind_Ribbons {
  emitterUpdateScript [
   &EmitterState()
   &SpawnRate(
    SpawnRate=&MakeCustomFloatFromBool(Bool=$User.StopSpawning, True Float=0, False Float=12)
    Use Spawn Probability=false
   )
  ]
  particleSpawnScript [
   &InitializeParticle(
    Color=&MakeCustomLinearColorFromBool(Bool=$User.Golden, True Color=<rgba 1 1 1 1>, False Color=<rgba 1.7245 1.9084 2 1>)
    Lifetime Max=1
    Lifetime Min=0.4
    Lifetime Mode=::NewEnumerator1
    Mesh Scale Max=<2 2 10>
    Mesh Scale Min=<1 1 6>
    Mesh Scale Mode=::NewEnumerator2
    UsePositionOffset=false
    Write Material Random=true
   )
   &ShapeLocation(
    Handle Radius:NiagaraFloat=&Lerp_Float(A=10, B=20, Alpha=$User.Size)
    Large Radius=&Lerp_Float(A=195, B=375, Alpha=$User.Size)
    MaskSpawnGroup=false
    Override Seed=false
    Rotation Axis=<0 1 0>
    Rotation Mode=::NewEnumerator1
    Shape Primitive=::NewEnumerator3
    Surface Distribution=1
    UseOwnerScale=true
   )
   &InitialMeshOrientation(
    Mesh Axis To Orient=<0 0 -1>
    Mesh Orientation Mode=::NewEnumerator2
    Orientation Coordinate Space=::World
    Orientation Vector=<0 0 1>
    Rotation=&RandomRangeVector(Minimum=<0 0 -1>, Maximum=<0 0 1>)
    Use Rotation Vector=true
   )
   &CameraOffset(Camera Offset Amount=50)
  ]
  particleUpdateScript [
   &ParticleState()
   &ScaleMeshSize(Scale Factor=&Multiply_Vector(
     A=&MakeVector(
      X=1
      Y=1
      Z=&FloatFromCurve(FloatCurve=DataInterface() {propertyValues {Curve {
          defaultValue=0
          keys [
           {arriveTangent=4.1785, interpMode=RCIM_Cubic, leaveTangent=4.1785, tangentMode=RCTM_User, time=0, value=0}
           {arriveTangent=0.0224, interpMode=RCIM_Cubic, leaveTangent=0.0224, tangentMode=RCTM_User, time=1, value=1.5}
          ]
         }}})
     )
     B=&VectorFromFloat(Value=&Lerp_Float(A=1, B=2, Alpha=$User.Size))
    ))
   &DynamicMaterialParameters(
    Index 0 Param 1=1
    Index 0 Param 2=&Lerp_Float(
     A=&RandomRangeFloat(Minimum=175, Maximum=350)
     B=&RandomRangeFloat(Minimum=350, Maximum=750)
     Alpha=$User.Size
    )
    Index 0 Param 4=-1
    Index 1 Param 2=1
    Index 1 Param 3=50
    Index 1 Param 4=&Multiply_Float(
     A=&RandomRangeFloat(Minimum=1.7, Maximum=2.2)
     B=&UniformAOrBFloat(A=-1, Evaluation Type=::NewEnumerator0, Recalculate Random Each Loop=true, Override Seed=false)
    )
    Index 2 Param 1=&FloatFromCurve(FloatCurve=DataInterface() {propertyValues {Curve {keys [<at 0 0.1 RCIM_Cubic>,<at 1 0.2 RCIM_Cubic>]}}})
    Index 2 Param 2=1
    Index 2 Param 3=50
    Index 2 Param 4=0.5
    Module.Index0Float0Write=true
    Module.Index0Float1Write=true
    Module.Index0Float2Write=true
    Module.Index0Float3Write=true
    Module.Index1Float0Write=true
    Module.Index1Float1Write=true
    Module.Index1Float2Write=true
    Module.Index1Float3Write=true
    Module.Index2Float0Write=true
    Module.Index2Float1Write=true
    Module.Index2Float2Write=true
    Module.Index2Float3Write=true
    Write Parameter Index 1=::NewEnumerator1
    Write Parameter Index 2=::NewEnumerator1
   )
   &ScaleColor(
    Scale Alpha=&Multiply_Float(
     A=&FloatFromCurve(
      FloatCurve=DataInterface() {propertyValues {Curve {
         defaultValue=1
         keys [
          {arriveTangent=9.0725, interpMode=RCIM_Cubic, leaveTangent=9.0725, tangentMode=RCTM_User, time=0, value=0}
          {arriveTangent=-2.5249, interpMode=RCIM_Cubic, leaveTangent=-2.5249, tangentMode=RCTM_User, time=0.1408, value=0.9847}
          {arriveTangent=-0.0405, interpMode=RCIM_Cubic, leaveTangent=-0.0405, tangentMode=RCTM_User, time=1, value=0}
         ]
        }}}
      Scale Curve=&MakeCustomFloatFromBool(Bool=$User.StopSpawning, True Float=0, False Float=&RandomRangeFloat(Minimum=0.05, Maximum=0.15))
     )
     B=&FloatFromCurve(
      FloatCurve=DataInterface() {propertyValues {Curve {
         defaultValue=1
         keys [
          {interpMode=RCIM_Cubic, tangentMode=RCTM_User, time=0, value=0}
          {arriveTangent=1.0649, interpMode=RCIM_Cubic, leaveTangent=1.0649, tangentMode=RCTM_User, time=1.5, value=1}
         ]
        }}}
      CurveIndex=$Engine.System.Age
     )
    )
    ScaleA=true
    ScaleRGB=true
   )
  ]
  rendererClasses [&NiagaraMeshRendererProperties]
  renderers [{rendererClass=&NiagaraMeshRendererProperties, rendererIndex=0}]
 }
}
summary {
 systemName=NS_SonicSpear_Fall
 emitters [
  {emitterName=Wind_Ribbons, rendererClasses [&NiagaraMeshRendererProperties]}
  {emitterName=Food_Debris, rendererClasses [&NiagaraMeshRendererProperties]}
  {emitterName=NE_SmallRadiusDroplets001, rendererClasses [&NiagaraSpriteRendererProperties]}
 ]
 userVariables [
  User.Golden {type=NiagaraBool}
  User.Size {type=NiagaraFloat, defaultValue=0, description=""}
  User.StopSpawning {type=NiagaraBool}
 ]
}
