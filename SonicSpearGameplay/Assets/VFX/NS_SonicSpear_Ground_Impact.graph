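# Niagara one-shot burst effect for the Sonic Spear ground impact.
# Layers a two-stage flare flash, speed lines, dark streaks, gravity-driven GPU debris and three mesh smoke emitters (circular, puffs, stretched ring).
# User.Size lerps radii, speeds and scales between small and large impacts, and System.SmokeColorTOD tints all smoke.
# FN_Debug is disabled; User.Frosting and User.Golden are exposed but no shown module reads them.
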NiagaraSystem /SonicSpearGameplay/Assets/VFX/NS_SonicSpear_Ground_Impact
dependencies {
 usedDataInterfaces [
  &NiagaraDataInterfaceMeshRendererInfo
  &NiagaraDataInterfaceCurve
  &NiagaraDataInterfaceVectorCurve
  &NiagaraDataInterfaceArrayFloat3
 ]
 usedDynamicInputs []
 usedModules [
  &self:SystemSpawnScript.SystemScriptSource.SystemScriptGraph.NiagaraNodeAssignment_0.SetVariables_#1
  &SystemState
 ]
}
emitters {
 DarkStreaks {
  emitterUpdateScript [
   &EmitterState()
   &SpawnBurst_Instantaneous(
    Spawn Count=&RandomRangeInt(Minimum=6, Maximum=11)
    Use Spawn Probability=false
    UseLoopCountLimit=false
   )
  ]
  particleSpawnScript [
   /SetVariables(Particles.RandomFloat=&UniformRangedFloat(Minimum=0, Maximum=3)) {isSetParameters=true}
   &InitializeParticle(
    Color Maximum=<rgba 0.2448 0.2046 0.1543 5>
    Color Minimum=<rgba 0.1823 0.1397 0.1149 5>
    Color Mode=::NewEnumerator2
    Lifetime Max=0.5
    Lifetime Min=0.3
    Lifetime Mode=::NewEnumerator1
    Sprite Size Max=<75 495>
    Sprite Size Min=<67.5 360>
    Sprite Size Mode=::NewEnumerator2
    UsePositionOffset=false
    Write Material Random=true
   )
   &ShapeLocation(
    MaskSpawnGroup=false
    Override Seed=false
    Ring Radius=&Lerp_Float(A=100, B=300, Alpha=$User.Size)
    Rotation Axis=<0 0 1>
    Rotation Mode=::NewEnumerator1
    Shape Primitive=::NewEnumerator4
    UseOwnerScale=true
   )
   &AddVelocity(
    Module.Use Falloff Distance=false
    Override Seed=false
    Velocity Mode=::NewEnumerator1
    Velocity Speed=&Lerp_Float(
     A=87.5
     B=125
     Alpha=&FloatFromCurve(
      FloatCurve=DataInterface() {propertyValues {Curve {keys [{interpMode=RCIM_Cubic, tangentMode=RCTM_Break, time=0, value=0}, {interpMode=RCIM_Cubic, tangentMode=RCTM_Break, time=1, value=1}]}}}
      CurveIndex=&RandomRangeFloat(Minimum=0, Maximum=1)
     )
    )
   )
   &AddVelocity(Velocity=&RandomRangeVector(Minimum=<0 0 100>, Maximum=<0 0 150>)) {module=AddVelocity001}
   &SubUVAnimation(End Frame=3, SubUV Animation Mode:ENiagara_SubUVLookupMode=::NewEnumerator2)
  ]
  particleUpdateScript [
   &ParticleState()
   &Drag(Drag=4, Use Linear Drag=true, Use Rotational Drag=false)
   &SolveForcesAndVelocity(Clamp Velocity=false, Limit Acceleration=false)
   &ScaleSpriteSize(Scale Factor=&Vector2DFromFloat(Value=&FloatFromCurve(
      FloatCurve=DataInterface() {propertyValues {Curve {keys [
          {arriveTangent=1.6317, interpMode=RCIM_Cubic, leaveTangent=1.6317, tangentMode=RCTM_User, time=0, value=0.1}
          {arriveTangent=0.2754, interpMode=RCIM_Cubic, leaveTangent=0.2754, tangentMode=RCTM_User, time=1.0463, value=1}
         ]}}}
      Scale Curve=&Lerp_Float(A=2, B=4, Alpha=$User.Size)
     )))
   &ScaleColor(
    Scale Alpha=&FloatFromCurve(
     FloatCurve=DataInterface() {propertyValues {Curve {keys [{interpMode=RCIM_Cubic, tangentMode=RCTM_Break, time=0, value=1}, {interpMode=RCIM_Cubic, tangentMode=RCTM_Break, time=1, value=0}]}}}
     Scale Curve=&RandomRangeFloat(Minimum=1, Maximum=1.25)
    )
    ScaleA=true
    ScaleRGB=false
   )
   &DynamicMaterialParameters(
    Index 0 Param 2=1
    Index 0 Param 3=1
    Index 0 Param 4=&FloatFromCurve(FloatCurve=DataInterface() {propertyValues {Curve {keys [<at 0.3 1>,<at 1 0.25>]}}})
    Module.Param0WriteEnabled=true
    Module.Param1WriteEnabled=true
    Module.Param2WriteEnabled=true
    Module.Param3WriteEnabled=true
   )
  ]
  rendererClasses [&NiagaraSpriteRendererProperties]
  renderers [{rendererClass=&NiagaraSpriteRendererProperties}]
 }
 Debris {
  simTarget=GPUComputeSim
  emitterUpdateScript [
   &EmitterState()
   &SpawnBurst_Instantaneous(Spawn Count=60, Spawn Time=0.03, Use Spawn Probability=false, UseLoopCountLimit=false)
  ]
  particleSpawnScript [
   &InitializeParticle(
    Color Maximum=<rgba 0.526 0.511 0.5069 1>
    Color Minimum=<rgba 0.8108 0.7348 0.7136 1>
    Color Mode=::NewEnumerator2
    Lifetime Max=1.25
    Lifetime Mode=::NewEnumerator1
    Sprite Rotation Mode=::NewEnumerator2
    Sprite Size Mode=::NewEnumerator1
    Uniform Sprite Size Max=30
    UsePositionOffset=false
    Write Material Random=true
   )
   &SubUVAnimation(End Frame=3, SubUV Animation Mode:ENiagara_SubUVLookupMode=::NewEnumerator2)
   &ShapeLocation(
    Hemisphere Distribution=<1 0.5>
    MaskSpawnGroup=false
    Non Uniform Scale=<1 1 0.8>
    Override Seed=false
    Rotation Axis=<0 0 1>
    Rotation Mode=::NewEnumerator1
    Sphere Radius=&Lerp_Float(A=250, B=675, Alpha=$User.Size)
    Sphere Surface Distribution=1
    UseOwnerScale=true
   )
   &AddVelocityFromPoint(
    Module.Use Falloff Distance=false
    Origin Offset=<0 0 -100>
    Velocity Strength=&RandomRangeFloat(Minimum=1200, Maximum=1500)
   )
   &AddVelocity(
    Cone Axis=<0 0 1>
    Module.UseConeFalloff=false
    Module.UseVelDistribution=false
    Override Seed=false
    Velocity Mode=::NewEnumerator2
    Velocity Speed=156.25
   )
  ]
  particleUpdateScript [
   &ParticleState()
   &Drag(Drag=&RandomRangeFloat(Minimum=1, Maximum=4), Use Linear Drag=true, Use Rotational Drag=true)
   &GravityForce(Gravity=<0 0 -1960>)
   &SpriteRotationRate(Rotation Rate=&Multiply_Float(A=&RandomRangeFloat(Minimum=-1, Maximum=1), B=720))
   &SolveForcesAndVelocity(Clamp Velocity=false, Limit Acceleration=false)
   &ScaleSpriteSize(
    Uniform Curve Scale=&Lerp_Float(A=1, B=3, Alpha=$User.Size)
    Uniform Curve Sprite Scale=DataInterface() {propertyValues {Curve {keys [
        {tangentMode=RCTM_User, time=0, value=1}
        {arriveTangent=-0.0405, interpMode=RCIM_Cubic, leaveTangent=-0.0405, tangentMode=RCTM_User, time=0.7009, value=1}
        {arriveTangent=-8.0089, interpMode=RCIM_Cubic, leaveTangent=-8.0089, tangentMode=RCTM_User, time=1, value=0}
       ]}}}
   )
  ]
  rendererClasses [&NiagaraSpriteRendererProperties]
  renderers [{rendererClass=&NiagaraSpriteRendererProperties}]
 }
 FN_Debug {
  bEnabled=false
  emitterUpdateScript [
   &EmitterState()
   &SpawnBurst_Instantaneous(Spawn Count=1, Use Spawn Probability=false)
  ]
  particleSpawnScript [/Niagara/Modules/Spawn/Initialization/InitializeParticle(
    Lifetime=5
    Mesh Scale=&VectorFromFloat(Value=100)
    Module.Write Color=true
    Module.Write Lifetime=true
    Module.Write Mass=true
    Module.Write Material Random=true
    Module.Write Position=false
    Module.Write Scale=true
    Module.Write SpriteRotation=false
    Module.Write SpriteSize=true
    Position=&SimulationPosition()
    Sprite Size=<100 100>
   )]
  particleUpdateScript [
   &ParticleState()
   /SetVariables(Particles.Debug=0.1) {isSetParameters=true}
   &DynamicMaterialParameters(
    Index 0 Param 1=$Particles.Debug
    Module.Param0WriteEnabled=true
    Module.Param1WriteEnabled=true
    Module.Param2WriteEnabled=true
    Module.Param3WriteEnabled=true
   )
  ]
  rendererClasses [&NiagaraSpriteRendererProperties, &NiagaraMeshRendererProperties]
  renderers [
   {rendererClass=&NiagaraSpriteRendererProperties}
   {rendererClass=&NiagaraMeshRendererProperties, rendererIndex=1}
  ]
 }
 Flash_Flare {
  emitterUpdateScript [
   &EmitterState()
   &SpawnBurst_Instantaneous(Use Spawn Probability=false, UseLoopCountLimit=false)
   &SpawnBurst_Instantaneous(Spawn Group=1, Spawn Time=0.05, Use Spawn Probability=false, UseLoopCountLimit=false) {module=SpawnBurst_Instantaneous001}
  ]
  particleSpawnScript [
   /Niagara/Modules/Spawn/Initialization/V2/InitializeParticle(
    Color=<rgba 1 0.8781 0.5653 1>
    Lifetime=&MergeHelper_FloatSwitch(0=0.166, 1=0.33, 2=?, 3=?, Index=&SpawnGroup(), Num=::NewEnumerator0)
    Mesh Scale Mode=::NewEnumerator3
    Mesh Uniform Scale=&Lerp_Float(A=14, B=24, Alpha=$User.Size)
    Sprite Size Mode=::NewEnumerator3
    Uniform Sprite Size=&Lerp_Float(A=500, B=700, Alpha=$User.Size)
    UsePositionOffset=false
    Write Material Random=true
   )
   &CameraOffset(Camera Offset Amount=&MergeHelper_FloatSwitch(0=100, 1=0, 2=?, 3=?, Index=&SpawnGroup(), Num=::NewEnumerator0))
   /SetVariables(Particles.VisibilityTag=&SpawnGroup()) {isSetParameters=true}
  ]
  particleUpdateScript [
   &ParticleState()
   &DynamicMaterialParameters(
    Index 0 Param 1=&MergeHelper_FloatSwitch(
     0=0.2
     1=&FloatFromCurve(FloatCurve=DataInterface() {propertyValues {Curve {keys [<at 0 1>,<at 0.3172 0.85>,<at 1 1>]}}})
     2=?
     3=?
     Index=&SpawnGroup()
     Num=::NewEnumerator0
    )
    Index 0 Param 2=&MergeHelper_FloatSwitch(
     0=1
     1=&FloatFromCurve(FloatCurve=DataInterface() {propertyValues {Curve {keys [<at 0 15>,<at 1 0>]}}})
     2=?
     3=?
     Index=&SpawnGroup()
     Num=::NewEnumerator0
    )
    Module.Param0WriteEnabled=true
    Module.Param1WriteEnabled=true
    Module.Param2WriteEnabled=true
    Module.Param3WriteEnabled=true
   )
   &ScaleSpriteSize(
    Scale Sprite Size Mode=::NewEnumerator0
    Uniform Scale Factor=&FloatFromCurve(
     FloatCurve=DataInterface() {propertyValues {Curve {keys [
         {interpMode=RCIM_Cubic, tangentMode=RCTM_User, time=0.333, value=1}
         {arriveTangent=-1.9817, interpMode=RCIM_Cubic, leaveTangent=-1.9817, tangentMode=RCTM_User, time=1, value=0.333}
        ]}}}
     Scale Curve=3
    )
   )
   &ScaleMeshSize(Scale Factor=&VectorFromFloat(Value=&FloatFromCurve(FloatCurve=DataInterface() {propertyValues {Curve {keys [
          {arriveTangent=6.399, interpMode=RCIM_Cubic, leaveTangent=6.399, tangentMode=RCTM_User, time=0, value=0}
          {arriveTangent=1, interpMode=RCIM_Cubic, leaveTangent=1, time=0.3, value=0.75}
          {interpMode=RCIM_Cubic, leaveTangent=-0.0405, time=1, value=1}
         ]}}})))
   &ScaleColor(
    Scale Alpha=&MergeHelper_FloatSwitch(
     0=&FloatFromCurve(
      FloatCurve=DataInterface() {propertyValues {Curve {keys [{interpMode=RCIM_Cubic, tangentMode=RCTM_User, time=0.5, value=1}, {interpMode=RCIM_Cubic, tangentMode=RCTM_User, time=1, value=0}]}}}
      Scale Curve=2
     )
     1=&FloatFromCurve(FloatCurve=DataInterface() {propertyValues {Curve {keys [
          {arriveTangent=-2.451, interpMode=RCIM_Cubic, leaveTangent=-2.451, tangentMode=RCTM_User, time=0, value=1}
          {interpMode=RCIM_Cubic, tangentMode=RCTM_User, time=1, value=0}
         ]}}})
     2=?
     3=?
     Index=&SpawnGroup()
     Num=::NewEnumerator0
    )
    Scale RGB=&MergeHelper_VectorSwitch(
     0=&VectorFromFloat(Value=1)
     1=&Lerp_Vector(
      A=<0.2 0.2 0.2>
      B=<1 1 1>
      Alpha=&FloatFromCurve(FloatCurve=DataInterface() {propertyValues {Curve {keys [{arriveTangent=-2.6099, interpMode=RCIM_Cubic, time=0, value=1}, {interpMode=RCIM_Cubic, tangentMode=RCTM_User, time=0.5, value=0}]}}})
     )
     2=?
     3=?
     Index=&SpawnGroup()
     Num=::NewEnumerator0
    )
    ScaleA=true
    ScaleRGB=true
   )
  ]
  rendererClasses [&NiagaraSpriteRendererProperties, &NiagaraMeshRendererProperties]
  renderers [
   {rendererClass=&NiagaraSpriteRendererProperties}
   {rendererClass=&NiagaraMeshRendererProperties, rendererIndex=1}
  ]
 }
 Smoke_Circular {
  emitterUpdateScript [
   &EmitterState()
   &SpawnBurst_Instantaneous(Spawn Count=4, Use Spawn Probability=false, UseLoopCountLimit=false)
  ]
  particleSpawnScript [
   &InitializeParticle(
    Lifetime Max=1.2
    Lifetime Min=0.9
    Lifetime Mode=::NewEnumerator1
    Mesh Scale Mode=::NewEnumerator1
    Mesh Uniform Scale Max=26
    Mesh Uniform Scale Min=25
    UsePositionOffset=false
    Write Material Random=true
   )
   &InitialMeshOrientation(
    Mesh Axis To Orient=&SelectVectorFromArray(
     Direct Array Index=&ReturnExecIndex()
     Vector Selection Array=DataInterface() {propertyValues {FloatData [<-1 -1 0>, <1 1 0>, <1 -1 0>, <-1 1 0>]}}
    )
    Mesh Orientation Mode=::NewEnumerator2
    Rotation=&SelectVectorFromArray(
     Direct Array Index=&ReturnExecIndex()
     Vector Selection Array=DataInterface() {propertyValues {FloatData []}}
    )
    Use Rotation Vector=true
   )
   &CameraOffset(Camera Offset Amount=100)
  ]
  particleUpdateScript [
   &ParticleState()
   &ScaleMeshSize(Scale Factor=&VectorFromFloat(Value=&FloatFromCurve(
      FloatCurve=DataInterface() {propertyValues {Curve {
         defaultValue=1
         keys [
          {arriveTangent=3.0252, interpMode=RCIM_Cubic, leaveTangent=3.0252, tangentMode=RCTM_User, time=0, value=0.2}
          {interpMode=RCIM_Cubic, leaveTangent=0.9093, time=1, value=1}
         ]
        }}}
      Scale Curve=&Lerp_Float(A=1, B=1.8, Alpha=$User.Size)
     )))
   &ScaleColor(
    Scale Alpha=&FloatFromCurve(FloatCurve=DataInterface() {propertyValues {Curve {keys [
         {arriveTangent=-0.2696, interpMode=RCIM_Cubic, leaveTangent=-0.2696, tangentMode=RCTM_User, time=0, value=0.575}
         {arriveTangent=-0.2792, interpMode=RCIM_Cubic, leaveTangent=-0.2792, tangentMode=RCTM_User, time=0.1528, value=0.5055}
         {leaveTangent=0.3906, time=0.7, value=0.4172}
        ]}}})
    Scale RGB=&MakeVectorFromLinearColorRGB(Color=$System.SmokeColorTOD)
    ScaleA=true
    ScaleRGB=true
   )
   &FN_Mesh_Attributes(
    Color=&MultiplyLinearColor(
     Linear Color=<rgba 1 1 1 1>
     Scale Factor=&MakeVector4FromVectorAndFloat(
      "Vector (RGB)"=&MakeVectorFromLinearColorRGB(Color=$System.SmokeColorTOD)
      "Float (Alpha)"=&FloatFromCurve(
       FloatCurve=DataInterface() {propertyValues {Curve {
          defaultValue=0
          keys [
           {arriveTangent=-2.1566, interpMode=RCIM_Cubic, leaveTangent=-2.1566, tangentMode=RCTM_User, time=0, value=1}
           {arriveTangent=-0.1914, interpMode=RCIM_Cubic, leaveTangent=-0.1914, tangentMode=RCTM_User, time=1, value=0}
          ]
         }}}
       Scale Curve=1.5
      )
     )
    )
    Create Color Attribute=true
    Create Scale Attribute=true
    Scale=&Multiply_Vector(A=$Particles.Scale, B=<1 1 1>)
    Write to Module Namespace=true
   )
  ]
  rendererClasses [&NiagaraMeshRendererProperties]
  renderers [
   {rendererClass=&NiagaraMeshRendererProperties, rendererIndex=0}
   {rendererClass=&NiagaraMeshRendererProperties, rendererIndex=1}
  ]
 }
 Smoke_Masked002 {
  emitterUpdateScript [
   &EmitterState()
   &SpawnBurst_Instantaneous(Spawn Count=14, Use Spawn Probability=false, UseLoopCountLimit=false)
   &SpawnBurst_Instantaneous(Spawn Count=5, Spawn Group=1, Spawn Time=0.001, Use Spawn Probability=true, UseLoopCountLimit=false) {module=SpawnBurst_Instantaneous001}
  ]
  particleSpawnScript [
   &InitializeParticle(
    Lifetime Max=1.4
    Lifetime Min=0.9
    Lifetime Mode=::NewEnumerator1
    Mesh Scale Mode=::NewEnumerator1
    Mesh Uniform Scale Max=8
    Mesh Uniform Scale Min=5
    UsePositionOffset=false
    Write Material Random=true
   )
   &ShapeLocation(
    Cylinder Distribution=::NewEnumerator2
    Cylinder Height=10
    Cylinder Height Midpoint=0
    Cylinder Radius=60
    MaskSpawnGroup=true
    Offset Coordinate Space=::World
    Offset Mode=::NewEnumerator0
    Rotation Axis=<0 0 1>
    Rotation Mode=::NewEnumerator1
    Shape Primitive=::NewEnumerator1
    U Position=&ReturnNormalizedExecIndex(Normalized Index Includes One=false)
    UseOwnerScale=true
   )
   &ShapeLocation(
    MaskSpawnGroup=true
    Offset Coordinate Space=::World
    Offset Mode=::NewEnumerator0
    Rotation Axis=<0 0 1>
    Rotation Mode=::NewEnumerator1
    Spawn Group Mask=1
    Sphere Distribution=::NewEnumerator1
    Sphere Radius=20
    U Position=&ReturnNormalizedExecIndex(Normalized Index Includes One=false)
    UseOwnerScale=true
    V Position=0.7
   ) {module=ShapeLocation001}
   &AddVelocity(
    Velocity=&Lerp_Vector(
     A=$Particles.ShapeLocation.ShapeVector
     B=&Add_Vector(
      A=$Particles.ShapeLocation001.ShapeVector
      B=&Add_Vector(
       A=&Multiply_VectorByFloat(Vector=$Particles.ShapeLocation.ShapeVector, Float=0.5)
       B=&Lerp_Vector(
        A=&RandomRangeVector(Minimum=<0 0 11.2>, Maximum=<0 0 24>)
        B=&RandomRangeVector(Minimum=<0 0 4>, Maximum=<0 0 8>)
        Alpha=$User.Size
       )
      )
     )
     Alpha=&MakeFloatFromInt(Int=$Emitter.SpawnGroup)
    )
    Velocity Speed Scale=&Lerp_Float(A=15, B=36, Alpha=$User.Size)
   )
   &InitialMeshOrientation(Use Rotation Vector=false)
   &CameraOffset(Camera Offset Amount=100)
  ]
  particleUpdateScript [
   &ParticleState()
   &Drag(Drag=&Lerp_Float(A=1.5, B=2.25, Alpha=$User.Size), Use Linear Drag=true, Use Rotational Drag=true)
   &SolveForcesAndVelocity(Clamp Velocity=false, Limit Acceleration=false)
   &ScaleMeshSize(Scale Factor=&Multiply_VectorByFloat(
     Vector=<1 1 1>
     Float=&FloatFromCurve(
      FloatCurve=DataInterface() {propertyValues {Curve {
         defaultValue=1
         keys [
          {arriveTangent=2.037, interpMode=RCIM_Cubic, leaveTangent=2.037, tangentMode=RCTM_User, time=0, value=0.25}
          {arriveTangent=0.6972, interpMode=RCIM_Cubic, leaveTangent=0.6972, tangentMode=RCTM_User, time=0.2466, value=0.7796}
          {tangentMode=RCTM_User, time=1, value=1}
         ]
        }}}
      Scale Curve=&Lerp_Float(A=1, B=3, Alpha=$User.Size)
     )
    ))
   &UpdateMeshOrientation(Rotation Rate=0.5, Rotation Vector=<0 1 0>)
   &ScaleColor(
    Scale Alpha=&FloatFromCurve(FloatCurve=DataInterface() {propertyValues {Curve {keys [
         {arriveTangent=-0.9949, interpMode=RCIM_Cubic, leaveTangent=-0.9949, tangentMode=RCTM_User, time=0, value=1}
         {arriveTangent=-0.1641, interpMode=RCIM_Cubic, leaveTangent=-0.1641, tangentMode=RCTM_User, time=0.8, value=0.4}
        ]}}})
    Scale RGB=&MakeVectorFromLinearColorRGB(Color=$System.SmokeColorTOD)
    ScaleA=true
    ScaleRGB=true
   )
   &ScaleColor(
    Scale Alpha=&FloatFromCurve(FloatCurve=DataInterface() {propertyValues {Curve {keys [
         {arriveTangent=-0.9949, interpMode=RCIM_Cubic, leaveTangent=-0.9949, tangentMode=RCTM_User, time=0, value=1}
         {arriveTangent=-0.1641, interpMode=RCIM_Cubic, leaveTangent=-0.1641, tangentMode=RCTM_User, time=0.8, value=0.4}
        ]}}})
    Scale RGB=&MakeVectorFromLinearColorRGB(Color=&FNTimeOfDayColor(
      Night Color=<rgba 0.1195 0.0954 0.0844 1>
      Morning Color=<rgba 0.2705 0.1303 0.0666 1>
      Day Color=<rgba 0.2188 0.1175 0.0707 1>
      Evening Color=<rgba 0.3467 0.1517 0.063 1>
     ))
    ScaleA=true
    ScaleRGB=true
   ) {enabled=false, module=ScaleColor001}
   &FN_Mesh_Attributes(
    Color=&MultiplyLinearColor(
     Linear Color=<rgba 1 1 1 0.5>
     Scale Factor=&MakeVector4FromVectorAndFloat(
      "Vector (RGB)"=&MakeVectorFromLinearColorRGB(Color=$System.SmokeColorTOD)
      "Float (Alpha)"=&FloatFromCurve(FloatCurve=DataInterface() {propertyValues {Curve {
          defaultValue=0
          keys [
           {arriveTangent=0.1155, interpMode=RCIM_Cubic, leaveTangent=0.1155, tangentMode=RCTM_User, time=0.4, value=1}
           {arriveTangent=-2.9743, interpMode=RCIM_Cubic, leaveTangent=-2.9743, tangentMode=RCTM_User, time=1, value=0}
          ]
         }}})
     )
    )
    Create Color Attribute=true
    Create Scale Attribute=true
    Scale=&Multiply_VectorByFloat(
     Vector=&VectorFromFloat(Value=&FloatFromCurve(
       FloatCurve=DataInterface() {propertyValues {Curve {
          defaultValue=0
          keys [
           {arriveTangent=1.6599, interpMode=RCIM_Cubic, leaveTangent=1.6599, tangentMode=RCTM_User, time=0, value=0.5}
           {interpMode=RCIM_Cubic, tangentMode=RCTM_User, time=1, value=1}
          ]
         }}}
       Scale Curve=&MakeFloatFromVector(Vector=$Particles.Initial.Scale, Channel=::NewEnumerator0)
      ))
     Float=1.3
    )
    Write to Module Namespace=true
   )
  ]
  rendererClasses [&NiagaraMeshRendererProperties]
  renderers [
   {rendererClass=&NiagaraMeshRendererProperties, rendererIndex=0}
   {rendererClass=&NiagaraMeshRendererProperties, rendererIndex=1}
  ]
 }
 Smoke_Stretch {
  emitterUpdateScript [
   &EmitterState()
   &SpawnBurst_Instantaneous(
    Spawn Count=&RandomRangeInt(Minimum=4, Maximum=7)
    Use Spawn Probability=false
    UseLoopCountLimit=false
   )
  ]
  particleSpawnScript [
   &InitializeParticle(
    Lifetime Max=1.1
    Lifetime Min=0.75
    Lifetime Mode=::NewEnumerator1
    Mesh Scale Mode=::NewEnumerator1
    Mesh Uniform Scale Max=25
    Mesh Uniform Scale Min=17
    UsePositionOffset=false
    Write Material Random=true
   )
   &ShapeLocation(
    Handle Radius=1
    Large Radius=&Lerp_Float(A=200, B=400, Alpha=$User.Size)
    MaskSpawnGroup=false
    Offset=<0 0 50>
    Offset Mode=::NewEnumerator0
    Rotation Angle:NiagaraFloat=&RandomRangeFloat(Minimum=-0.1, Maximum=0.1, Fixed Random Seed=true)
    Rotation Axis=<0 0 1>
    Rotation Mode=::NewEnumerator1
    Shape Primitive=::NewEnumerator3
    Torus Distribution Mode:ENiagaraTorusDistributionMode=::NewEnumerator1
    U Position=&ReturnNormalizedExecIndex(Normalized Index Includes One=false)
    UseOwnerScale=true
    V Position=0
   )
   &AddVelocity(
    Module.Use Falloff Distance=false
    Origin Offset=&Multiply_VectorByFloat(
     Vector=&RandomRangeVector(Minimum=<0 0 -100>, Maximum=<0 0 -180>)
     Float=&Lerp_Float(A=2, B=&RandomRangeFloat(Minimum=2.5, Maximum=4), Alpha=$User.Size)
    )
    Override Seed=false
    Velocity Mode=::NewEnumerator1
    Velocity Speed=50
   ) {enabled=false}
   &AddVelocity(
    Velocity=$Particles.ShapeLocation.ShapeNormal
    Velocity Speed Scale=&RandomRangeFloat(Minimum=1.5, Maximum=2)
   ) {module=AddVelocity001}
   &AddVelocity(Velocity=<0 0 1>) {module=AddVelocity002}
   &InitialMeshOrientation(
    Mesh Axis To Orient=&RandomRangeVector(Minimum=<0 0 -1>, Maximum=<0 0 1>)
    Mesh Orientation Mode=::NewEnumerator2
    Use Rotation Vector=false
   )
  ]
  particleUpdateScript [
   &ParticleState()
   &ScaleMeshSize(Scale Factor=&VectorFromCurve(
     VectorCurve=DataInterface() {propertyValues {
       XCurve {keys [
         {arriveTangent=1.5574, interpMode=RCIM_Cubic, leaveTangent=1.5574, tangentMode=RCTM_User, time=0, value=-0.0937}
         {arriveTangent=0.6997, interpMode=RCIM_Cubic, leaveTangent=0.6997, tangentMode=RCTM_User, time=1, value=1}
        ]}
       YCurve {keys [
         {arriveTangent=1.5574, interpMode=RCIM_Cubic, leaveTangent=1.5574, tangentMode=RCTM_User, time=0, value=-0.0937}
         {arriveTangent=0.6997, interpMode=RCIM_Cubic, leaveTangent=0.6997, tangentMode=RCTM_User, time=0.7, value=1}
        ]}
       ZCurve {keys [
         {arriveTangent=0.7039, interpMode=RCIM_Cubic, leaveTangent=0.7039, tangentMode=RCTM_User, time=-0.0043, value=0.5795}
         {arriveTangent=0.1691, interpMode=RCIM_Cubic, leaveTangent=0.1691, tangentMode=RCTM_User, time=1, value=1}
        ]}
      }}
     Scale Curve=&Multiply_VectorByFloat(Vector=<0.7 0.7 1>, Float=&Lerp_Float(A=1, B=1.8, Alpha=$User.Size))
    ))
   &Drag(Use Linear Drag=true, Use Rotational Drag=true)
   &SolveForcesAndVelocity(Clamp Velocity=false, Limit Acceleration=false)
   &ScaleColor(
    Scale Alpha=&FloatFromCurve(FloatCurve=DataInterface() {propertyValues {Curve {keys [
         {arriveTangent=-0.5026, interpMode=RCIM_Cubic, leaveTangent=-0.5026, tangentMode=RCTM_User, time=0, value=1}
         {arriveTangent=-0.6, interpMode=RCIM_Cubic, leaveTangent=-0.6, time=0.4, value=0.5418}
         {arriveTangent=-0.1231, interpMode=RCIM_Cubic, leaveTangent=-0.1231, tangentMode=RCTM_User, time=1, value=0.4}
        ]}}})
    Scale RGB=&MakeVectorFromLinearColorRGB(Color=$System.SmokeColorTOD)
    ScaleA=true
    ScaleRGB=true
   )
   &FN_Mesh_Attributes(
    Color=&MultiplyLinearColor(
     Linear Color=<rgba 1 1 1 1>
     Scale Factor=&MakeVector4FromVectorAndFloat(
      "Vector (RGB)"=&MakeVectorFromLinearColorRGB(Color=$System.SmokeColorTOD)
      "Float (Alpha)"=&FloatFromCurve(
       FloatCurve=DataInterface() {propertyValues {Curve {
          defaultValue=0
          keys [
           {arriveTangent=9.0725, interpMode=RCIM_Cubic, leaveTangent=9.0725, tangentMode=RCTM_User, time=0.002, value=0.5517}
           {arriveTangent=-1.5921, interpMode=RCIM_Cubic, leaveTangent=-1.5921, tangentMode=RCTM_User, time=0.1408, value=0.9847}
           {arriveTangent=-0.0405, interpMode=RCIM_Cubic, leaveTangent=-0.0405, tangentMode=RCTM_User, time=1, value=0}
          ]
         }}}
       Scale Curve=1.5
      )
     )
    )
    Create Color Attribute=true
    Create Scale Attribute=true
    Scale=&Multiply_VectorByFloat(Vector=$Particles.Scale, Float=1.25)
    Write to Module Namespace=true
   )
  ]
  rendererClasses [&NiagaraMeshRendererProperties]
  renderers [
   {rendererClass=&NiagaraMeshRendererProperties, rendererIndex=0}
   {rendererClass=&NiagaraMeshRendererProperties, rendererIndex=1}
  ]
 }
 SpeedLine001 {
  emitterUpdateScript [
   &EmitterState()
   &SpawnBurst_Instantaneous(Spawn Count=3, Use Spawn Probability=false, UseLoopCountLimit=false)
  ]
  particleSpawnScript [&InitializeParticle(
    Color=<rgba 1 1 1 10>
    Lifetime=0.25
    Sprite Rotation Mode=::NewEnumerator2
    Sprite Size Mode=::NewEnumerator1
    Uniform Sprite Size Max=1700
    Uniform Sprite Size Min=1020
    UsePositionOffset=false
    Write Material Random=true
   )]
  particleUpdateScript [
   &ParticleState()
   &ScaleSpriteSize(
    Scale Sprite Size Mode=::NewEnumerator0
    Uniform Scale Factor=&FloatFromCurve(
     FloatCurve=DataInterface() {propertyValues {Curve {keys [{arriveTangent=3.9086, interpMode=RCIM_Cubic, leaveTangent=3.9086, tangentMode=RCTM_User, time=0, value=0}, <at 1 1.4 RCIM_Cubic>]}}}
     Scale Curve=&Lerp_Float(A=0.5, B=1, Alpha=$User.Size)
    )
   )
   &DynamicMaterialParameters(
    Index 0 Param 1=&FloatFromCurve(FloatCurve=DataInterface() {propertyValues {Curve {keys [<at 0 1 RCIM_Cubic>, {arriveTangent=-2.742, interpMode=RCIM_Cubic, leaveTangent=-2.742, tangentMode=RCTM_User, time=1, value=0}]}}})
    Module.Index0Float0Write=true
    Module.Index0Float1Write=true
    Module.Index0Float2Write=true
    Module.Index0Float3Write=true
   )
   &ScaleColor(
    Scale Alpha=&FloatFromCurve(
     FloatCurve=DataInterface() {propertyValues {Curve {keys [
         {arriveTangent=-13.9062, interpMode=RCIM_Cubic, leaveTangent=-13.9062, tangentMode=RCTM_User, time=0, value=7}
         {arriveTangent=-7, interpMode=RCIM_Cubic, leaveTangent=-7, time=0.2488, value=3}
         <at 1 0 RCIM_Cubic>
        ]}}}
     Scale Curve=2
    )
    ScaleA=true
    ScaleRGB=true
   )
   &CameraOffset(Camera Offset Amount=50)
  ]
  rendererClasses [&NiagaraSpriteRendererProperties]
  renderers [{rendererClass=&NiagaraSpriteRendererProperties}]
 }
}
summary {
 systemName=NS_SonicSpear_Ground_Impact
 emitters [
  {emitterName=Flash_Flare, rendererClasses [&NiagaraSpriteRendererProperties, &NiagaraMeshRendererProperties]}
  {emitterName=Debris, simTarget=GPUComputeSim, rendererClasses [&NiagaraSpriteRendererProperties]}
  {emitterName=DarkStreaks, rendererClasses [&NiagaraSpriteRendererProperties]}
  {emitterName=SpeedLine001, rendererClasses [&NiagaraSpriteRendererProperties]}
  {emitterName=Smoke_Stretch, rendererClasses [&NiagaraMeshRendererProperties]}
  {emitterName=Smoke_Circular, rendererClasses [&NiagaraMeshRendererProperties]}
  {emitterName=Smoke_Masked002, rendererClasses [&NiagaraMeshRendererProperties]}
  {
   bEnabled=false
   emitterName=FN_Debug
   rendererClasses [&NiagaraSpriteRendererProperties, &NiagaraMeshRendererProperties]
  }
 ]
 userVariables [
  User.Frosting {type=LinearColor, description="", defaultValue {struct=LinearColor, value=<rgba 0.9062 0.0746 0.6196 1>}}
  User.Golden {type=NiagaraBool}
  User.Size {type=NiagaraFloat, defaultValue=0, description=""}
 ]
}
