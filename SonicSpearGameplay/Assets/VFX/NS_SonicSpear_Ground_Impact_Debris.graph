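# Niagara system for Sonic Spear ground impacts, throwing a one-shot burst of dessert-themed debris.
# Two crumb emitters and multicolored sprinkles launch tumbling meshes under gravity and drag; a frosting-splat sprite emitter, tinted by User.Frosting, is disabled.
# User.Size scales velocities and sizes, while User.Golden zeroes every spawn count.
# Crumbs001 and Sprinkles add time-of-day tinting and platform-quality-gated collision, and all mesh emitters use kill volumes.
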NiagaraSystem /SonicSpearGameplay/Assets/VFX/NS_SonicSpear_Ground_Impact_Debris
dependencies {
 usedDataInterfaces [
  &NiagaraDataInterfaceArrayColor
  &NiagaraDataInterfacePlatformSet
  &NiagaraDataInterfaceCurve
  &NiagaraDataInterfaceArrayFloat
 ]
 usedDynamicInputs [&UniformRangedFloat]
 usedModules [&SystemState]
}
emitters {
 Crumbs {
  emitterUpdateScript [
   &EmitterState()
   &SpawnBurst_Instantaneous(
    Spawn Count=&MakeCustomIntFromBool(Bool=$User.Golden, True INT=0, False INT=&RandomRangeInt(Minimum=18, Maximum=32))
    Use Spawn Probability=false
    UseLoopCountLimit=false
   )
  ]
  particleSpawnScript [
   &InitializeParticle(
    Lifetime Max=1.6
    Lifetime Min=0.8
    Lifetime Mode=::NewEnumerator1
    Mesh Scale Mode=::NewEnumerator1
    Mesh Uniform Scale Max=8
    Mesh Uniform Scale Min=2
    Sprite Size Max=<3 14>
    Sprite Size Min=<2 6>
    Sprite Size Mode=::NewEnumerator2
    UsePositionOffset=false
    Write Material Random=true
   )
   &ShapeLocation(MaskSpawnGroup=false, Override Seed=false, Sphere Radius=50, UseOwnerScale=true)
   &AddVelocity(
    Module.Use Falloff Distance=false
    Origin Offset=<0 0 -10>
    Override Seed=false
    Velocity Mode=::NewEnumerator1
    Velocity Speed=&Lerp_Float(
     A=&RandomRangeFloat(Minimum=1000, Maximum=1800)
     B=&RandomRangeFloat(Minimum=2000, Maximum=3600)
     Alpha=$User.Size
    )
   ) {module=AddVelocity001}
   &AddVelocity(
    Cone Angle=180
    Cone Axis=<0 0 1>
    Module.UseConeFalloff=false
    Module.UseVelDistribution=false
    Override Seed=false
    Velocity Mode=::NewEnumerator2
    Velocity Speed=&Lerp_Float(A=800, B=1800, Alpha=$User.Size)
   ) {module=AddVelocity002}
   &InitialMeshOrientation(Use Rotation Vector=false)
   &AddRotationalVelocity(Rotation Rate=&RandomVector(Vector Scale=&RandomRangeFloat(Minimum=1, Maximum=1.5)))
  ]
  particleUpdateScript [
   &ParticleState()
   &Drag(Drag=&Lerp_Float(A=2, B=4, Alpha=$User.Size), Use Linear Drag=true, Use Rotational Drag=false)
   &GravityForce(Gravity=<0 0 -2500>)
   &SolveForcesAndVelocity(Clamp Velocity=false, Limit Acceleration=false)
   &ScaleMeshSize(Scale Factor=&VectorFromFloat(Value=&FloatFromCurve(
      FloatCurve=DataInterface() {propertyValues {Curve {
         defaultValue=1
         keys [
          <at 0 0 RCIM_Cubic>
          {arriveTangentWeight=0.0916, interpMode=RCIM_Cubic, leaveTangentWeight=0.281, tangentMode=RCTM_User, time=0.1229, value=1}
          {arriveTangentWeight=0.1257, interpMode=RCIM_Cubic, leaveTangentWeight=0.1667, tangentMode=RCTM_User, time=0.5, value=1}
          {interpMode=RCIM_Cubic, tangentMode=RCTM_User, time=1, value=0}
         ]
        }}}
      Scale Curve=&Lerp_Float(A=1.35, B=2.5, Alpha=$User.Size)
     )))
   &KillParticlesInVolume(Kill Shape=::NewEnumerator2, Origin Offset:Vector3f=<0 0 -200>)
  ]
  rendererClasses [&NiagaraMeshRendererProperties]
  renderers [
   {rendererClass=&NiagaraMeshRendererProperties, rendererIndex=0}
   {rendererClass=&NiagaraMeshRendererProperties, rendererIndex=1}
  ]
 }
 Crumbs001 {
  emitterUpdateScript [
   &EmitterState()
   &SpawnBurst_Instantaneous(
    Spawn Count=&MakeCustomIntFromBool(Bool=$User.Golden, True INT=0, False INT=&RandomRangeInt(Minimum=18, Maximum=32))
    Use Spawn Probability=false
    UseLoopCountLimit=false
   )
  ]
  particleSpawnScript [
   &InitializeParticle(
    Color=&FNTimeOfDayColor(
     Night Color=<rgba 0.151 0.1052 0.081 1>
     Morning Color=<rgba 0.6105 0.3004 0.0859 1>
     Day Color=<rgba 0.6445 0.314 0.0931 1>
     Evening Color=<rgba 0.6823 0.3351 0.1031 1>
    )
    Lifetime=&MakeCustomFloatFromBool(
     Bool=&IsPlatformSetActive(New Platform Set=DataInterface() {propertyValues {Platforms {qualityLevelMask=-8}}})
     True Float=&RandomRangeFloat(Minimum=1.2, Maximum=2.4)
     False Float=&RandomRangeFloat(Minimum=0.7, Maximum=1.4)
    )
    Mesh Scale Mode=::NewEnumerator1
    Mesh Uniform Scale Max=0.2
    Mesh Uniform Scale Min=0.05
    Sprite Size Max=<3 14>
    Sprite Size Min=<2 6>
    Sprite Size Mode=::NewEnumerator2
    UsePositionOffset=false
    Write Material Random=true
   )
   &ShapeLocation(MaskSpawnGroup=false, Override Seed=false, Sphere Radius=50, UseOwnerScale=true)
   &AddVelocity(
    Module.Use Falloff Distance=false
    Origin Offset=<0 0 -10>
    Override Seed=false
    Velocity Mode=::NewEnumerator1
    Velocity Speed=&Lerp_Float(
     A=&RandomRangeFloat(Minimum=1000, Maximum=1800)
     B=&RandomRangeFloat(Minimum=2000, Maximum=3600)
     Alpha=$User.Size
    )
   ) {module=AddVelocity001}
   &AddVelocity(
    Cone Angle=180
    Cone Axis=<0 0 1>
    Module.UseConeFalloff=false
    Module.UseVelDistribution=false
    Override Seed=false
    Velocity Mode=::NewEnumerator2
    Velocity Speed=&Lerp_Float(A=800, B=1800, Alpha=$User.Size)
   ) {module=AddVelocity002}
   &InitialMeshOrientation(Use Rotation Vector=false)
   &AddRotationalVelocity(Rotation Rate=&RandomVector(Vector Scale=&RandomRangeFloat(Minimum=1, Maximum=1.5)))
  ]
  particleUpdateScript [
   &ParticleState()
   &Drag(Drag=&Lerp_Float(A=2, B=4, Alpha=$User.Size), Use Linear Drag=true, Use Rotational Drag=false)
   &GravityForce(Gravity=<0 0 -2500>)
   &Collision(
    Age Colliding Particles=false
    Collision Enabled=&Boolean_AndOperation(
     Bool1=&IsPlatformSetActive(New Platform Set=DataInterface() {propertyValues {Platforms {qualityLevelMask=-8}}})
     Bool2=&SetBoolByFloatComparison(A=$Particles.NormalizedAge, B=0.2, Comparison Type=::NewEnumerator2)
    )
    KillOnCollision=false
    Module.Control Rest State Via Surface Normal=false
    Module.EnableMaxCollisionCount=false
    Module.Randomize Collision Normal Vector=false
   )
   &SolveForcesAndVelocity(Clamp Velocity=false, Limit Acceleration=false)
   &ScaleMeshSize(Scale Factor=&VectorFromFloat(Value=&FloatFromCurve(
      FloatCurve=DataInterface() {propertyValues {Curve {
         defaultValue=1
         keys [
          <at 0 0 RCIM_Cubic>
          {arriveTangentWeight=0.0916, interpMode=RCIM_Cubic, leaveTangentWeight=0.281, tangentMode=RCTM_User, time=0.1229, value=1}
          {arriveTangentWeight=0.1257, interpMode=RCIM_Cubic, leaveTangentWeight=0.1667, tangentMode=RCTM_User, time=0.5, value=1}
          {interpMode=RCIM_Cubic, tangentMode=RCTM_User, time=1, value=0}
         ]
        }}}
      Scale Curve=&Lerp_Float(A=1.5, B=2.5, Alpha=$User.Size)
     )))
   &KillParticlesInVolume(Kill Shape=::NewEnumerator2, Origin Offset:Vector3f=<0 0 -200>)
  ]
  rendererClasses [&NiagaraMeshRendererProperties]
  renderers [{rendererClass=&NiagaraMeshRendererProperties, rendererIndex=0}]
 }
 Frosting_Splats {
  bEnabled=false
  emitterUpdateScript [
   &EmitterState()
   &SpawnBurst_Instantaneous(
    Spawn Count=&MakeCustomIntFromBool(Bool=$User.Golden, True INT=0, False INT=&RandomRangeInt(Minimum=16, Maximum=28))
    Use Spawn Probability=true
    UseLoopCountLimit=false
   )
   &SpawnBurst_Instantaneous(
    Spawn Count=&MakeCustomIntFromBool(Bool=$User.Golden, True INT=0, False INT=&RandomRangeInt(Minimum=9, Maximum=18))
    Spawn Group=1
    Use Spawn Probability=true
    UseLoopCountLimit=false
   ) {module=SpawnBurst_Instantaneous001}
  ]
  particleSpawnScript [
   &InitializeParticle(
    Color=$User.Frosting
    Lifetime Max=1.3
    Lifetime Min=0.75
    Lifetime Mode=::NewEnumerator1
    Sprite Rotation Mode=::NewEnumerator2
    Sprite Size=/Niagara/DynamicInputs/Helpers/MaskVector2d(
     Mask Source:NiagaraInt32=0
     Mask Parameter:NiagaraInt32=?
     Mask:NiagaraInt32=0
     Masked Vector2D:Vector2f=&RandomRangeVector2D(Minimum=<15 30>, Maximum=<30 60>)
     Passthrough Vector2D:Vector2f=&Vector2DFromFloat(Value=&Multiply_Float(
       A=&RandomRangeFloat(Minimum=75, Maximum=210)
       B=&Lerp_Float(A=1.2, B=1.4, Alpha=$User.Size)
      ))
    )
    Sprite Size Mode=::NewEnumerator4
    Sprite UV Mode=::NewEnumerator3
    UsePositionOffset=false
    Write Material Random=true
   )
   /Niagara/Modules/Update/SubUV/SubUVAnimation(Number Of Frames=8, SubUV Animation Mode=::NewEnumerator1)
   &ShapeLocation(
    Handle Radius:NiagaraFloat=&Lerp_Float(A=90, B=190, Alpha=$User.Size)
    Large Radius=&Lerp_Float(A=195, B=375, Alpha=$User.Size)
    MaskSpawnGroup=false
    Non Uniform Scale=<1 1 -0.8>
    Offset Mode=::NewEnumerator0
    Override Seed=false
    Rotation Axis=<0 1 0>
    Rotation Mode=::NewEnumerator1
    Shape Primitive=::NewEnumerator3
    Surface Distribution=1
    UseOwnerScale=true
    V Distribution=0.5
   )
   &AddVelocity(
    Module.Use Falloff Distance=false
    Override Seed=false
    Velocity Mode=::NewEnumerator1
    Velocity Speed=&RandomRangeFloat(Minimum=200, Maximum=720)
   )
   &AddVelocity(
    Velocity=&RandomRangeVector(Minimum=<0 125 300>, Maximum=<0 300 750>)
    Velocity Speed Scale=&Lerp_Float(A=1, B=1.2, Alpha=$User.Size)
   ) {module=AddVelocity001}
   &AddVelocity(Rotation Coordinate Space=::World, Velocity=&RandomRangeVector(Minimum=<0 0 50>, Maximum=<0 0 150>)) {module=AddVelocity002}
   /SetVariables(Particles.VisibilityTag=&SpawnGroup()) {isSetParameters=true}
   &CameraOffset(Camera Offset Amount=50)
  ]
  particleUpdateScript [
   &ParticleState()
   &Drag(
    Drag=&SelectFloatFromArray(
     Direct Array Index=&SpawnGroup()
     Float Selection Array=DataInterface() {propertyValues {FloatData [1.25,1,1.5]}}
    )
    Use Linear Drag=true
    Use Rotational Drag=true
   )
   &GravityForce(Gravity=<0 0 -1470>)
   &SolveForcesAndVelocity(Clamp Velocity=false, Limit Acceleration=false)
   &SpriteRotationRate(Rotation Rate=&RandomRangeFloat(Minimum=-60, Maximum=60))
   &ScaleSpriteSize(
    Scale Factor=&Multiply_Vector2D_ByFloat(
     Float=&FloatFromCurve(
      FloatCurve=DataInterface() {propertyValues {Curve {keys [
          {arriveTangent=4.1785, interpMode=RCIM_Cubic, leaveTangent=4.1785, tangentMode=RCTM_User, time=0, value=0.2}
          {arriveTangent=0.8, interpMode=RCIM_Cubic, leaveTangent=0.8, time=0.2042, value=0.7}
          {arriveTangent=0.0224, interpMode=RCIM_Cubic, leaveTangent=0.0224, tangentMode=RCTM_User, time=1, value=1}
         ]}}}
      Scale Curve=1.5
     )
     Vector2D=&Lerp_Vector2(
      A=<1 0.5>
      B=<1 1>
      Alpha=&FloatFromCurve(
       FloatCurve=DataInterface() {propertyValues {Curve {keys [
           {arriveTangent=4.1785, interpMode=RCIM_Cubic, leaveTangent=4.1785, tangentMode=RCTM_User, time=0, value=0}
           {arriveTangent=1.1179, interpMode=RCIM_Cubic, leaveTangent=1.1179, tangentMode=RCTM_User, time=0.2042, value=0.6228}
           {arriveTangent=0.0224, interpMode=RCIM_Cubic, leaveTangent=0.0224, tangentMode=RCTM_User, time=1, value=1}
          ]}}}
       Scale Curve=&Lerp_Float(A=0.8, B=2, Alpha=$User.Size)
      )
     )
    )
    Scale Sprite Size Mode=::NewEnumerator2
   )
   &DynamicMaterialParameters(
    Index 0 Param 1=&RandomRangeFloat(Minimum=0, Maximum=1)
    Index 0 Param 2=&FloatFromCurve(
     FloatCurve=DataInterface() {propertyValues {Curve {keys [{interpMode=RCIM_Cubic, tangentMode=RCTM_User, time=0.33, value=0}, {interpMode=RCIM_Cubic, leaveTangent=1.2673, time=0.8, value=1}]}}}
     Scale Curve=0
    )
    Index 0 Param 3=&FloatFromCurve(FloatCurve=DataInterface() {propertyValues {Curve {keys [{tangentMode=RCTM_User, time=0, value=1}, {tangentMode=RCTM_User, time=1, value=1}]}}})
    Index 0 Param 4=&FloatFromCurve(FloatCurve=DataInterface() {propertyValues {Curve {keys [
         {arriveTangent=-14.9399, interpMode=RCIM_Cubic, leaveTangent=-14.9399, tangentMode=RCTM_User, time=0, value=1}
         {interpMode=RCIM_Cubic, tangentMode=RCTM_User, time=0.15, value=0}
         {interpMode=RCIM_Cubic, tangentMode=RCTM_User, time=0.33, value=0}
         {arriveTangent=2.7964, interpMode=RCIM_Cubic, leaveTangent=2.7964, tangentMode=RCTM_User, time=1, value=1}
        ]}}})
    Index 1 Param 2=&FloatFromCurve(FloatCurve=DataInterface() {propertyValues {Curve {keys [
         {arriveTangent=23.2225, interpMode=RCIM_Cubic, leaveTangent=23.2225, tangentMode=RCTM_User, time=0, value=0}
         <at 0.1 1 RCIM_Cubic>
         {arriveTangent=-3.2993, interpMode=RCIM_Cubic, leaveTangent=-3.2993, tangentMode=RCTM_User, time=1, value=0}
        ]}}})
    Module.Index0Float0Write=true
    Module.Index0Float1Write=true
    Module.Index0Float2Write=true
    Module.Index0Float3Write=true
    Module.Index1Float0Write=true
    Module.Index1Float1Write=true
    Module.Index1Float2Write=true
    Module.Index1Float3Write=true
    Write Parameter Index 1=::NewEnumerator1
   )
   &ScaleColor(
    Color Scale:LinearColor=&FNTimeOfDayColor(Night Color=<rgba 0.4167 0.4167 0.4167 1>, Morning Color=<rgba 1 1 1 1>, Day Color=<rgba 1 1 1 1>, Evening Color=<rgba 1 1 1 1>)
    Scale Mode=3
   )
  ]
  rendererClasses [&NiagaraSpriteRendererProperties]
  renderers [
   {rendererClass=&NiagaraSpriteRendererProperties}
   {rendererClass=&NiagaraSpriteRendererProperties, rendererIndex=1}
  ]
 }
 Sprinkles {
  emitterUpdateScript [
   &EmitterState()
   &SpawnBurst_Instantaneous(
    Spawn Count=&MakeCustomIntFromBool(Bool=$User.Golden, True INT=0, False INT=&RandomRangeInt(Minimum=18, Maximum=28))
    Spawn Group=1
    Use Spawn Probability=false
    UseLoopCountLimit=false
   )
  ]
  particleSpawnScript [
   &InitializeParticle(
    Color=&SelectLinearColorFromArray(
     Array Sampling Mode=::NewEnumerator0
     Direct Array Index=?
     Color Selection Array=DataInterface() {propertyValues {ColorData [
        <rgba 1 0.0437 0.0685 1>
        <rgba 0 0.4621 0.9387 1>
        <rgba 0.007 0.5972 0.2016 1>
        <rgba 1 1 0.8308 1>
        <rgba 0.6445 0.314 0.0931 1>
        <rgba 1 1 0 1>
       ]}}
     Random Seed=0
    )
    Lifetime Max=1.5
    Lifetime Min=0.7
    Lifetime Mode=::NewEnumerator1
    Mesh Scale=&Lerp_Vector(A=<18 10 10>, B=<36 20 20>, Alpha=$User.Size)
    Mesh Scale Mode=::NewEnumerator4
    Sprite Size Max=<20 80>
    Sprite Size Min=<20 60>
    Sprite Size Mode=::NewEnumerator2
    UsePositionOffset=false
    Write Material Random=true
   )
   &ShapeLocation(
    Handle Radius:NiagaraFloat=&Lerp_Float(A=90, B=190, Alpha=$User.Size)
    Large Radius=&Lerp_Float(A=195, B=375, Alpha=$User.Size)
    MaskSpawnGroup=false
    Non Uniform Scale=<1 1 -0.8>
    Offset Mode=::NewEnumerator0
    Override Seed=false
    Rotation Axis=<0 1 0>
    Rotation Mode=::NewEnumerator1
    Shape Primitive=::NewEnumerator3
    Surface Distribution=1
    UseOwnerScale=true
    V Distribution=0.5
   )
   &AddVelocity(
    Module.Use Falloff Distance=false
    Origin Offset=<0 0 -60>
    Override Seed=false
    Velocity Mode=::NewEnumerator1
    Velocity Speed=&RandomRangeFloat(Minimum=200, Maximum=400)
   )
   &AddVelocity(
    Velocity=&RandomRangeVector(Minimum=<0 250 800>, Maximum=<0 600 1000>)
    Velocity Speed Scale=&Lerp_Float(A=1, B=1.3, Alpha=$User.Size)
   ) {module=AddVelocity001}
   &AddRotationalVelocity(Rotation Rate=&RandomVector(Vector Scale=1))
  ]
  particleUpdateScript [
   &ParticleState()
   &Drag(Use Linear Drag=true, Use Rotational Drag=true)
   &GravityForce(Gravity=<0 0 -1470>)
   &Collision(
    Age Colliding Particles=false
    Collision Enabled=&Boolean_AndOperation(
     Bool1=&IsPlatformSetActive(New Platform Set=DataInterface() {propertyValues {Platforms {qualityLevelMask=-8}}})
     Bool2=&SetBoolByFloatComparison(A=$Particles.NormalizedAge, B=0.2, Comparison Type=::NewEnumerator2)
    )
    KillOnCollision=false
    Module.Control Rest State Via Surface Normal=false
    Module.EnableMaxCollisionCount=false
    Module.Randomize Collision Normal Vector=false
   )
   &SolveForcesAndVelocity(Clamp Velocity=false, Limit Acceleration=false)
   &ScaleMeshSize(Scale Factor=&VectorFromFloat(Value=&FloatFromCurve(
      FloatCurve=DataInterface() {propertyValues {Curve {
         defaultValue=1
         keys [
          {arriveTangent=9.0725, interpMode=RCIM_Cubic, leaveTangent=9.0725, tangentMode=RCTM_User, time=0, value=0}
          <at 0.1408 1 RCIM_Cubic>
          {arriveTangent=-4.0309, interpMode=RCIM_Cubic, leaveTangent=-4.0309, tangentMode=RCTM_User, time=1, value=0}
         ]
        }}}
      Scale Curve=1.4
     )))
   &ScaleColor(
    Color Scale:LinearColor=&FNTimeOfDayColor(Night Color=<rgba 0.4167 0.4167 0.4167 1>, Morning Color=<rgba 1 1 1 1>, Day Color=<rgba 1 1 1 1>, Evening Color=<rgba 1 1 1 1>)
    Scale Mode=3
   )
   &KillParticlesInVolume(Kill Shape=::NewEnumerator2, Origin Offset:Vector3f=<0 0 -200>)
  ]
  rendererClasses [&NiagaraMeshRendererProperties]
  renderers [
   {rendererClass=&NiagaraMeshRendererProperties, rendererIndex=0}
   {rendererClass=&NiagaraMeshRendererProperties, rendererIndex=1}
  ]
 }
}
summary {
 systemName=NS_SonicSpear_Ground_Impact_Debris
 emitters [
  {emitterName=Sprinkles, rendererClasses [&NiagaraMeshRendererProperties]}
  {bEnabled=false, emitterName=Frosting_Splats, rendererClasses [&NiagaraSpriteRendererProperties]}
  {emitterName=Crumbs, rendererClasses [&NiagaraMeshRendererProperties]}
  {emitterName=Crumbs001, rendererClasses [&NiagaraMeshRendererProperties]}
 ]
 userVariables [
  User.Frosting {type=LinearColor, description="", defaultValue {struct=LinearColor, value=<rgba 0.9062 0.0746 0.6196 1>}}
  User.Golden {type=NiagaraBool}
  User.Size {type=NiagaraFloat, defaultValue=0, description=""}
 ]
}
