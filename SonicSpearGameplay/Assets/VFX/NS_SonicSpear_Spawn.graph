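# One-shot Niagara burst for the Sonic Spear spawn, layering cracks, a flare shockwave, lingering flares, a ring, fragments and food debris.
# Spawn groups vary crack and debris looks; debris and fragments fly with drag and gravity, some debris tinted by time of day.
# Curves over particle life drive scale, color and dynamic material parameters.
# User.Size scales most layers; User.Golden turns cracks and linger flare gold and suppresses food debris.
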NiagaraSystem /SonicSpearGameplay/Assets/VFX/NS_SonicSpear_Spawn
dependencies {
 usedDataInterfaces [
  &NiagaraDataInterfaceArrayColor
  &NiagaraDataInterfaceCurve
  &NiagaraDataInterfaceArrayFloat
  &NiagaraDataInterfaceMeshRendererInfo
  &NiagaraDataInterfaceVectorCurve
 ]
 usedDynamicInputs [&UniformRangedFloat]
 usedModules [&SystemState]
}
emitters {
 Cracks {
  emitterUpdateScript [
   &EmitterState()
   &SpawnBurst_Instantaneous(Spawn Count=3, Use Spawn Probability=false, UseLoopCountLimit=false)
  ]
  particleSpawnScript [
   &FN_SetSpawnGroup(Number of Spawn Groups=3)
   &InitializeParticle(
    Color=&MaskLinearColorBySpawnGroup(
     Masked Linear Color=&MakeCustomLinearColorFromBool(Bool=$User.Golden, True Color=<rgba 1 0.428 0 1>, False Color=<rgba 0 0.5 2 4>)
     Passthrough Linear Color=<rgba 1 1 1 1>
     Spawn Group Mask=2
    )
    Lifetime=&SelectFloatFromArray(
     Direct Array Index=&SpawnGroup()
     Float Selection Array=DataInterface() {propertyValues {FloatData [1.36,1.2,1.45]}}
    )
    Mesh Scale Mode=::NewEnumerator3
    Mesh Uniform Scale=&Multiply_Float(
     A=&MaskFloatBySpawnGroup(Masked Float=18, Passthrough Float=12)
     B=&Lerp_Float(A=0.6, B=1, Alpha=$User.Size)
    )
    Sprite Size Mode=::NewEnumerator3
    Uniform Sprite Size=&Lerp_Float(A=1750, B=3500, Alpha=$User.Size)
    UsePositionOffset=false
    Write Material Random=true
   )
   &InitialMeshOrientation(
    Mesh Orientation Mode=::NewEnumerator2
    Orientation Vector=<0 0 1>
    Rotation=&RandomRangeVector(Minimum=<0 0 -1>, Maximum=<0 0 1>)
    Use Rotation Vector=true
   )
   &CameraOffset(Camera Offset Amount=&MaskFloatBySpawnGroup(Masked Float=400, Passthrough Float=100, Spawn Group Mask=2))
   /SetVariables(Particles.VisibilityTag=&SpawnGroup()) {isSetParameters=true}
  ]
  particleUpdateScript [
   &ParticleState()
   &DynamicMaterialParameters(
    Index 0 Param 1=&FloatFromCurve(FloatCurve=DataInterface() {propertyValues {Curve {keys [
         {
          arriveTangent=9.4191
          interpMode=RCIM_Cubic
          leaveTangentWeight=0.1
          tangentMode=RCTM_User
          tangentWeightMode=RCTWM_WeightedBoth
          time=0
          value=0
         }
         {
          arriveTangent=8.3043
          arriveTangentWeight=0.8364
          interpMode=RCIM_Cubic
          leaveTangent=-1.0882
          leaveTangentWeight=0.3448
          tangentMode=RCTM_Break
          tangentWeightMode=RCTWM_WeightedBoth
          time=0.3
          value=1
         }
         {
          arriveTangent=0.009
          arriveTangentWeight=0.4829
          interpMode=RCIM_Cubic
          leaveTangent=0.009
          leaveTangentWeight=0.4829
          tangentMode=RCTM_User
          tangentWeightMode=RCTWM_WeightedBoth
          time=1
          value=0
         }
        ]}}})
    Index 0 Param 2=&FloatFromCurve(
     FloatCurve=DataInterface() {propertyValues {Curve {keys [
         <at 0.4 0.8>
         {arriveTangent=-0.1352, interpMode=RCIM_Cubic, leaveTangent=-0.1352, tangentMode=RCTM_User, time=0.5, value=0.8}
         {arriveTangent=-4.386, interpMode=RCIM_Cubic, leaveTangent=-4.386, tangentMode=RCTM_User, time=1, value=0.2}
        ]}}}
     Scale Curve=&Lerp_Float(A=2, B=3, Alpha=$User.Size)
    )
    Index 0 Param 3=1
    Index 0 Param 4=1
    Index 2 Param 1=1
    Index 2 Param 2=1
    Index 2 Param 3=1
    Index 2 Param 4=1
    Module.Index0Float0Write=true
    Module.Index0Float1Write=true
    Module.Index0Float2Write=true
    Module.Index0Float3Write=true
    Module.Index2Float0Write=true
    Module.Index2Float1Write=true
    Module.Index2Float2Write=true
    Module.Index2Float3Write=true
    Write Parameter Index 2=::NewEnumerator1
   )
   &ScaleColor(
    Scale Alpha=&MaskFloatBySpawnGroup(
     Masked Float=&FloatFromCurve(FloatCurve=DataInterface() {propertyValues {Curve {keys [
          {arriveTangent=11.8795, interpMode=RCIM_Cubic, leaveTangent=11.8795, tangentMode=RCTM_User, time=0, value=0}
          {arriveTangent=-0.3146, leaveTangent=-0.3146, time=0.1, value=3}
          {arriveTangent=-11.223, interpMode=RCIM_Cubic, leaveTangent=-11.223, tangentMode=RCTM_User, time=0.4, value=3}
          {arriveTangent=-0.3888, interpMode=RCIM_Cubic, leaveTangent=-0.3888, tangentMode=RCTM_User, time=1, value=0}
         ]}}})
     Passthrough Float=1
     Spawn Group Mask=2
    )
    ScaleA=true
    ScaleRGB=true
   )
  ]
  rendererClasses [&NiagaraMeshRendererProperties, &NiagaraSpriteRendererProperties]
  renderers [
   {rendererClass=&NiagaraMeshRendererProperties, rendererIndex=0}
   {rendererClass=&NiagaraMeshRendererProperties, rendererIndex=1}
   {rendererClass=&NiagaraSpriteRendererProperties, rendererIndex=2}
  ]
 }
 FlareLightShockwave {
  emitterUpdateScript [
   &EmitterState()
   &SpawnBurst_Instantaneous(Use Spawn Probability=false, UseLoopCountLimit=false)
  ]
  particleSpawnScript [
   &InitializeParticle(
    Lifetime=0.8
    Mesh Scale Mode=::NewEnumerator3
    Mesh Uniform Scale=&Lerp_Float(A=3.35, B=6.5, Alpha=$User.Size)
    Sprite Size Mode=::NewEnumerator3
    Uniform Sprite Size=&Lerp_Float(A=1750, B=2400, Alpha=$User.Size)
    UsePositionOffset=false
    Write Material Random=true
   )
   &ShapeLocation(MaskSpawnGroup=false, Override Seed=false, Sphere Radius=0, UseOwnerScale=true)
  ]
  particleUpdateScript [
   &ParticleState()
   &DynamicMaterialParameters(
    Index 0 Param 1=1
    Index 0 Param 2=&Lerp_Float(
     A=1
     B=0.3
     Alpha=&FloatFromCurve(FloatCurve=DataInterface() {propertyValues {Curve {keys [
          {arriveTangent=5.1433, interpMode=RCIM_Cubic, time=0, value=0}
          {arriveTangent=4.5726, interpMode=RCIM_Cubic, leaveTangent=4.5726, tangentMode=RCTM_User, time=0.25, value=0.7}
          {arriveTangent=-6.4921, interpMode=RCIM_Cubic, leaveTangent=2.9849, tangentMode=RCTM_Break, time=0.4, value=0.35}
          {arriveTangent=0.6574, interpMode=RCIM_Cubic, leaveTangent=0.6574, tangentMode=RCTM_User, time=1, value=1}
         ]}}})
    )
    Index 3 Param 1=&Multiply_Float(A=&MakeFloatFromVector(Vector=$Particles.Scale, Channel=::NewEnumerator0), B=120)
    Index 3 Param 2=&FloatFromCurve(FloatCurve=DataInterface() {propertyValues {Curve {
        defaultValue=0
        keys [
         {arriveTangent=0.2606, interpMode=RCIM_Cubic, leaveTangent=0.2606, tangentMode=RCTM_User, time=0, value=0}
         {interpMode=RCIM_Cubic, leaveTangent=1.82, time=1, value=1}
        ]
       }}})
    Module.Index0Float0Write=true
    Module.Index0Float1Write=true
    Module.Index0Float2Write=true
    Module.Index0Float3Write=true
    Module.Index3Float0Write=true
    Module.Index3Float1Write=true
    Module.Index3Float2Write=true
    Module.Index3Float3Write=true
    Write Parameter Index 3=::NewEnumerator1
   )
   &ScaleColor(
    Color Scale:LinearColor=&LerpLinearColors(
     StartColor=<rgba 0.7113 0.6033 1 0.5>
     EndColor=<rgba 0 1.7486 4 1>
     LerpFactor=&FloatFromCurve(
      FloatCurve=DataInterface() {propertyValues {Curve {keys [
          <at 0.15 1>
          {arriveTangentWeight=0.0997, interpMode=RCIM_Cubic, leaveTangentWeight=0.3488, tangentMode=RCTM_User, time=0.3, value=1.1}
          {interpMode=RCIM_Cubic, leaveTangent=-2.4954, time=1, value=0}
         ]}}}
      Scale Curve=1.25
     )
    )
    Scale Mode=3
   )
  ]
  rendererClasses [&NiagaraMeshRendererProperties]
  renderers [{rendererClass=&NiagaraMeshRendererProperties, rendererIndex=0}]
 }
 Food_Debris {
  emitterUpdateScript [
   &EmitterState()
   &SpawnBurst_Instantaneous(
    Spawn Count=&MakeCustomIntFromBool(Bool=$User.Golden, True INT=0, False INT=&RandomRangeInt(Minimum=9, Maximum=15))
    Use Spawn Probability=false
    UseLoopCountLimit=false
   )
   &SpawnBurst_Instantaneous(
    Spawn Count=&MakeCustomIntFromBool(Bool=$User.Golden, True INT=0, False INT=&RandomRangeInt(Minimum=18, Maximum=28))
    Spawn Group=1
    Use Spawn Probability=false
    UseLoopCountLimit=false
   ) {module=SpawnBurst_Instantaneous001}
   &SpawnBurst_Instantaneous(
    Spawn Count=&MakeCustomIntFromBool(Bool=$User.Golden, True INT=0, False INT=&RandomRangeInt(Minimum=12, Maximum=16))
    Spawn Group=2
    Use Spawn Probability=false
    UseLoopCountLimit=false
   ) {module=SpawnBurst_Instantaneous002}
  ]
  particleSpawnScript [
   /SetVariables(Particles.VisibilityTag=&MaskIntBySpawnGroup(Masked Int=0, Passthrough Int=1, Spawn Group Mask=0)) {isSetParameters=true}
   &InitializeParticle(
    Color=&MakeLinearColorFromVectorAndFloat(
     "Vector (RGB)"=&MaskVector(
      Mask Source=0
      Mask Parameter=?
      Mask=0
      Masked Vector=&MakeVectorFromLinearColorRGB(Color=<rgba 1 1 1 1>)
      Passthrough Vector=&MakeVectorFromLinearColorRGB(Color=&SelectLinearColorFromArray(
        Array Sampling Mode=::NewEnumerator0
        Direct Array Index=?
        Color Selection Array=DataInterface() {propertyValues {ColorData [
           <rgba 1 0.0437 0.0685 1>
           <rgba 0 0.4621 0.9387 1>
           <rgba 0.007 0.5972 0.2016 1>
           <rgba 1 1 0.8308 1>
           <rgba 0.6445 0.314 0.0931 1>
           <rgba 1 1 0 1>
          ]}}
        Random Seed=0
       ))
     )
     "Float (Alpha)"=1
    )
    Lifetime Max=2.5
    Lifetime Min=0.5
    Lifetime Mode=::NewEnumerator1
    Mass=&MaskFloatBySpawnGroup(Masked Float=4, Passthrough Float=1)
    Mass Mode=::NewEnumerator1
    Mesh Scale=&MaskVector(
     Mask Source=0
     Mask Parameter=?
     Mask=0
     Masked Vector=&RandomRangeVector(Minimum=<5 5 5>, Maximum=<9 9 9>)
     Passthrough Vector=&Lerp_Vector(A=<28 15.6 15.6>, B=<36 20 20>, Alpha=$User.Size)
    )
    Mesh Scale Mode=::NewEnumerator4
    UsePositionOffset=false
    Write Material Random=true
   )
   &ShapeLocation(
    MaskSpawnGroup=false
    Offset Mode=::NewEnumerator0
    Override Seed=false
    Rotation Axis=<0 0 1>
    Rotation Mode=::NewEnumerator1
    Sphere Radius=&Lerp_Float(A=85, B=140, Alpha=$User.Size)
    UseOwnerScale=true
   )
   &InitialMeshOrientation(Use Rotation Vector=false)
   &AddVelocity(
    Module.Use Falloff Distance=false
    Override Seed=false
    Velocity Mode=::NewEnumerator1
    Velocity Speed=&MaskFloat(
     Mask Source=0
     Mask Parameter=?
     Mask=0
     Masked Float=750
     Passthrough Float=&MaskFloatBySpawnGroup(Masked Float=1200, Passthrough Float=&RandomRangeFloat(Minimum=200, Maximum=750), Spawn Group Mask=1)
    )
   )
   &AddVelocity(Velocity=&MaskVector(
     Mask Source=0
     Mask Parameter=?
     Mask=0
     Masked Vector=<0 0 0>
     Passthrough Vector=&MaskVectorBySpawnGroup(
      Masked Vector=&RandomRangeVector(Minimum=<0 0 -250>, Maximum=<0 0 250>)
      Passthrough Vector=&RandomRangeVector(Minimum=<0 0 -2500>, Maximum=<0 0 -250>)
      Spawn Group Mask=1
     )
    )) {module=AddVelocity001}
   &AddRotationalVelocity(Rotation Rate=&RandomVector(Vector Scale=2))
   &CameraOffset(Camera Offset Amount=50)
  ]
  particleUpdateScript [
   &ParticleState()
   &Drag(Drag=2, Use Linear Drag=true, Use Rotational Drag=true)
   &GravityForce()
   &SolveForcesAndVelocity(Clamp Velocity=false, Limit Acceleration=false)
   &ScaleMeshSize(Scale Factor=&VectorFromFloat(Value=&FloatFromCurve(FloatCurve=DataInterface() {propertyValues {Curve {
         defaultValue=1
         keys [
          {arriveTangent=38.0665, interpMode=RCIM_Cubic, leaveTangent=38.0665, tangentMode=RCTM_User, time=0, value=0}
          {arriveTangent=1.4286, leaveTangent=1.4286, time=0.07, value=1}
          {arriveTangent=0.1167, interpMode=RCIM_Cubic, leaveTangent=0.1167, tangentMode=RCTM_User, time=0.7, value=1}
          {arriveTangent=-10.8391, interpMode=RCIM_Cubic, leaveTangent=-10.8391, tangentMode=RCTM_User, time=1, value=0}
         ]
        }}})))
   &ScaleColor(
    Color Scale:LinearColor=&MaskLinearColorBySpawnGroup(
     Masked Linear Color=<rgba 1 1 1 1>
     Passthrough Linear Color=&FNTimeOfDayColor(Night Color=<rgba 0.4167 0.4167 0.4167 1>, Morning Color=<rgba 1 1 1 1>, Day Color=<rgba 1 1 1 1>, Evening Color=<rgba 1 1 1 1>)
    )
    Scale Mode=3
   )
  ]
  rendererClasses [&NiagaraMeshRendererProperties]
  renderers [
   {rendererClass=&NiagaraMeshRendererProperties, rendererIndex=0}
   {rendererClass=&NiagaraMeshRendererProperties, rendererIndex=1}
   {rendererClass=&NiagaraMeshRendererProperties, rendererIndex=2}
   {rendererClass=&NiagaraMeshRendererProperties, rendererIndex=3}
  ]
 }
 Fragments {
  emitterUpdateScript [
   &EmitterState()
   &SpawnBurst_Instantaneous(Spawn Count=15, Use Spawn Probability=false, UseLoopCountLimit=true)
  ]
  particleSpawnScript [
   /SetVariables(Particles.SpawnRotationVector=&RandomRangeVector(Minimum=<0.1 1.5 -0.1>, Maximum=<0.4 2 0.1>, Evaluation Type=::NewEnumerator1)) {isSetParameters=true}
   &InitializeParticle(
    Color=<rgba 10 10 10 1>
    Lifetime=&RandomRangeFloat(Minimum=1.5, Maximum=2, Evaluation Type=::NewEnumerator1)
    Mesh Renderer Array Visibility Mode=::NewEnumerator3
    Mesh Renderer Info=DataInterface() {propertyValues {MeshRenderer=&self:Frags1_4.NiagaraMeshRendererProperties_8}}
    Mesh Scale=&Multiply_VectorByFloat(
     Vector=&RandomRangeVector(Minimum=<2 2 3>, Maximum=<5 3 3>, Evaluation Type=::NewEnumerator1)
     Float=&Lerp_Float(A=1, B=2, Alpha=$User.Size)
    )
    Mesh Scale Mode=::NewEnumerator4
    Position Offset=<0 0 -50>
    UsePositionOffset=true
    Write Material Random=true
   )
   &ShapeLocation(
    Hemisphere Distribution=<1 0.5>
    MaskSpawnGroup=false
    Non Uniform Scale=<1 1 -1>
    Override Seed=false
    Sphere Radius=&RandomRangeFloat(Minimum=10, Maximum=25, Evaluation Type=::NewEnumerator1)
    UseOwnerScale=true
   )
   &AddVelocity(Velocity=&Add_Vector(
     A=&Multiply_VectorByFloat(
      Vector=$Particles.ShapeLocation.ShapeVector
      Float=&RandomRangeFloat(Minimum=25, Maximum=80, Evaluation Type=::NewEnumerator1)
     )
     B=&Multiply_VectorByFloat(
      Vector=&RandomRangeVector(Minimum=<0 0 -3500>, Maximum=<0 0 -200>, Evaluation Type=::NewEnumerator1)
      Float=&Lerp_Float(A=0.4, B=1, Alpha=$User.Size)
     )
    ))
   &CameraOffset(Camera Offset Amount=-50)
   &InitialMeshOrientation(Use Rotation Vector=false)
  ]
  particleUpdateScript [
   &ParticleState()
   &Drag(Drag=&RandomRangeFloat(Minimum=2, Maximum=3), Use Linear Drag=true, Use Rotational Drag=true)
   &GravityForce(Gravity=<0 0 -400>)
   &SolveForcesAndVelocity(Clamp Velocity=false, Limit Acceleration=false)
   &UpdateMeshOrientation(
    Rotation Rate=&FloatFromCurve(FloatCurve=DataInterface() {propertyValues {Curve {keys [{arriveTangent=-1.8976, interpMode=RCIM_Cubic, leaveTangent=-1.8976, tangentMode=RCTM_User, time=0, value=1}, <at 1 0 RCIM_Cubic>]}}})
    Rotation Vector=$Particles.SpawnRotationVector
   )
   &ScaleMeshSize(Scale Factor=&VectorFromCurve(VectorCurve=DataInterface() {propertyValues {
       XCurve {keys [<at 0.2 1 RCIM_Cubic>, {arriveTangent=-1.25, interpMode=RCIM_Cubic, leaveTangent=-1.25, time=0.7932, value=0.7509}, <at 1 0>]}
       YCurve {keys [<at 0.2 1 RCIM_Cubic>, {arriveTangent=-0.25, interpMode=RCIM_Cubic, leaveTangent=-0.25, time=0.7932, value=0.85}, <at 1 0.8>]}
       ZCurve {keys [<at 0.2 1 RCIM_Cubic>,<at 0.7932 1 RCIM_Cubic>,<at 1 1>]}
      }}))
   &ScaleColor(
    Scale Alpha=&FloatFromCurve(FloatCurve=DataInterface() {propertyValues {Curve {keys [<at 0.75 1 RCIM_Cubic>,<at 1 0>]}}})
    Scale RGB=&VectorFromFloat(Value=&FloatFromCurve(FloatCurve=DataInterface() {propertyValues {Curve {keys [<at 0.1 25 RCIM_Cubic>, {arriveTangent=-26.6667, interpMode=RCIM_Cubic, leaveTangent=-26.6667, time=0.1961, value=7.5192}, <at 1 1>]}}}))
    ScaleA=true
    ScaleRGB=true
   )
  ]
  rendererClasses [&NiagaraMeshRendererProperties]
  renderers [{rendererClass=&NiagaraMeshRendererProperties, rendererIndex=0}]
 }
 LingerFlare {
  emitterUpdateScript [
   &EmitterState()
   &SpawnBurst_Instantaneous(Use Spawn Probability=false, UseLoopCountLimit=true)
  ]
  particleSpawnScript [
   &InitializeParticle(
    Color=&MakeCustomLinearColorFromBool(Bool=$User.Golden, True Color=<rgba 2 0.8559 0 1>, False Color=<rgba 0 0.9717 2 4>)
    Lifetime=1.4
    Sprite Size=&Lerp_Vector2(A=<500 200>, B=<1500 400>, Alpha=$User.Size)
    Sprite Size Mode=::NewEnumerator4
    UsePositionOffset=false
    Write Material Random=true
   )
   &ShapeLocation(MaskSpawnGroup=false, Override Seed=false, Sphere Radius=0, UseOwnerScale=true)
   &CameraOffset(Camera Offset Amount=400)
  ]
  particleUpdateScript [
   &ParticleState()
   &ScaleSpriteSize(
    Uniform Curve Scale=4
    Uniform Curve Sprite Scale=DataInterface() {propertyValues {Curve {keys [
        {arriveTangent=21.6769, interpMode=RCIM_Cubic, leaveTangent=21.6769, tangentMode=RCTM_User, time=0, value=0.2}
        <at 0.1 1>
        {arriveTangentWeight=0.2899, interpMode=RCIM_Cubic, leaveTangentWeight=0.1115, tangentMode=RCTM_User, time=0.75, value=1}
        {arriveTangent=-8.9452, interpMode=RCIM_Cubic, leaveTangent=-8.9452, tangentMode=RCTM_User, time=1, value=0.2}
       ]}}}
   )
   &ScaleColor(
    Scale Alpha=&FloatFromCurve(FloatCurve=DataInterface() {propertyValues {Curve {keys [
         {arriveTangent=21.6769, interpMode=RCIM_Cubic, leaveTangent=21.6769, tangentMode=RCTM_User, time=0, value=0}
         <at 0.1 1>
         {arriveTangent=-1.1111, interpMode=RCIM_Cubic, leaveTangent=-1.1111, time=0.75, value=1}
         <at 1 0>
        ]}}})
    ScaleA=true
    ScaleRGB=false
   )
   /SetVariables(
    Particles.Flare Color:LinearColor=&MultiplyLinearColor(
     Linear Color=&MakeCustomLinearColorFromBool(Bool=$User.Golden, True Color=<rgba 1.2 0.5136 0 1>, False Color=<rgba 0 0.9717 2 4>)
     Scale Factor=&MakeVector4FromVectorAndFloat(
      "Vector (RGB)"=&VectorFromFloat(Value=&FloatFromCurve(FloatCurve=DataInterface() {propertyValues {Curve {keys [{arriveTangent=-8.475, interpMode=RCIM_Cubic, leaveTangent=-8.475, tangentMode=RCTM_User, time=0, value=5}, <at 1 1 RCIM_Cubic>]}}}))
      "Float (Alpha)"=&FloatFromCurve(FloatCurve=DataInterface() {propertyValues {Curve {keys [<at 0 1 RCIM_Cubic>, {arriveTangent=-1.9891, interpMode=RCIM_Cubic, leaveTangent=-1.9891, tangentMode=RCTM_User, time=1, value=0}]}}})
     )
    )
    Particles.Flare Size:Vector2f=&Vector2DFromFloat(Value=&Multiply_Float(
      A=&Lerp_Float(A=750, B=1500, Alpha=$User.Size)
      B=&FloatFromCurve(FloatCurve=DataInterface() {propertyValues {Curve {keys [
           {arriveTangent=16.2971, interpMode=RCIM_Cubic, leaveTangent=16.2971, tangentMode=RCTM_User, time=0, value=0}
           {arriveTangentWeight=0.0479, interpMode=RCIM_Cubic, leaveTangentWeight=0.1858, tangentMode=RCTM_User, time=0.05, value=1}
           {arriveTangent=-2.3265, interpMode=RCIM_Cubic, leaveTangent=-2.3265, tangentMode=RCTM_User, time=1, value=0}
          ]}}})
     ))
   ) {isSetParameters=true}
   /SetVariables(
    Particles.Corona Color:LinearColor=&MultiplyLinearColor(
     Linear Color=&MakeCustomLinearColorFromBool(Bool=$User.Golden, True Color=<rgba 10 4.2797 0 1>, False Color=<rgba 0 0.9717 2 4>)
     Scale Factor=&MakeVector4FromVectorAndFloat(
      "Vector (RGB)"=<1 1 1>
      "Float (Alpha)"=&FloatFromCurve(FloatCurve=DataInterface() {propertyValues {Curve {keys [<at 0 0 RCIM_Cubic>,<at 0.25 1 RCIM_Cubic>,<at 1 0 RCIM_Cubic>]}}})
     )
    )
    Particles.Corona Dynm:Vector4f=&MakeVector4(
     X=&FloatFromCurve(FloatCurve=DataInterface() {propertyValues {Curve {keys [{arriveTangent=-2.7401, interpMode=RCIM_Cubic, leaveTangent=-2.7401, tangentMode=RCTM_User, time=0, value=2}, <at 1 0.7 RCIM_Cubic>]}}})
     Y=0.4
     Z=0.6
     W=1
    )
    Particles.Corona Size:Vector2f=&Vector2DFromFloat(Value=&Lerp_Float(A=1200, B=3000, Alpha=$User.Size))
   ) {isSetParameters=true}
  ]
  rendererClasses [&NiagaraSpriteRendererProperties]
  renderers [
   {rendererClass=&NiagaraSpriteRendererProperties}
   {rendererClass=&NiagaraSpriteRendererProperties, rendererIndex=1}
   {rendererClass=&NiagaraSpriteRendererProperties, rendererIndex=2}
  ]
 }
 Ring {
  emitterUpdateScript [
   &EmitterState()
   &SpawnBurst_Instantaneous(Use Spawn Probability=false, UseLoopCountLimit=true)
  ]
  particleSpawnScript [
   &InitializeParticle(
    Color=<rgba 0 0.6775 1 0.5>
    Lifetime=0.3
    Mesh Scale=<12 6.2 6.2>
    Mesh Scale Mode=::NewEnumerator4
    Sprite Rotation Mode=::NewEnumerator2
    Sprite Size=<2500 2500>
    Sprite Size Mode=::NewEnumerator4
    UsePositionOffset=false
    Write Material Random=true
   )
   &CameraOffset(Camera Offset Amount=15)
   &InitialMeshOrientation(
    Mesh Orientation Mode=::NewEnumerator2
    Rotation=&RandomRangeVector(Minimum=<-1 0 0>, Maximum=<1 0 0>)
    Use Rotation Vector=true
   )
  ]
  particleUpdateScript [
   &ParticleState()
   &DynamicMaterialParameters(
    Index 0 Param 1=1
    Index 0 Param 2=1
    Index 0 Param 3=1
    Index 0 Param 4=1
    Index 1 Param 1=1
    Index 2 Param 1=&RandomRangeFloat(Minimum=2, Maximum=20)
    Index 2 Param 2=&RandomRangeFloat(Minimum=0.1, Maximum=1)
    Index 2 Param 3=1
    Index 2 Param 4=&FloatFromCurve(FloatCurve=DataInterface() {propertyValues {Curve {keys [
         <at 0 1 RCIM_Cubic>
         {arriveTangent=-2, interpMode=RCIM_Cubic, leaveTangent=-2, time=0.4, value=0.25}
         {arriveTangent=-2.1199, interpMode=RCIM_Cubic, leaveTangent=-2.1199, tangentMode=RCTM_User, time=1, value=-1}
        ]}}})
    Index 3 Param 1=&FloatFromCurve(FloatCurve=DataInterface() {propertyValues {Curve {keys [<at 0.45 0>,<at 1 1>]}}})
    Index 3 Param 2=45
    Module.Index0Float0Write=true
    Module.Index0Float1Write=true
    Module.Index0Float2Write=true
    Module.Index0Float3Write=true
    Module.Index1Float0Write=true
    Module.Index1Float1Write=true
    Module.Index1Float2Write=true
    Module.Index1Float3Write=true
    Module.Index2Float0Write=true
    Module.Index2Float1Write=true
    Module.Index2Float2Write=true
    Module.Index2Float3Write=true
    Module.Index3Float0Write=true
    Module.Index3Float1Write=true
    Module.Index3Float2Write=true
    Module.Index3Float3Write=true
    Write Parameter Index 1=::NewEnumerator1
    Write Parameter Index 2=::NewEnumerator1
    Write Parameter Index 3=::NewEnumerator1
   )
   &ScaleSpriteSize(
    Uniform Curve Scale=&Lerp_Float(A=1, B=1.5, Alpha=$User.Size)
    Uniform Curve Sprite Scale=DataInterface() {propertyValues {Curve {keys [{interpMode=RCIM_Cubic, tangentMode=RCTM_User, time=0, value=1}, {interpMode=RCIM_Cubic, tangentMode=RCTM_User, time=1, value=0}]}}}
   )
   &ScaleColor(
    Scale Alpha=&FloatFromCurve(FloatCurve=DataInterface() {propertyValues {Curve {keys [
         {arriveTangent=-7.8941, interpMode=RCIM_Cubic, leaveTangent=-7.8941, tangentMode=RCTM_User, time=0, value=3}
         {interpMode=RCIM_Cubic, tangentMode=RCTM_User, time=1, value=0}
        ]}}})
    ScaleA=true
    ScaleRGB=true
   )
   &ScaleMeshSize(Scale Factor=&VectorFromFloat(Value=&FloatFromCurve(
      FloatCurve=DataInterface() {propertyValues {Curve {
         defaultValue=0.35
         keys [
          {arriveTangent=4.1785, interpMode=RCIM_Cubic, leaveTangent=4.1785, tangentMode=RCTM_User, time=0, value=0}
          {arriveTangent=1.1179, interpMode=RCIM_Cubic, leaveTangent=1.1179, tangentMode=RCTM_User, time=0.2042, value=0.6228}
          {arriveTangent=0.0224, interpMode=RCIM_Cubic, leaveTangent=0.0224, tangentMode=RCTM_User, time=1, value=1}
         ]
        }}}
      Scale Curve=&Lerp_Float(A=1, B=2, Alpha=$User.Size)
     )))
   /SetVariables(Particles.Mesh ring Color:LinearColor=<rgba 0.0133 0.0686 0.1719 0.5>) {isSetParameters=true}
  ]
  rendererClasses [&NiagaraSpriteRendererProperties, &NiagaraMeshRendererProperties]
  renderers [
   {rendererClass=&NiagaraSpriteRendererProperties}
   {rendererClass=&NiagaraMeshRendererProperties, rendererIndex=1}
  ]
 }
}
summary {
 systemName=NS_SonicSpear_Spawn
 emitters [
  {emitterName=Food_Debris, rendererClasses [&NiagaraMeshRendererProperties]}
  {emitterName=FlareLightShockwave, rendererClasses [&NiagaraMeshRendererProperties]}
  {emitterName=LingerFlare, rendererClasses [&NiagaraSpriteRendererProperties]}
  {emitterName=Cracks, rendererClasses [&NiagaraMeshRendererProperties, &NiagaraSpriteRendererProperties]}
  {emitterName=Fragments, rendererClasses [&NiagaraMeshRendererProperties]}
  {emitterName=Ring, rendererClasses [&NiagaraSpriteRendererProperties, &NiagaraMeshRendererProperties]}
 ]
 userVariables [
  User.Golden {type=NiagaraBool}
  User.Size {type=NiagaraFloat, defaultValue=0, description=""}
 ]
}
