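# One-shot stereo MetaSound source for the Sonic Spear's water impact sound.
# On play, a random pick from the Water Impact wave array feeds a stereo wave player, which also signals when the sound finishes.
# The Golden, Size and attenuation distance inputs are not wired into the graph.
# Playback uses impact attenuation, the Fort_Impacts sound class, a tight concurrency override with retrigger time, and CB_SonicSpear volume modulation.
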MetaSoundSource /SonicSpearGameplay/Assets/Audio/MSS_SonicSpear_Impact_Water
inputs [
 Golden {type=Bool}
 Size {type=Float}
 UE.Attenuation.Distance {type=Float}
 UE.Source.OnPlay {type=Trigger}
 Water Impact {type=WaveAsset:Array}
]
interfaces [UE.Attenuation:1.0, UE.OutputFormat.Stereo:1.0, UE.Source.OneShot:1.0, UE.Source:1.1]
outputs [
 "UE.OutputFormat.Stereo.Audio:0" {type=Audio}
 "UE.OutputFormat.Stereo.Audio:1" {type=Audio}
 UE.Source.OneShot.OnFinished {type=Trigger}
]
pages [Default {
  input_defaults {
   Golden=null
   Size=0
   UE.Attenuation.Distance=0
   UE.Source.OnPlay=null
   Water Impact [/SonicSpearGameplay/Assets/Audio/wav/SonicSpear_Impact_Water_01, /SonicSpearGameplay/Assets/Audio/wav/SonicSpear_Impact_Water_02]
  }
  nodes [
   ArrayRandomGet_0 = Array.Random Get:WaveAsset:Array(Next=^UE.Source.OnPlay, In Array=^Water Impact) {also {defaults {Weights []}}}
   WavePlayer_0 = UE.Wave Player:Stereo(Play=@ArrayRandomGet_0.On Next, Wave Asset=@ArrayRandomGet_0.Value) {also {version [1,1]}}
  ]
  links [
   {dst="@UE.Source.OneShot.OnFinished", src=WavePlayer_0.On Finished}
   {dst="@UE.OutputFormat.Stereo.Audio:0", src=WavePlayer_0.Out Left}
   {dst="@UE.OutputFormat.Stereo.Audio:1", src=WavePlayer_0.Out Right}
  ]
 }]
properties {
 assetImportData=&self:AssetImportData
 attenuationSettings=&/ATT_SonicSpear_Impact
 bOverrideConcurrency=true
 maxDistance=132000
 numChannels=2
 outputFormat=Stereo
 sampleRate=48000
 soundClassObject=&Fort_Impacts
 concurrencyOverrides {maxCount=2, retriggerTime=0.4, volumeScale=0.8}
 modulationSettings {volumeRouting=Override, volumeModulationDestination {bEnableModulation=true, modulators [&CB_SonicSpear]}}
}
